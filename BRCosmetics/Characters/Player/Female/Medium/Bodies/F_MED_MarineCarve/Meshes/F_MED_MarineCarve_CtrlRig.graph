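# Control Rig for the F_MED_MarineCarve female character body that applies procedural corrections to its t-shirt, side bag, bandolier, sleeves, shoe tongues and wrist.
# Parent constraints blend shirt bases between pelvis and thighs, and attach pins to weighted upper-body bones.
# Joint delta rotations, spherical pose readers and animation attributes weight additive bone offsets.
# State variables for crouching, sliding, emotes and weapon grip add further corrections, many smoothed with alpha interpolation.
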ControlRigBlueprint /BRCosmetics/Characters/Player/Female/Medium/Bodies/F_MED_MarineCarve/Meshes/F_MED_MarineCarve_CtrlRig
graphs [
 CollapseNode_1_ContainedGraph {nodes [
   RigVMFunctionEntryNode()
   n1 = RigVMVariableNode() {variable=CrouchingFloat}
   > &RigUnit_ModifyTransforms(
    ItemToModify="((Item=(Type=Bone,Name=\"dyn_tshirt_side_l_01\"),Transform=(Rotation=(X=0,Y=-0.1736,Z=0,W=0.9848),Translation=(X=0,Y=0,Z=0),Scale3D=(X=1,Y=1,Z=1))),(Item=(Type=Bone,Name=\"dyn_tshirt_front_l_01\"),Transform=(Rotation=(X=0,Y=0,Z=-0.1736,W=0.9848),Translation=(X ...(truncated)"
    WeightMinimum=0
    Weight=@n1.Value
   )
   n2 = &RigVMFunction_AnimEvalRichCurve(
    Curve=(EditorCurveData=(Keys=((InterpMode=RCIM_Cubic),(InterpMode=RCIM_Cubic,Time=1,Value=1))))
    SourceMinimum=0
    TargetMinimum=0
    Value=&RigVMFunction_MathDoubleRemap(
     SourceMaximum=75
     SourceMinimum=45
     TargetMaximum=1
     TargetMinimum=0
     bClamp=false
     Value=&RigUnit_GetRelativeTransformForItem(Child=(Type=Bone,Name="pelvis"), Parent=(Type=Bone,Name="root"), bChildInitial=False, bParentInitial=False).RelativeTransform.Translation.Z
    ).Result
   )
   Multiply_2 = RigVMAggregateNode(B=1, C=@n2.Result, A=RigVMVariableNode() {variable=CrouchingFloat}.Value)
   > &RigUnit_ModifyTransforms(
    ItemToModify="((Item=(Type=Bone,Name=\"dyn_tshirt_front_r_01\"),Transform=(Rotation=(X=0.0442,Y=-0.0098,Z=0.076,W=-0.9961),Translation=(X=0,Y=0,Z=0),Scale3D=(X=1,Y=1,Z=1))),(Item=(Type=Bone,Name=\"dyn_tshirt_front_01\"),Transform=(Rotation=(X=-0.2493,Y=-0.0116,Z=0.1003,W=-0 ...(truncated)"
    WeightMinimum=0
    Weight=@Multiply_2.Result
   )
   > RigVMFunctionReturnNode()
   &RigVMFunction_MathDoubleMul(B=0.7, A=@n1.Value)
  ]}
 CollapseNode_ContainedGraph {nodes [
   n1 = RigVMFunctionEntryNode()
   n2 = &RigVMFunction_AlphaInterp(
    InterpSpeedDecreasing=8
    InterpSpeedIncreasing=8
    Scale=1
    bClampResult=False
    Value=&RigVMFunction_MathBoolToFloat(Value=RigVMVariableNode() {variable=PlayingEmote}.Value).Result
   )
   &RigUnit_ModifyTransforms(
    ItemToModify="((Item=(Type=Bone,Name=\"base_sideBag_01\"),Transform=(Rotation=(X=0,Y=0,Z=0.0872,W=0.9962),Translation=(X=0,Y=0,Z=0),Scale3D=(X=1,Y=1,Z=1))),(Item=(Type=Bone,Name=\"dyn_sideBag_02\"),Transform=(Rotation=(X=0,Y=0,Z=0.0872,W=0.9962),Translation=(X=0,Y=0,Z=0),Sc ...(truncated)"
    WeightMinimum=0
    Weight=@n2.Result
    ExecutePin=@n1.ExecutePin
   )
   > &RigUnit_ModifyTransforms(
    ItemToModify=((Item=(Type=Bone,Name="dyn_sideBag_01"),Transform=(Rotation=(X=0,Y=0,Z=-0.0436,W=0.999),Translation=(X=0,Y=0,Z=0),Scale3D=(X=1,Y=1,Z=1))))
    Weight=1
    WeightMinimum=0
   )
   > n3 = &RigUnit_SphericalPoseReader(
    ActiveRegionScaleFactors=(PositiveWidth=1,NegativeWidth=1,PositiveHeight=1,NegativeHeight=1)
    ActiveRegionSize=0.05
    Debug=(bDrawDebug=false,bDraw2D=False,bDrawLocalAxes=False,DebugScale=25,DebugSegments=20,DebugThickness=0.25)
    DriverAxis=(X=1,Y=0,Z=0)
    DriverItem=(Type=Bone,Name="dyn_sideBag_01")
    FalloffRegionScaleFactors=(PositiveWidth=1,NegativeWidth=1,PositiveHeight=1,NegativeHeight=1)
    FalloffSize=0.45
    FlipHeightScaling=False
    FlipWidthScaling=False
    OptionalParentItem=(Type=Bone,Name="None")
    RotationOffset=(X=0,Y=-10,Z=0)
   )
   > &RigUnit_ModifyTransforms(
    ItemToModify=((Item=(Type=Bone,Name="dyn_bandolierFront_01"),Transform=(Rotation=(X=0,Y=0.6275,Z=0,W=-0.7786),Translation=(X=0,Y=0,Z=0),Scale3D=(X=1,Y=1,Z=1))))
    WeightMinimum=0
    Weight=@n3.OutputParam
   )
   > n4 = &RigUnit_SphericalPoseReader(
    ActiveRegionScaleFactors=(PositiveWidth=1,NegativeWidth=1,PositiveHeight=1,NegativeHeight=1)
    ActiveRegionSize=0.05
    Debug=(bDrawDebug=false,bDraw2D=False,bDrawLocalAxes=False,DebugScale=25,DebugSegments=20,DebugThickness=0.25)
    DriverAxis=(X=1,Y=0,Z=0)
    DriverItem=(Type=Bone,Name="dyn_sideBag_01")
    FalloffRegionScaleFactors=(PositiveWidth=1,NegativeWidth=1,PositiveHeight=1,NegativeHeight=1)
    FalloffSize=0.35
    FlipHeightScaling=False
    FlipWidthScaling=False
    OptionalParentItem=(Type=Bone,Name="None")
    RotationOffset=(X=0,Y=-160,Z=0)
   )
   > n5 = &RigUnit_ModifyTransforms(
    ItemToModify=((Item=(Type=Bone,Name="dyn_bandolierBack_01"),Transform=(Rotation=(X=0.2842,Y=0.8019,Z=0.193,W=0.4888),Translation=(X=0,Y=0,Z=0),Scale3D=(X=1,Y=1,Z=1))))
    WeightMinimum=0
    Weight=@n4.OutputParam
   )
   RigVMFunctionReturnNode(ExecutePin=@n5.ExecutePin)
   &RigUnit_ModifyTransforms(
    ExecutePin=()
    ItemToModify=((Item=(Type=Bone,Name="dyn_sideBag_02"),Transform=(Rotation=(X=0,Y=0,Z=0,W=1),Translation=(X=50,Y=0,Z=0),Scale3D=(X=1,Y=1,Z=1))))
    Weight=1
    WeightMinimum=0
   )
  ]}
 CollapseNode_ContainedGraph {nodes [
   n1 = RigVMFunctionEntryNode()
   &RigUnit_ParentConstraint(
    AdvancedSettings=(InterpolationType=Average,RotationOrderForFilter=XZY)
    Child=(Type=Bone,Name="base_tshirt_front_l_01")
    Filter=(TranslationFilter=(bX=false,bY=false,bZ=false),RotationFilter=(bX=True,bY=True,bZ=True),ScaleFilter=(bX=false,bY=false,bZ=false))
    Parents=((Item=(Type=Bone,Name="pelvis"),Weight=0.3),(Item=(Type=Bone,Name="thigh_l"),Weight=0.7))
    Weight=1
    ExecutePin=@n1.ExecutePin
   )
   > &RigUnit_ParentConstraint(
    AdvancedSettings=(InterpolationType=Average,RotationOrderForFilter=XZY)
    Child=(Type=Bone,Name="base_tshirt_side_l_01")
    Filter=(TranslationFilter=(bX=false,bY=false,bZ=false),RotationFilter=(bX=True,bY=True,bZ=True),ScaleFilter=(bX=false,bY=false,bZ=false))
    Parents=((Item=(Type=Bone,Name="pelvis"),Weight=0.5),(Item=(Type=Bone,Name="thigh_l"),Weight=0.5))
    Weight=1
   )
   > &RigUnit_ParentConstraint(
    AdvancedSettings=(InterpolationType=Average,RotationOrderForFilter=XZY)
    Child=(Type=Bone,Name="base_tshirt_rear_l_01")
    Filter=(TranslationFilter=(bX=false,bY=false,bZ=false),RotationFilter=(bX=True,bY=True,bZ=True),ScaleFilter=(bX=false,bY=false,bZ=false))
    Parents=((Item=(Type=Bone,Name="pelvis"),Weight=0.9),(Item=(Type=Bone,Name="thigh_l"),Weight=0.1))
    Weight=1
   )
   > &RigUnit_ParentConstraint(
    AdvancedSettings=(InterpolationType=Average,RotationOrderForFilter=XZY)
    Child=(Type=Bone,Name="base_tshirt_front_r_01")
    Filter=(TranslationFilter=(bX=false,bY=false,bZ=false),RotationFilter=(bX=True,bY=True,bZ=True),ScaleFilter=(bX=false,bY=false,bZ=false))
    Parents=((Item=(Type=Bone,Name="pelvis"),Weight=0.3),(Item=(Type=Bone,Name="thigh_r"),Weight=0.7))
    Weight=1
   )
   > &RigUnit_ParentConstraint(
    AdvancedSettings=(InterpolationType=Average,RotationOrderForFilter=XZY)
    Child=(Type=Bone,Name="base_tshirt_side_r_01")
    Filter=(TranslationFilter=(bX=false,bY=false,bZ=false),RotationFilter=(bX=True,bY=True,bZ=True),ScaleFilter=(bX=false,bY=false,bZ=false))
    Parents=((Item=(Type=Bone,Name="pelvis"),Weight=0.5),(Item=(Type=Bone,Name="thigh_r"),Weight=0.5))
    Weight=1
   )
   > &RigUnit_ParentConstraint(
    AdvancedSettings=(InterpolationType=Average,RotationOrderForFilter=XZY)
    Child=(Type=Bone,Name="base_tshirt_rear_r_01")
    Filter=(TranslationFilter=(bX=false,bY=false,bZ=false),RotationFilter=(bX=True,bY=True,bZ=True),ScaleFilter=(bX=false,bY=false,bZ=false))
    Parents=((Item=(Type=Bone,Name="pelvis"),Weight=0.9),(Item=(Type=Bone,Name="thigh_r"),Weight=0.1))
    Weight=1
   )
   > &RigUnit_ParentConstraint(
    AdvancedSettings=(RotationOrderForFilter=XZY)
    Child=(Type=Bone,Name="base_pin_01")
    Parents="((Item=(Type=Bone,Name=\"neck_01\"),Weight=0.0019),(Item=(Type=Bone,Name=\"deform_pec_r\"),Weight=0.0035),(Item=(Type=Bone,Name=\"dyn_bandolierFront_01\"),Weight=0.0394),(Item=(Type=Bone,Name=\"clavicle_r\"),Weight=0.0467),(Item=(Type=Bone,Name=\"spine_04\"),Weight= ...(truncated)"
    Weight=1
   )
   > n2 = &RigUnit_ParentConstraint(
    AdvancedSettings=(RotationOrderForFilter=XZY)
    Child=(Type=Bone,Name="base_pin_02")
    Parents="((Item=(Type=Bone,Name=\"spine_05\"),Weight=0.003),(Item=(Type=Bone,Name=\"deform_pec_l\"),Weight=0.0079),(Item=(Type=Bone,Name=\"clavicle_r\"),Weight=0.0199),(Item=(Type=Bone,Name=\"dyn_bandolierFront_01\"),Weight=0.104),(Item=(Type=Bone,Name=\"deform_pec_r\"),Weig ...(truncated)"
    Weight=1
   )
   RigVMFunctionReturnNode(ExecutePin=@n2.ExecutePin)
   &RigUnit_ParentConstraint(
    AdvancedSettings=(InterpolationType=Average,RotationOrderForFilter=XZY)
    Child=(Type=Bone,Name="base_tshirt_front_01")
    ExecutePin=()
    Filter=(TranslationFilter=(bX=false,bY=false,bZ=false),RotationFilter=(bX=True,bY=True,bZ=True),ScaleFilter=(bX=false,bY=false,bZ=false))
    Parents=((Item=(Type=Bone,Name="base_tshirt_front_l_01"),Weight=0.5),(Item=(Type=Bone,Name="base_tshirt_front_r_01"),Weight=0.5))
    Weight=1
   )
   > &RigUnit_ParentConstraint(
    AdvancedSettings=(InterpolationType=Average,RotationOrderForFilter=XZY)
    Child=(Type=Bone,Name="base_tshirt_rear_01")
    Filter=(TranslationFilter=(bX=false,bY=false,bZ=false),RotationFilter=(bX=True,bY=True,bZ=True),ScaleFilter=(bX=false,bY=false,bZ=false))
    Parents=((Item=(Type=Bone,Name="base_tshirt_rear_r_01"),Weight=0.5),(Item=(Type=Bone,Name="base_tshirt_rear_l_01"),Weight=0.5))
    Weight=1
   )
  ]}
 CollapseNode_ContainedGraph {nodes [
   RigVMFunctionEntryNode()
   > RigVMFunction_Sequence_2 = RigVMAggregateNode()
   DeltaRotation = RigVMFunctionReferenceNode(
    Rotation Order=ZYX
    child=(Type=Bone,Name="thigh_l")
    debugRotation=false
    outputMinMax=(X=0,Y=1)
    parent=(Type=Null,Name="thigh_l_null")
    premult=1
    rotationMinMax=(X=-5,Y=-90)
    triggerAxis=Z
    twistAxis=(X=1,Y=0,Z=0)
    ExecuteContext=@RigVMFunction_Sequence_2.A
   )
   Get_Animation_Attribute = RigVMDispatchNode(BoneName=root, Default=0, Name=deformRig_l_thigh_up_alpha) {template="DISPATCH_RigDispatch_GetAnimAttribute(in Name,in BoneName,in Default,out Value,out Found)"}
   AlphaInterp = &RigVMFunction_AlphaInterp(InterpSpeedDecreasing=10, InterpSpeedIncreasing=10, Scale=1, bClampResult=False, bInterpResult=False, Value=@Get_Animation_Attribute.Value)
   > &RigUnit_ModifyTransforms(
    ItemToModify="((Item=(Type=Bone,Name=\"dyn_tshirt_front_l_01\"),Transform=(Rotation=(X=-0.0273,Y=0.0279,Z=-0.3247,W=0.945),Translation=(X=-0.9424,Y=-1.0481,Z=-5.9604),Scale3D=(X=1,Y=1,Z=1))),(Item=(Type=Bone,Name=\"dyn_tshirt_front_01\"),Transform=(Rotation=(X=-0.3807,Y=0.2 ...(truncated)"
    WeightMinimum=0
    Weight=@AlphaInterp.Result
   )
   > DeltaRotation_1 = RigVMFunctionReferenceNode(
    Rotation Order=ZYX
    child=(Type=Bone,Name="thigh_r")
    outputMinMax=(X=0,Y=1)
    parent=(Type=Null,Name="thigh_r_null")
    premult=1
    rotationMinMax=(X=0,Y=-90)
    triggerAxis=Z
    twistAxis=(X=1,Y=0,Z=0)
   )
   Get_Animation_Attribute_1 = RigVMDispatchNode(BoneName=root, Default=0, Name=deformRig_r_thigh_up_alpha) {template="DISPATCH_RigDispatch_GetAnimAttribute(in Name,in BoneName,in Default,out Value,out Found)"}
   n1 = &RigVMFunction_AlphaInterp(InterpSpeedDecreasing=10, InterpSpeedIncreasing=10, Scale=1, bClampResult=False, bInterpResult=False, Value=@Get_Animation_Attribute_1.Value)
   > &RigUnit_ModifyTransforms(
    ItemToModify="((Item=(Type=Bone,Name=\"dyn_tshirt_front_r_01\"),Transform=(Rotation=(X=0.035,Y=-0.0313,Z=-0.3409,W=0.9389),Translation=(X=0,Y=0,Z=0),Scale3D=(X=1,Y=1,Z=1))),(Item=(Type=Bone,Name=\"dyn_tshirt_front_01\"),Transform=(Rotation=(X=0.3154,Y=-0.162,Z=-0.5363,W=0.7 ...(truncated)"
    WeightMinimum=0
    Weight=@n1.Result
   )
   Multiply = &RigVMFunction_MathDoubleMul(A=@AlphaInterp.Result, B=@n1.Result)
   > &RigUnit_ModifyTransforms(
    ItemToModify="((Item=(Type=Bone,Name=\"dyn_tshirt_front_01\"),Transform=(Rotation=(X=-0.116,Y=0.5016,Z=0.3512,W=0.782),Translation=(X=0,Y=0,Z=0),Scale3D=(X=1,Y=1,Z=1))),(Item=(Type=Bone,Name=\"dyn_tshirt_front_l_01\"),Transform=(Rotation=(X=-0.006,Y=-0.0374,Z=-0.1803,W=-0.9 ...(truncated)"
    WeightMinimum=0
    Weight=@Multiply.Result
   )
   > DeltaRotation_0 = RigVMFunctionReferenceNode(
    Rotation Order=ZYX
    child=(Type=Bone,Name="thigh_l")
    debugRotation=false
    outputMinMax=(X=0,Y=1)
    parent=(Type=Null,Name="thigh_l_null")
    premult=1
    rotationMinMax=(X=-89,Y=-120)
    triggerAxis=Z
    twistAxis=(X=1,Y=0,Z=0)
   )
   > &RigUnit_ModifyTransforms(
    ItemToModify="((Item=(Type=Bone,Name=\"base_tshirt_front_l_01\"),Transform=(Rotation=(X=0,Y=0,Z=0.0175,W=0.9998),Translation=(X=-9,Y=-8,Z=3),Scale3D=(X=1,Y=1,Z=1))),(Item=(Type=Bone,Name=\"base_tshirt_front_01\"),Transform=(Rotation=(X=0,Y=0,Z=-0.1822,W=0.9833),Translation= ...(truncated)"
    WeightMinimum=0
    Weight=@DeltaRotation_0.outputWeight
   )
   > DeltaRotation_10 = RigVMFunctionReferenceNode(
    Rotation Order=ZYX
    child=(Type=Bone,Name="thigh_r")
    debugRotation=false
    outputMinMax=(X=0,Y=1)
    parent=(Type=Null,Name="thigh_r_null")
    premult=1
    rotationMinMax=(X=-89,Y=-120)
    triggerAxis=Z
    twistAxis=(X=1,Y=0,Z=0)
   )
   > &RigUnit_ModifyTransforms(
    ItemToModify="((Item=(Type=Bone,Name=\"base_tshirt_front_r_01\"),Transform=(Rotation=(X=0.0081,Y=0.0546,Z=0.1725,W=0.9835),Translation=(X=-11,Y=-11,Z=-3),Scale3D=(X=1,Y=1,Z=1.5214))),(Item=(Type=Bone,Name=\"base_tshirt_front_01\"),Transform=(Rotation=(X=0,Y=0,Z=0,W=-1),Tran ...(truncated)"
    WeightMinimum=0
    Weight=@DeltaRotation_10.outputWeight
   )
   DeltaRotation_2 = RigVMFunctionReferenceNode(
    Rotation Order=ZYX
    child=(Type=Bone,Name="thigh_l")
    outputMinMax=(X=0,Y=1)
    parent=(Type=Null,Name="thigh_l_null")
    premult=1
    rotationMinMax=(X=0,Y=90)
    triggerAxis=Z
    twistAxis=(X=1,Y=0,Z=0)
    ExecuteContext=@RigVMFunction_Sequence_2.B
   )
   > &RigUnit_ModifyTransforms(
    ItemToModify="((Item=(Type=Bone,Name=\"dyn_tshirt_front_l_01\"),Transform=(Rotation=(X=0.0661,Y=0.099,Z=0.4432,W=-0.8885),Translation=(X=0,Y=0,Z=0),Scale3D=(X=1,Y=1,Z=1))),(Item=(Type=Bone,Name=\"dyn_tshirt_rear_l_01\"),Transform=(Rotation=(X=-0.1274,Y=-0.3355,Z=0.5773,W=-0 ...(truncated)"
    WeightMinimum=0
    Weight=@DeltaRotation_2.outputWeight
   )
   > DeltaRotation_3 = RigVMFunctionReferenceNode(
    Rotation Order=ZYX
    child=(Type=Bone,Name="thigh_r")
    outputMinMax=(X=0,Y=1)
    parent=(Type=Null,Name="thigh_r_null")
    premult=1
    rotationMinMax=(X=0,Y=90)
    triggerAxis=Z
    twistAxis=(X=1,Y=0,Z=0)
   )
   > &RigUnit_ModifyTransforms(
    ItemToModify="((Item=(Type=Bone,Name=\"dyn_tshirt_front_r_01\"),Transform=(Rotation=(X=0.0526,Y=0.0749,Z=-0.4987,W=0.8619),Translation=(X=0,Y=0,Z=0),Scale3D=(X=1,Y=1,Z=1))),(Item=(Type=Bone,Name=\"dyn_tshirt_rear_r_01\"),Transform=(Rotation=(X=0.2158,Y=0.2033,Z=0.6627,W=-0. ...(truncated)"
    WeightMinimum=0
    Weight=@DeltaRotation_3.outputWeight
   )
   > &RigUnit_ModifyTransforms(
    ItemToModify="((Item=(Type=Bone,Name=\"dyn_tshirt_rear_01\"),Transform=(Rotation=(X=-0.2407,Y=0.4985,Z=-0.412,W=-0.7238),Translation=(X=0,Y=0,Z=0),Scale3D=(X=1,Y=1,Z=1))),(Item=(Type=Bone,Name=\"dyn_tshirt_rear_r_01\"),Transform=(Rotation=(X=-0.5208,Y=-0.1913,Z=-0.2854,W=-0 ...(truncated)"
    WeightMinimum=0
    Weight=&RigVMFunction_MathDoubleMul(A=@DeltaRotation_2.outputWeight, B=@DeltaRotation_3.outputWeight).Result
   )
   DeltaRotation_4 = RigVMFunctionReferenceNode(
    Rotation Order=ZYX
    child=(Type=Bone,Name="thigh_r")
    outputMinMax=(X=0,Y=1)
    parent=(Type=Null,Name="thigh_r_null")
    premult=1
    rotationMinMax=(X=0,Y=-90)
    triggerAxis=Y
    twistAxis=(X=1,Y=0,Z=0)
    ExecuteContext=@RigVMFunction_Sequence_2.C
   )
   > &RigUnit_ModifyTransforms(
    ItemToModify="((Item=(Type=Bone,Name=\"dyn_tshirt_side_r_01\"),Transform=(Rotation=(X=0,Y=0,Z=0.4189,W=-0.908),Translation=(X=0,Y=0,Z=0),Scale3D=(X=1,Y=1,Z=1))),(Item=(Type=Bone,Name=\"dyn_tshirt_rear_r_01\"),Transform=(Rotation=(X=0.1651,Y=-0.1199,Z=0.371,W=-0.9059),Transl ...(truncated)"
    WeightMinimum=0
    Weight=@DeltaRotation_4.outputWeight
   )
   > DeltaRotation_5 = RigVMFunctionReferenceNode(
    Rotation Order=ZYX
    child=(Type=Bone,Name="thigh_l")
    outputMinMax=(X=0,Y=1)
    parent=(Type=Null,Name="thigh_l_null")
    premult=1
    rotationMinMax=(X=0,Y=-90)
    triggerAxis=Y
    twistAxis=(X=1,Y=0,Z=0)
   )
   > &RigUnit_ModifyTransforms(
    ItemToModify="((Item=(Type=Bone,Name=\"dyn_tshirt_side_l_01\"),Transform=(Rotation=(X=0,Y=0,Z=0.3825,W=-0.9239),Translation=(X=0,Y=0,Z=0),Scale3D=(X=1,Y=1,Z=1))),(Item=(Type=Bone,Name=\"dyn_tshirt_rear_l_01\"),Transform=(Rotation=(X=0.0927,Y=0.2922,Z=0.4847,W=-0.8192),Trans ...(truncated)"
    WeightMinimum=0
    Weight=@DeltaRotation_5.outputWeight
   )
   > DeltaRotation_8 = RigVMFunctionReferenceNode(
    Rotation Order=ZYX
    child=(Type=Bone,Name="thigh_r")
    outputMinMax=(X=0,Y=1)
    parent=(Type=Null,Name="thigh_r_null")
    premult=1
    rotationMinMax=(X=0,Y=50)
    triggerAxis=Y
    twistAxis=(X=1,Y=0,Z=0)
   )
   > &RigUnit_ModifyTransforms(
    ItemToModify="((Item=(Type=Bone,Name=\"None\"),Transform=(Rotation=(X=0,Y=0,Z=0,W=1),Translation=(X=0,Y=0,Z=0),Scale3D=(X=1,Y=1,Z=1))),(Item=(Type=Bone,Name=\"dyn_tshirt_side_r_01\"),Transform=(Rotation=(X=0,Y=0,Z=-0.198,W=0.9802),Translation=(X=0,Y=0,Z=0),Scale3D=(X=1,Y=1, ...(truncated)"
    WeightMinimum=0
    Weight=@DeltaRotation_8.outputWeight
   )
   > DeltaRotation_9 = RigVMFunctionReferenceNode(
    Rotation Order=ZYX
    child=(Type=Bone,Name="thigh_l")
    outputMinMax=(X=0,Y=1)
    parent=(Type=Null,Name="thigh_l_null")
    premult=1
    rotationMinMax=(X=0,Y=50)
    triggerAxis=Y
    twistAxis=(X=1,Y=0,Z=0)
   )
   > &RigUnit_ModifyTransforms(
    ItemToModify="((Item=(Type=Bone,Name=\"dyn_tshirt_front_l_01\"),Transform=(Rotation=(X=0.0125,Y=-0.2614,Z=0.0462,W=-0.964),Translation=(X=0,Y=0,Z=0),Scale3D=(X=1,Y=1,Z=1))),(Item=(Type=Bone,Name=\"dyn_tshirt_side_l_01\"),Transform=(Rotation=(X=0,Y=0,Z=0.1975,W=-0.9803),Tran ...(truncated)"
    WeightMinimum=0
    Weight=@DeltaRotation_9.outputWeight
   )
   &RigUnit_ModifyTransforms(
    ItemToModify="((Item=(Type=Bone,Name=\"base_tshirt_rear_r_01\"),Transform=(Rotation=(X=0,Y=0,Z=0,W=1),Translation=(X=0,Y=0,Z=0),Scale3D=(X=1,Y=1,Z=1))),(Item=(Type=Bone,Name=\"base_tshirt_rear_l_01\"),Transform=(Rotation=(X=0,Y=0,Z=-0.0175,W=0.9998),Translation=(X=0,Y=-1,Z= ...(truncated)"
    Weight=1
    WeightMinimum=0
    ExecutePin=@RigVMFunction_Sequence_2.D
   )
   n2 = &RigVMFunction_AlphaInterp(
    InterpSpeedDecreasing=10
    InterpSpeedIncreasing=5
    Scale=1
    bClampResult=False
    bInterpResult=False
    Value=RigVMVariableNode() {variable=CrouchingFloat}.Value
   )
   > &RigUnit_ModifyTransforms(
    ItemToModify="((Item=(Type=Bone,Name=\"dyn_tshirt_front_r_02\"),Transform=(Rotation=(X=0,Y=0,Z=0.0872,W=0.9962),Translation=(X=0,Y=-0.54,Z=0),Scale3D=(X=1,Y=1,Z=1))),(Item=(Type=Bone,Name=\"dyn_tshirt_front_02\"),Transform=(Rotation=(X=0,Y=0,Z=0,W=1),Translation=(X=0,Y=5,Z= ...(truncated)"
    WeightMinimum=0
    Weight=@n2.Result
   )
   n3 = &RigVMFunction_AlphaInterp(InterpSpeedDecreasing=10, InterpSpeedIncreasing=5, Scale=1, bClampResult=False, Value=RigVMVariableNode() {variable=CrounchingWalk2D}.Value)
   > &RigUnit_ModifyTransforms(
    ItemToModify="((Item=(Type=Bone,Name=\"base_tshirt_front_r_01\"),Transform=(Rotation=(X=0,Y=0,Z=-0.0436,W=0.999),Translation=(X=0,Y=-3,Z=0),Scale3D=(X=1,Y=1,Z=1))),(Item=(Type=Bone,Name=\"base_tshirt_front_01\"),Transform=(Rotation=(X=0,Y=0,Z=-0.0698,W=0.9976),Translation=( ...(truncated)"
    WeightMinimum=0
    Weight=@n3.Result
   )
   > n4 = &RigUnit_SphericalPoseReader(
    ActiveRegionScaleFactors=(PositiveWidth=1,NegativeWidth=1,PositiveHeight=1,NegativeHeight=1)
    ActiveRegionSize=0.02
    Debug=(bDrawDebug=false,bDraw2D=False,bDrawLocalAxes=False,DebugScale=25,DebugSegments=20,DebugThickness=0.25)
    DriverAxis=(X=1,Y=0,Z=0)
    DriverItem=(Type=Bone,Name="thigh_r")
    FalloffRegionScaleFactors=(PositiveWidth=1,NegativeWidth=1,PositiveHeight=1,NegativeHeight=1)
    FalloffSize=0.2
    FlipHeightScaling=False
    FlipWidthScaling=False
    OptionalParentItem=(Type=Bone,Name="None")
    RotationOffset=(X=-83,Y=0,Z=-3)
   )
   > &RigUnit_ModifyTransforms(
    ItemToModify="((Item=(Type=Bone,Name=\"dyn_tshirt_front_r_01\"),Transform=(Rotation=(X=0,Y=0,Z=-0.0958,W=0.9954),Translation=(X=0,Y=0,Z=0),Scale3D=(X=1,Y=1,Z=1))),(Item=(Type=Bone,Name=\"dyn_tshirt_front_02\"),Transform=(Rotation=(X=0.0871,Y=-0.0038,Z=-0.0435,W=0.9952),Tran ...(truncated)"
    WeightMinimum=0
    Weight=@n4.OutputParam
   )
   > n5 = &RigUnit_SphericalPoseReader(
    ActiveRegionScaleFactors=(PositiveWidth=0.75,NegativeWidth=0.75,PositiveHeight=0.75,NegativeHeight=0.75)
    ActiveRegionSize=0.1
    Debug=(bDrawDebug=false,bDraw2D=False,bDrawLocalAxes=False,DebugScale=25,DebugSegments=20,DebugThickness=0.25)
    DriverAxis=(X=1,Y=0,Z=0)
    DriverItem=(Type=Bone,Name="thigh_r")
    FalloffRegionScaleFactors=(PositiveWidth=1,NegativeWidth=1,PositiveHeight=1,NegativeHeight=1)
    FalloffSize=0.2
    FlipHeightScaling=False
    FlipWidthScaling=False
    OptionalParentItem=(Type=Bone,Name="pelvis")
    RotationOffset=(X=-105,Y=20,Z=0)
   )
   > &RigUnit_ModifyTransforms(
    ItemToModify="((Item=(Type=Bone,Name=\"base_tshirt_front_01\"),Transform=(Rotation=(X=0.0338,Y=0.2566,Z=-0.1261,W=0.9577),Translation=(X=0,Y=-1,Z=-2),Scale3D=(X=1,Y=1,Z=1))),(Item=(Type=Bone,Name=\"base_tshirt_front_r_01\"),Transform=(Rotation=(X=-0.0046,Y=0.0434,Z=0.1044,W ...(truncated)"
    WeightMinimum=0
    Weight=@n5.OutputParam
   )
   RigVMFunction_MathBoolToFloat = &RigVMFunction_MathBoolToFloat(Value=RigVMVariableNode() {variable=PlayingEmote}.Value)
   n6 = &RigVMFunction_AlphaInterp(InterpSpeedDecreasing=8, InterpSpeedIncreasing=8, Scale=1, bClampResult=False, Value=@RigVMFunction_MathBoolToFloat.Result)
   > &RigUnit_ModifyTransforms(
    ItemToModify="((Item=(Type=Bone,Name=\"base_tshirt_rear_r_01\"),Transform=(Rotation=(X=0,Y=0,Z=0,W=1),Translation=(X=0,Y=-0.5,Z=-0.5),Scale3D=(X=1,Y=1,Z=1))),(Item=(Type=Bone,Name=\"base_tshirt_side_r_01\"),Transform=(Rotation=(X=0,Y=0,Z=0,W=1),Translation=(X=0,Y=-0.5,Z=0), ...(truncated)"
    WeightMinimum=0
    Weight=@n6.Result
   )
   RigVMFunctionReturnNode(ExecuteContext=@RigVMFunction_Sequence_2.E)
   &RigVMFunction_AlphaInterp(InterpSpeedDecreasing=8, InterpSpeedIncreasing=3, Scale=1, Value=0, bClampResult=true)
   &RigVMFunction_AlphaInterp(InterpSpeedDecreasing=8, InterpSpeedIncreasing=3, Scale=1, Value=0, bClampResult=true)
   OG_values_thighUp = RigVMCollapseNode(ExecuteContext=())
  ]}
 CollapseNode_ContainedGraph {nodes [
   RigVMFunctionEntryNode()
   > DeltaRotation_6 = RigVMFunctionReferenceNode(
    Rotation Order=ZYX
    child=(Type=Bone,Name="thigh_l")
    debugRotation=false
    outputMinMax=(X=0,Y=1)
    parent=(Type=Null,Name="thigh_l_null")
    premult=1
    rotationMinMax=(X=-5,Y=-90)
    triggerAxis=Z
    twistAxis=(X=1,Y=0,Z=0)
   )
   > &RigUnit_ModifyTransforms(
    ItemToModify="((Item=(Type=Bone,Name=\"dyn_tshirt_front_l_01\"),Transform=(Rotation=(X=-0.0208,Y=0.0236,Z=-0.2582,W=0.9656),Translation=(X=0,Y=0,Z=0),Scale3D=(X=1,Y=1,Z=1))),(Item=(Type=Bone,Name=\"dyn_tshirt_front_01\"),Transform=(Rotation=(X=-0.3765,Y=0.2216,Z=-0.512,W=0. ...(truncated)"
    WeightMinimum=0
    Weight=@DeltaRotation_6.outputWeight
   )
   > DeltaRotation_7 = RigVMFunctionReferenceNode(
    Rotation Order=ZYX
    child=(Type=Bone,Name="thigh_r")
    outputMinMax=(X=0,Y=1)
    parent=(Type=Null,Name="thigh_r_null")
    premult=1
    rotationMinMax=(X=0,Y=-90)
    triggerAxis=Z
    twistAxis=(X=1,Y=0,Z=0)
   )
   > &RigUnit_ModifyTransforms(
    ItemToModify="((Item=(Type=Bone,Name=\"dyn_tshirt_front_r_01\"),Transform=(Rotation=(X=0.0376,Y=-0.0281,Z=-0.2578,W=0.9651),Translation=(X=0,Y=0,Z=0),Scale3D=(X=1,Y=1,Z=1))),(Item=(Type=Bone,Name=\"dyn_tshirt_front_01\"),Transform=(Rotation=(X=0.3154,Y=-0.162,Z=-0.5363,W=0. ...(truncated)"
    WeightMinimum=0
    Weight=@DeltaRotation_7.outputWeight
   )
   > &RigUnit_ModifyTransforms(
    ItemToModify="((Item=(Type=Bone,Name=\"dyn_tshirt_front_01\"),Transform=(Rotation=(X=-0.071,Y=0.4194,Z=0.2544,W=0.8685),Translation=(X=0,Y=0,Z=0),Scale3D=(X=1,Y=1,Z=1))),(Item=(Type=Bone,Name=\"dyn_tshirt_front_l_01\"),Transform=(Rotation=(X=0.0013,Y=0.0627,Z=0.1803,W=0.981 ...(truncated)"
    WeightMinimum=0
    Weight=&RigVMFunction_MathDoubleMul(A=@DeltaRotation_6.outputWeight, B=@DeltaRotation_7.outputWeight).Result
   )
   > RigVMFunctionReturnNode()
  ]}
 CollapseNode_ContainedGraph {nodes [
   RigVMFunctionEntryNode()
   > &RigUnit_ModifyTransforms(
    ItemToModify=((Item=(Type=Bone,Name="dyn_sleeveScoot_l_01"),Transform=(Rotation=(X=0.2725,Y=0.536,Z=0.296,W=-0.7422),Translation=(X=-0.3015,Y=0.0395,Z=-0.2156),Scale3D=(X=1,Y=1,Z=1))))
    WeightMinimum=0
    Weight=RigVMVariableNode() {variable=L lowerarm bend}.Value
   )
   > &RigUnit_ModifyTransforms(
    ItemToModify=((Item=(Type=Bone,Name="dyn_sleeveScoot_r_01"),Transform=(Rotation=(X=-0.0753,Y=-0.4383,Z=0.1981,W=-0.8735),Translation=(X=-6.0583,Y=0.5517,Z=1.5667),Scale3D=(X=0.6101,Y=1,Z=1))))
    WeightMinimum=0
    Weight=RigVMVariableNode() {variable=R lowerarm bend}.Value
   )
   > RigVMFunctionReturnNode()
  ]}
 CollapseNode_ContainedGraph {nodes [
   n1 = RigVMFunctionEntryNode()
   RigVMFunction_MathBoolToFloat = &RigVMFunction_MathBoolToFloat(Value=RigVMVariableNode() {variable=KneeSliding}.Value)
   AlphaInterp = &RigVMFunction_AlphaInterp(InterpSpeedDecreasing=10, InterpSpeedIncreasing=10, Scale=1, bClampResult=False, Value=@RigVMFunction_MathBoolToFloat.Result)
   &RigUnit_ModifyTransforms(
    ItemToModify="((Item=(Type=Bone,Name=\"dyn_tshirt_side_l_01\"),Transform=(Rotation=(X=-0.0851,Y=0.1485,Z=-0.1368,W=0.9757),Translation=(X=0,Y=0,Z=0),Scale3D=(X=1,Y=1,Z=1))),(Item=(Type=Bone,Name=\"dyn_tshirt_front_01\"),Transform=(Rotation=(X=-0.0578,Y=-0.1854,Z=-0.1245,W=0 ...(truncated)"
    WeightMinimum=0
    ExecutePin=@n1.ExecutePin
    Weight=@AlphaInterp.Result
   )
   > &RigUnit_ModifyTransforms(
    ItemToModify="((Item=(Type=Bone,Name=\"dyn_bandolierFront_01\"),Transform=(Rotation=(X=0,Y=0,Z=-0.2736,W=0.9618),Translation=(X=0,Y=0,Z=0),Scale3D=(X=1,Y=1,Z=1))),(Item=(Type=Bone,Name=\"dyn_sideBag_02\"),Transform=(Rotation=(X=-0.3409,Y=0.059,Z=0.1724,W=-0.9223),Translatio ...(truncated)"
    WeightMinimum=0
    Weight=@AlphaInterp.Result
   )
   n2 = &RigVMFunction_AlphaInterp(
    InterpSpeedDecreasing=10
    InterpSpeedIncreasing=10
    Scale=1
    bClampResult=False
    Value=&RigVMFunction_MathBoolToFloat(Value=RigVMVariableNode() {variable=BaseballSliding}.Value).Result
   )
   > n3 = &RigUnit_ModifyTransforms(
    ItemToModify="((Item=(Type=Bone,Name=\"dyn_tshirt_side_l_01\"),Transform=(Rotation=(X=0.0072,Y=0.1308,Z=-0.079,W=0.9882),Translation=(X=0,Y=0,Z=0),Scale3D=(X=1,Y=1,Z=1))),(Item=(Type=Bone,Name=\"dyn_tshirt_front_01\"),Transform=(Rotation=(X=0,Y=0,Z=0,W=1),Translation=(X=0,Y ...(truncated)"
    WeightMinimum=0
    Weight=@n2.Result
   )
   RigVMFunctionReturnNode(ExecutePin=@n3.ExecutePin)
   &RigUnit_ModifyTransforms(
    ExecutePin=()
    ItemToModify="((Item=(Type=Bone,Name=\"dyn_tshirt_side_l_01\"),Transform=(Rotation=(X=0.0072,Y=0.1308,Z=-0.079,W=0.9882),Translation=(X=0,Y=0,Z=0),Scale3D=(X=1,Y=1,Z=1))),(Item=(Type=Bone,Name=\"dyn_tshirt_front_01\"),Transform=(Rotation=(X=0,Y=0,Z=0,W=1),Translation=(X=0,Y ...(truncated)"
    Weight=1
    WeightMinimum=0
   )
   &RigUnit_ModifyTransforms(
    ExecutePin=()
    ItemToModify="((Item=(Type=Bone,Name=\"dyn_tshirt_side_l_01\"),Transform=(Rotation=(X=-0.0851,Y=0.1486,Z=-0.1368,W=0.9757),Translation=(X=0,Y=0,Z=0),Scale3D=(X=1,Y=1,Z=1))),(Item=(Type=Bone,Name=\"dyn_tshirt_front_01\"),Transform=(Rotation=(X=0.0578,Y=0.1854,Z=0.1246,W=-0.9 ...(truncated)"
    WeightMinimum=0
    Weight=@AlphaInterp.Result
   )
  ]}
 CollapseNode_ContainedGraph {nodes [
   RigVMFunctionEntryNode()
   > DeltaRotation_6 = RigVMFunctionReferenceNode(
    Rotation Order=ZYX
    child=(Type=Bone,Name="foot_l")
    outputMinMax=(X=0,Y=1)
    parent=(Type=Bone,Name="calf_l")
    premult=1
    rotationMinMax=(X=0,Y=-30)
    triggerAxis=Z
    twistAxis=(X=1,Y=0,Z=0)
   )
   > &RigUnit_ModifyTransforms(
    ItemToModify=((Item=(Type=Bone,Name="dyn_shoetongue_l_01"),Transform=(Rotation=(X=0,Y=0,Z=-0.4867,W=-0.8736),Translation=(X=-0.118,Y=0.9223,Z=0),Scale3D=(X=1,Y=1,Z=1))))
    WeightMinimum=0
    Weight=@DeltaRotation_6.outputWeight
   )
   > DeltaRotation_7 = RigVMFunctionReferenceNode(
    Rotation Order=ZYX
    child=(Type=Bone,Name="foot_r")
    outputMinMax=(X=0,Y=1)
    parent=(Type=Bone,Name="calf_r")
    premult=1
    rotationMinMax=(X=0,Y=-30)
    triggerAxis=Z
    twistAxis=(X=1,Y=0,Z=0)
   )
   > &RigUnit_ModifyTransforms(
    ItemToModify=((Item=(Type=Bone,Name="dyn_shoetongue_r_01"),Transform=(Rotation=(X=0,Y=0,Z=0.5911,W=-0.8066),Translation=(X=-0.0602,Y=-0.4979,Z=0),Scale3D=(X=1,Y=1,Z=1))))
    WeightMinimum=0
    Weight=@DeltaRotation_7.outputWeight
   )
   > RigVMFunctionReturnNode()
  ]}
 CollapseNode_ContainedGraph {nodes [
   RigVMFunctionEntryNode()
   n1 = &RigUnit_GetRelativeTransformForItem(Child=(Type=Bone,Name="lowerarm_l"), Parent=(Type=Bone,Name="upperarm_l"), bChildInitial=False, bParentInitial=False)
   > RigVMVariableNode(Value=&RigVMFunction_MathDoubleRemap(
     SourceMaximum=100
     SourceMinimum=80
     TargetMaximum=1
     TargetMinimum=0
     Value=&RigVMFunction_MathQuaternionToEuler(RotationOrder=XYZ, Value=@n1.RelativeTransform.Rotation).Result.Z
    ).Result) {variable=L lowerarm bend}
   n2 = &RigUnit_GetRelativeTransformForItem(Child=(Type=Bone,Name="lowerarm_r"), Parent=(Type=Bone,Name="upperarm_r"), bChildInitial=False, bParentInitial=False)
   > RigVMVariableNode(Value=&RigVMFunction_MathDoubleRemap(
     SourceMaximum=110
     SourceMinimum=80
     TargetMaximum=1
     TargetMinimum=0
     Value=&RigVMFunction_MathQuaternionToEuler(RotationOrder=XYZ, Value=@n2.RelativeTransform.Rotation).Result.Z
    ).Result) {variable=R lowerarm bend}
   > RigVMFunctionReturnNode()
  ]}
 CollapseNode_ContainedGraph {nodes [
   RigVMFunctionEntryNode()
   > DeltaRotation_0 = RigVMFunctionReferenceNode(
    Rotation Order=ZYX
    child=(Type=Bone,Name="hand_r")
    debugRotation=false
    outputMinMax=(X=0,Y=1)
    parent=(Type=Null,Name="hand_r_null")
    premult=1
    rotationMinMax=(X=0,Y=-30)
    triggerAxis=Y
    twistAxis=(X=1,Y=0,Z=0)
   )
   > &RigUnit_ModifyTransforms(
    ItemToModify="((Item=(Type=Bone,Name=\"base_wristHelper_01\"),Transform=(Rotation=(X=0,Y=0,Z=0,W=-1),Translation=(X=-8,Y=-1.1283,Z=-5.5621),Scale3D=(X=1,Y=1,Z=1))),(Item=(Type=Bone,Name=\"base_wristHelper_02\"),Transform=(Rotation=(X=0,Y=0,Z=0,W=-1),Translation=(X=0,Y=1,Z=- ...(truncated)"
    WeightMinimum=0
    Weight=@DeltaRotation_0.outputWeight
   )
   > DeltaRotation_2 = RigVMFunctionReferenceNode(
    Rotation Order=ZYX
    child=(Type=Bone,Name="hand_r")
    debugRotation=false
    outputMinMax=(X=0,Y=1)
    parent=(Type=Null,Name="hand_r_null")
    premult=1
    rotationMinMax=(X=31,Y=-60)
    triggerAxis=Y
    twistAxis=(X=1,Y=0,Z=0)
   )
   > &RigUnit_ModifyTransforms(
    ItemToModify="((Item=(Type=Bone,Name=\"base_wristHelper_01\"),Transform=(Rotation=(X=0,Y=0,Z=0,W=-1),Translation=(X=-0.9844,Y=-0.7031,Z=-0.1393),Scale3D=(X=1,Y=1,Z=1))),(Item=(Type=Bone,Name=\"base_wristHelper_02\"),Transform=(Rotation=(X=0,Y=0,Z=0,W=-1),Translation=(X=-3.5 ...(truncated)"
    WeightMinimum=0
    Weight=@DeltaRotation_2.outputWeight
   )
   > DeltaRotation_1 = RigVMFunctionReferenceNode(
    Rotation Order=ZYX
    child=(Type=Bone,Name="hand_r")
    debugRotation=false
    outputMinMax=(X=0,Y=1)
    parent=(Type=Null,Name="hand_r_null")
    premult=1
    rotationMinMax=(X=0,Y=40)
    triggerAxis=Y
    twistAxis=(X=1,Y=0,Z=0)
   )
   > &RigUnit_ModifyTransforms(
    ItemToModify="((Item=(Type=Bone,Name=\"base_wristHelper_01\"),Transform=(Rotation=(X=0,Y=0,Z=0,W=-1),Translation=(X=0,Y=0,Z=0),Scale3D=(X=1,Y=1,Z=1))),(Item=(Type=Bone,Name=\"base_wristHelper_02\"),Transform=(Rotation=(X=0,Y=0,Z=0,W=-1),Translation=(X=0,Y=8,Z=-5),Scale3D=(X ...(truncated)"
    WeightMinimum=0
    Weight=@DeltaRotation_1.outputWeight
   )
   > RigVMFunctionReturnNode()
  ]}
 CollapseNode_ContainedGraph {nodes [
   DeltaRotation_9 = RigVMFunctionReferenceNode(
    ExecuteContext=()
    Rotation Order=ZYX
    child=(Type=Bone,Name="spine_05")
    debugRotation=false
    onlyTwist=false
    outputMinMax=(X=0,Y=1)
    parent=(Type=Null,Name="spine_05_Spine01Delta_null")
    premult=1
    rotationMinMax=(X=0,Y=25)
    triggerAxis=Z
    twistAxis=(X=1,Y=0,Z=0)
   )
   n1 = &RigVMFunction_AlphaInterp(
    InterpSpeedDecreasing=8
    InterpSpeedIncreasing=8
    Scale=1
    bClampResult=False
    Value=&RigVMFunction_MathBoolToFloat(Value=RigVMVariableNode() {variable=PlayingEmote}.Value).Result
   )
   n2 = &RigVMFunction_AnimEvalRichCurve(
    Curve="(EditorCurveData=(Keys=((InterpMode=RCIM_Cubic,TangentMode=RCTM_User,ArriveTangent=1.4754,LeaveTangent=1.4754),(InterpMode=RCIM_Cubic,Time=0.4736,Value=0.636,ArriveTangent=1,LeaveTangent=1),(InterpMode=RCIM_Cubic,TangentMode=RCTM_User,Time=1,Value=1,Arrive ...(truncated)"
    SourceMinimum=0
    TargetMinimum=0
    Value=&RigVMFunction_MathDoubleMul(B=@n1.Result, A=@DeltaRotation_9.outputWeight).Result
   )
   > &RigUnit_ModifyTransforms(
    ItemToModify="((Item=(Type=Bone,Name=\"base_bandolierBack_01\"),Transform=(Rotation=(X=0,Y=0.0436,Z=0,W=0.999),Translation=(X=0,Y=0,Z=-3.5),Scale3D=(X=1,Y=1,Z=1))),(Item=(Type=Bone,Name=\"base_bandolierFront_01\"),Transform=(Rotation=(X=-0.0003,Y=0.0175,Z=0.0175,W=0.9997),T ...(truncated)"
    WeightMinimum=0
    Weight=@n2.Result
   )
   > DeltaRotation_10 = RigVMFunctionReferenceNode(
    Rotation Order=ZYX
    child=(Type=Bone,Name="spine_05")
    debugRotation=false
    onlyTwist=true
    outputMinMax=(X=0,Y=1)
    parent=(Type=Null,Name="spine_05_Spine01Delta_null")
    premult=1
    rotationMinMax=(X=0,Y=-24)
    triggerAxis=X
    twistAxis=(X=1,Y=0,Z=0)
   )
   > &RigUnit_ModifyTransforms(
    ItemToModify="((Item=(Type=Bone,Name=\"base_bandolierBack_01\"),Transform=(Rotation=(X=0,Y=0,Z=0,W=1),Translation=(X=0,Y=3,Z=-1.5),Scale3D=(X=1,Y=1,Z=1))),(Item=(Type=Bone,Name=\"base_bandolierFront_01\"),Transform=(Rotation=(X=0.4222,Y=0.0395,Z=-0.0184,W=0.9054),Translatio ...(truncated)"
    WeightMinimum=0
    Weight=@DeltaRotation_10.outputWeight
   )
   n3 = RigVMFunctionEntryNode()
   RigVMFunction_ControlFlowBranch = &RigVMFunction_ControlFlowBranch(ExecuteContext=@n3.ExecutePin, Condition=RigVMVariableNode() {variable=PlayingEmote}.Value)
   RigVMFunctionReturnNode(ExecutePin=@RigVMFunction_ControlFlowBranch.Completed)
   RigVMFunction_Sequence = RigVMAggregateNode(ExecuteContext=@RigVMFunction_ControlFlowBranch.False)
   &RigUnit_ModifyTransforms(
    ItemToModify=((Item=(Type=Bone,Name="dyn_sideBag_01"),Transform=(Rotation=(X=0,Y=0,Z=-0.0436,W=0.999),Translation=(X=0,Y=0,Z=0),Scale3D=(X=1,Y=1,Z=1))))
    Weight=1
    WeightMinimum=0
    ExecutePin=@RigVMFunction_Sequence.A
   )
   > n4 = &RigUnit_SphericalPoseReader(
    ActiveRegionScaleFactors=(PositiveWidth=1,NegativeWidth=1,PositiveHeight=1,NegativeHeight=1)
    ActiveRegionSize=0.05
    Debug=(bDrawDebug=false,bDraw2D=False,bDrawLocalAxes=False,DebugScale=25,DebugSegments=20,DebugThickness=0.25)
    DriverAxis=(X=1,Y=0,Z=0)
    DriverItem=(Type=Bone,Name="dyn_sideBag_01")
    FalloffRegionScaleFactors=(PositiveWidth=1,NegativeWidth=1,PositiveHeight=1,NegativeHeight=1)
    FalloffSize=0.45
    FlipHeightScaling=False
    FlipWidthScaling=False
    OptionalParentItem=(Type=Bone,Name="None")
    RotationOffset=(X=0,Y=-10,Z=0)
   )
   > &RigUnit_ModifyTransforms(
    ItemToModify=((Item=(Type=Bone,Name="dyn_bandolierFront_01"),Transform=(Rotation=(X=0,Y=0.6275,Z=0,W=-0.7786),Translation=(X=0,Y=0,Z=0),Scale3D=(X=1,Y=1,Z=1))))
    WeightMinimum=0
    Weight=@n4.OutputParam
   )
   > n5 = &RigUnit_SphericalPoseReader(
    ActiveRegionScaleFactors=(PositiveWidth=1,NegativeWidth=1,PositiveHeight=1,NegativeHeight=1)
    ActiveRegionSize=0.05
    Debug=(bDrawDebug=false,bDraw2D=False,bDrawLocalAxes=False,DebugScale=25,DebugSegments=20,DebugThickness=0.25)
    DriverAxis=(X=1,Y=0,Z=0)
    DriverItem=(Type=Bone,Name="dyn_sideBag_01")
    FalloffRegionScaleFactors=(PositiveWidth=1,NegativeWidth=1,PositiveHeight=1,NegativeHeight=1)
    FalloffSize=0.35
    FlipHeightScaling=False
    FlipWidthScaling=False
    OptionalParentItem=(Type=Bone,Name="None")
    RotationOffset=(X=0,Y=-160,Z=0)
   )
   > &RigUnit_ModifyTransforms(
    ItemToModify=((Item=(Type=Bone,Name="dyn_bandolierBack_01"),Transform=(Rotation=(X=0.2842,Y=0.8019,Z=0.193,W=0.4888),Translation=(X=0,Y=0,Z=0),Scale3D=(X=1,Y=1,Z=1))))
    WeightMinimum=0
    Weight=@n5.OutputParam
   )
   DeltaRotation = RigVMFunctionReferenceNode(
    Rotation Order=ZYX
    child=(Type=Bone,Name="thigh_l")
    debugRotation=false
    outputMinMax=(X=0,Y=1)
    parent=(Type=Null,Name="thigh_l_null")
    premult=1
    rotationMinMax=(X=-30,Y=-90)
    triggerAxis=Z
    twistAxis=(X=1,Y=0,Z=0)
    ExecuteContext=@RigVMFunction_Sequence.B
   )
   > &RigUnit_ModifyTransforms(
    ItemToModify="((Item=(Type=Bone,Name=\"dyn_sideBag_01\"),Transform=(Rotation=(X=-0.0745,Y=-0.049,Z=0.0227,W=-0.9958),Translation=(X=0,Y=0,Z=0),Scale3D=(X=1,Y=1,Z=1))),(Item=(Type=Bone,Name=\"dyn_sideBag_02\"),Transform=(Rotation=(X=-0.0536,Y=-0.0234,Z=0.0013,W=-0.9983),Tran ...(truncated)"
    WeightMinimum=0
    Weight=@DeltaRotation.outputWeight
   )
   > DeltaRotation_1 = RigVMFunctionReferenceNode(
    Rotation Order=ZYX
    child=(Type=Bone,Name="spine_04")
    debugRotation=false
    onlyTwist=true
    outputMinMax=(X=0,Y=1)
    parent=(Type=Null,Name="spine_04_Spine1Delta_null")
    premult=1
    rotationMinMax=(X=0,Y=40)
    triggerAxis=X
    twistAxis=(X=1,Y=0,Z=0)
   )
   > &RigUnit_ModifyTransforms(
    ItemToModify="((Item=(Type=Bone,Name=\"dyn_sideBag_01\"),Transform=(Rotation=(X=0.1126,Y=0.0753,Z=-0.1126,W=0.9844),Translation=(X=0,Y=-1,Z=1),Scale3D=(X=1,Y=1,Z=1))),(Item=(Type=Bone,Name=\"dyn_sideBag_02\"),Transform=(Rotation=(X=0.1337,Y=0.0317,Z=-0.092,W=0.9862),Transla ...(truncated)"
    WeightMinimum=0
    Weight=@DeltaRotation_1.outputWeight
   )
   > DeltaRotation_2 = RigVMFunctionReferenceNode(
    Rotation Order=ZYX
    child=(Type=Bone,Name="spine_04")
    debugRotation=false
    onlyTwist=false
    outputMinMax=(X=0,Y=1)
    parent=(Type=Null,Name="spine_04_Spine1Delta_null")
    premult=1
    rotationMinMax=(X=0,Y=20)
    triggerAxis=Z
    twistAxis=(X=1,Y=0,Z=0)
   )
   > &RigUnit_ModifyTransforms(
    ItemToModify="((Item=(Type=Bone,Name=\"base_sideBag_01\"),Transform=(Rotation=(X=0,Y=-0.0872,Z=0,W=0.9962),Translation=(X=0,Y=0,Z=0),Scale3D=(X=1,Y=1,Z=1))),(Item=(Type=Bone,Name=\"dyn_sideBag_02\"),Transform=(Rotation=(X=0.1063,Y=-0.2495,Z=0.1063,W=0.9566),Translation=(X=0 ...(truncated)"
    WeightMinimum=0
    Weight=@DeltaRotation_2.outputWeight
   )
   n6 = &RigVMFunction_AlphaInterp(
    InterpSpeedDecreasing=10
    InterpSpeedIncreasing=10
    Scale=1
    bClampResult=False
    Value=&RigVMFunction_MathBoolToFloat(Value=RigVMVariableNode() {variable=BaseballSliding}.Value).Result
   )
   > &RigUnit_ModifyTransforms(
    ItemToModify="((Item=(Type=Bone,Name=\"base_sideBag_01\"),Transform=(Rotation=(X=0,Y=0,Z=-0.0872,W=0.9962),Translation=(X=0,Y=0,Z=0),Scale3D=(X=1,Y=1,Z=1))),(Item=(Type=Bone,Name=\"dyn_sideBag_02\"),Transform=(Rotation=(X=0.2685,Y=0.1449,Z=-0.1277,W=0.9437),Translation=(X=0 ...(truncated)"
    WeightMinimum=0
    Weight=@n6.Result
   )
   > DeltaRotation_3 = RigVMFunctionReferenceNode(
    Rotation Order=ZYX
    child=(Type=Bone,Name="spine_04")
    debugRotation=false
    onlyTwist=true
    outputMinMax=(X=0,Y=1)
    parent=(Type=Null,Name="spine_04_Spine1Delta_null")
    premult=1
    rotationMinMax=(X=0,Y=54)
    triggerAxis=X
    twistAxis=(X=1,Y=0,Z=0)
   )
   RigVMFunction_MathBoolAnd = RigVMAggregateNode(
    A=RigVMVariableNode() {variable=DoubleHandedWeapon}.Value
    B=RigVMVariableNode() {variable=PickaxeSwing}.Value
    C=RigVMVariableNode() {variable=Crouching}.Value
   )
   AlphaInterp = &RigVMFunction_AlphaInterp(
    InterpSpeedDecreasing=6
    InterpSpeedIncreasing=10
    Scale=1
    bClampResult=true
    Value=&RigVMFunction_MathBoolToFloat(Value=@RigVMFunction_MathBoolAnd.Result).Result
   )
   RigVMFunction_AnimEvalRichCurve = &RigVMFunction_AnimEvalRichCurve(
    Curve=(EditorCurveData=(Keys=((InterpMode=RCIM_Cubic,TangentMode=RCTM_User,ArriveTangent=2.1396,LeaveTangent=2.1396),(InterpMode=RCIM_Cubic,Time=1,Value=1))))
    SourceMinimum=0
    TargetMinimum=0
    Value=@DeltaRotation_3.outputWeight
   )
   Multiply = &RigVMFunction_MathDoubleMul(B=@AlphaInterp.Result, A=@RigVMFunction_AnimEvalRichCurve.Result)
   > &RigUnit_ModifyTransforms(
    ItemToModify="((Item=(Type=Bone,Name=\"base_sideBag_01\"),Transform=(Rotation=(X=0.1449,Y=-0.1277,Z=-0.2393,W=0.9515),Translation=(X=0,Y=0,Z=0),Scale3D=(X=1,Y=1,Z=1))),(Item=(Type=Bone,Name=\"dyn_sideBag_02\"),Transform=(Rotation=(X=-0.2607,Y=0.2685,Z=-0.2607,W=0.8899),Tran ...(truncated)"
    WeightMinimum=0
    Weight=@Multiply.Result
   )
   > DeltaRotation_4 = RigVMFunctionReferenceNode(
    Rotation Order=ZYX
    child=(Type=Bone,Name="spine_04")
    debugRotation=false
    onlyTwist=true
    outputMinMax=(X=0,Y=1)
    parent=(Type=Null,Name="spine_04_Spine1Delta_null")
    premult=1
    rotationMinMax=(X=-20,Y=-50)
    triggerAxis=X
    twistAxis=(X=1,Y=0,Z=0)
   )
   RigVMFunction_MathBoolAnd_1 = RigVMAggregateNode(
    A=RigVMVariableNode() {variable=DoubleHandedWeapon}.Value
    B=RigVMVariableNode() {variable=PickaxeSwing}.Value
    C=RigVMVariableNode() {variable=Crouching}.Value
   )
   n7 = &RigVMFunction_AlphaInterp(
    InterpSpeedDecreasing=6
    InterpSpeedIncreasing=10
    Scale=1
    bClampResult=true
    Value=&RigVMFunction_MathBoolToFloat(Value=@RigVMFunction_MathBoolAnd_1.Result).Result
   )
   n8 = &RigVMFunction_AnimEvalRichCurve(
    Curve=(EditorCurveData=(Keys=((InterpMode=RCIM_Cubic,TangentMode=RCTM_User,ArriveTangent=2.1396,LeaveTangent=2.1396),(InterpMode=RCIM_Cubic,Time=1,Value=1))))
    SourceMinimum=0
    TargetMinimum=0
    Value=@DeltaRotation_4.outputWeight
   )
   > &RigUnit_ModifyTransforms(
    ItemToModify="((Item=(Type=Bone,Name=\"base_sideBag_01\"),Transform=(Rotation=(X=0,Y=0,Z=-0.0872,W=0.9962),Translation=(X=0,Y=0,Z=0),Scale3D=(X=1,Y=1,Z=1))),(Item=(Type=Bone,Name=\"dyn_sideBag_02\"),Transform=(Rotation=(X=-0.0858,Y=0.0151,Z=-0.173,W=0.9811),Translation=(X=0 ...(truncated)"
    WeightMinimum=0
    Weight=&RigVMFunction_MathDoubleMul(B=@n7.Result, A=@n8.Result).Result
   )
   &RigUnit_AimConstraintLocalSpaceOffset(
    Child=(Type=Bone,Name="dyn_sideBag_01")
    Filter=(bX=True,bY=True,bZ=True)
    Parents=((Item=(Type=Bone,Name="dyn_tshirt_side_l_03"),Weight=1))
    UpAxis=(X=0,Y=1,Z=0)
    Weight=1
    WorldUp=(Target=(X=0,Y=0,Z=1),Kind=Location,Space=(Type=Bone,Name="dyn_tshirt_front_01"))
    ExecutePin=@RigVMFunction_ControlFlowBranch.True
   )
   > DeltaRotation_7 = RigVMFunctionReferenceNode(
    Rotation Order=ZYX
    child=(Type=Bone,Name="spine_05")
    debugRotation=false
    onlyTwist=false
    outputMinMax=(X=0,Y=1)
    parent=(Type=Null,Name="spine_05_null")
    premult=1
    rotationMinMax=(X=-40,Y=40)
    triggerAxis=Z
    twistAxis=(X=1,Y=0,Z=0)
   )
   Remap = &RigVMFunction_MathDoubleRemap(SourceMaximum=0.2, SourceMinimum=0, TargetMaximum=0, TargetMinimum=1, Value=@DeltaRotation_7.outputWeight)
   > &RigUnit_ModifyTransforms(
    ItemToModify="((Item=(Type=Bone,Name=\"Belt_Spine05_Fr\"),Transform=(Rotation=(X=0,Y=0,Z=0,W=1),Translation=(X=0,Y=0,Z=1),Scale3D=(X=1,Y=1,Z=1))),(Item=(Type=Bone,Name=\"base_bandolierFront_01\"),Transform=(Rotation=(X=0,Y=0,Z=0,W=1),Translation=(X=-1,Y=0,Z=1),Scale3D=(X=1, ...(truncated)"
    WeightMinimum=0
    Weight=@Remap.Result
   )
   > DeltaRotation_11 = RigVMFunctionReferenceNode(
    Rotation Order=ZYX
    child=(Type=Bone,Name="spine_04")
    debugRotation=false
    onlyTwist=false
    outputMinMax=(X=0,Y=1)
    parent=(Type=Null,Name="spine_04_null")
    premult=1
    rotationMinMax=(X=-40,Y=40)
    triggerAxis=Z
    twistAxis=(X=1,Y=0,Z=0)
   )
   > &RigUnit_ModifyTransforms(
    ItemToModify="((Item=(Type=Bone,Name=\"base_bandolierFront_01\"),Transform=(Rotation=(X=0,Y=0,Z=0,W=1),Translation=(X=4,Y=3,Z=0),Scale3D=(X=1,Y=1,Z=1))),(Item=(Type=Bone,Name=\"base_bandolierFront_02\"),Transform=(Rotation=(X=-0.067,Y=-0.25,Z=-0.25,W=0.933),Translation=(X=- ...(truncated)"
    WeightMinimum=0
    Weight=&RigVMFunction_MathDoubleRemap(SourceMaximum=1, SourceMinimum=0.7, TargetMaximum=1, TargetMinimum=0, Value=@DeltaRotation_11.outputWeight).Result
   )
   > &RigUnit_ModifyTransforms(
    ItemToModify="((Item=(Type=Bone,Name=\"base_bandolierFront_01\"),Transform=(Rotation=(X=0,Y=0,Z=0,W=1),Translation=(X=0,Y=0,Z=1.75),Scale3D=(X=1,Y=1,Z=1))),(Item=(Type=Bone,Name=\"Belt_Spine05_Bk\"),Transform=(Rotation=(X=0,Y=0,Z=0,W=1),Translation=(X=0,Y=0,Z=4),Scale3D=(X= ...(truncated)"
    WeightMinimum=0
    Weight=&RigVMFunction_MathDoubleRemap(SourceMaximum=0.5, SourceMinimum=0, TargetMaximum=0, TargetMinimum=1, Value=@DeltaRotation_11.outputWeight).Result
   )
   > DeltaRotation_12 = RigVMFunctionReferenceNode(
    Rotation Order=ZYX
    child=(Type=Bone,Name="spine_03")
    debugRotation=false
    onlyTwist=false
    outputMinMax=(X=0,Y=1)
    parent=(Type=Null,Name="spine_03_null")
    premult=1
    rotationMinMax=(X=-40,Y=40)
    triggerAxis=Z
    twistAxis=(X=1,Y=0,Z=0)
   )
   > &RigUnit_ModifyTransforms(
    ItemToModify="((Item=(Type=Bone,Name=\"base_bandolierFront_01\"),Transform=(Rotation=(X=-0.0008,Y=0.0436,Z=0.0174,W=0.9989),Translation=(X=0,Y=2,Z=0),Scale3D=(X=1,Y=1,Z=1))),(Item=(Type=Bone,Name=\"base_bandolierBack_01\"),Transform=(Rotation=(X=0,Y=0,Z=0,W=1),Translation=( ...(truncated)"
    WeightMinimum=0
    Weight=&RigVMFunction_MathDoubleRemap(SourceMaximum=1, SourceMinimum=0.5, TargetMaximum=1, TargetMinimum=0, Value=@DeltaRotation_12.outputWeight).Result
   )
   > &RigUnit_ModifyTransforms(
    ItemToModify="((Item=(Type=Bone,Name=\"base_bandolierFront_01\"),Transform=(Rotation=(X=0,Y=0,Z=0,W=1),Translation=(X=0,Y=0,Z=2.6),Scale3D=(X=1,Y=1,Z=1))),(Item=(Type=Bone,Name=\"base_bandolierFront_02\"),Transform=(Rotation=(X=0,Y=0,Z=0,W=1),Translation=(X=0,Y=0,Z=3),Scale ...(truncated)"
    WeightMinimum=0
    Weight=&RigVMFunction_MathDoubleRemap(SourceMaximum=0.5, SourceMinimum=0, TargetMaximum=0, TargetMinimum=1, Value=@DeltaRotation_12.outputWeight).Result
   )
   > DeltaRotation_13 = RigVMFunctionReferenceNode(
    Rotation Order=ZYX
    child=(Type=Bone,Name="spine_02")
    debugRotation=false
    onlyTwist=false
    outputMinMax=(X=0,Y=1)
    parent=(Type=Null,Name="spine_02_null")
    premult=1
    rotationMinMax=(X=-40,Y=40)
    triggerAxis=Z
    twistAxis=(X=1,Y=0,Z=0)
   )
   > &RigUnit_ModifyTransforms(
    ItemToModify="((Item=(Type=Bone,Name=\"base_bandolierFront_01\"),Transform=(Rotation=(X=-0.0008,Y=0.0436,Z=0.0174,W=0.9989),Translation=(X=0,Y=0,Z=0),Scale3D=(X=1,Y=1,Z=1))),(Item=(Type=Bone,Name=\"base_bandolierBack_01\"),Transform=(Rotation=(X=0,Y=0.0654,Z=0,W=0.9979),Tra ...(truncated)"
    WeightMinimum=0
    Weight=&RigVMFunction_MathDoubleRemap(SourceMaximum=1, SourceMinimum=0.5, TargetMaximum=1, TargetMinimum=0, Value=@DeltaRotation_13.outputWeight).Result
   )
   > &RigUnit_ModifyTransforms(
    ItemToModify="((Item=(Type=Bone,Name=\"base_bandolierFront_01\"),Transform=(Rotation=(X=0,Y=0,Z=0,W=1),Translation=(X=0,Y=0,Z=2.6),Scale3D=(X=1,Y=1,Z=1))),(Item=(Type=Bone,Name=\"base_bandolierFront_02\"),Transform=(Rotation=(X=0,Y=0,Z=0,W=1),Translation=(X=0,Y=0,Z=3),Scale ...(truncated)"
    WeightMinimum=0
    Weight=&RigVMFunction_MathDoubleRemap(SourceMaximum=0.5, SourceMinimum=0, TargetMaximum=0, TargetMinimum=1, Value=@DeltaRotation_13.outputWeight).Result
   )
   > DeltaRotation_8 = RigVMFunctionReferenceNode(
    Rotation Order=ZYX
    child=(Type=Bone,Name="spine_05")
    debugRotation=false
    onlyTwist=true
    outputMinMax=(X=0,Y=1)
    parent=(Type=Null,Name="spine_05_Spine01Delta_null")
    premult=1
    rotationMinMax=(X=0,Y=-24)
    triggerAxis=X
    twistAxis=(X=1,Y=0,Z=0)
   )
   > &RigUnit_ModifyTransforms(
    ItemToModify="((Item=(Type=Bone,Name=\"base_bandolierBack_01\"),Transform=(Rotation=(X=0,Y=0,Z=0,W=1),Translation=(X=0,Y=3,Z=-1.5),Scale3D=(X=1,Y=1,Z=1))),(Item=(Type=Bone,Name=\"base_bandolierFront_01\"),Transform=(Rotation=(X=0.4222,Y=0.0395,Z=-0.0184,W=0.9054),Translatio ...(truncated)"
    WeightMinimum=0
    Weight=@DeltaRotation_8.outputWeight
   )
   n9 = &RigVMFunction_AlphaInterp(
    InterpSpeedDecreasing=8
    InterpSpeedIncreasing=8
    Scale=1
    bClampResult=False
    Value=&RigVMFunction_MathBoolToFloat(Value=RigVMVariableNode() {variable=PlayingEmote}.Value).Result
   )
   > &RigUnit_ModifyTransforms(
    ItemToModify="((Item=(Type=Bone,Name=\"base_bandolierFront_02\"),Transform=(Rotation=(X=0,Y=0,Z=0,W=1),Translation=(X=0,Y=0,Z=1),Scale3D=(X=1,Y=1,Z=1))),(Item=(Type=Bone,Name=\"base_bandolierBack_02\"),Transform=(Rotation=(X=0,Y=0,Z=0,W=1),Translation=(X=0,Y=0,Z=-1),Scale3D ...(truncated)"
    WeightMinimum=0
    Weight=@n9.Result
   )
   n10 = &RigVMFunction_AlphaInterp(
    InterpSpeedDecreasing=6
    InterpSpeedIncreasing=10
    Scale=1
    bClampResult=true
    Value=&RigVMFunction_MathBoolToFloat(Value=RigVMVariableNode() {variable=isDBNO}.Value).Result
   )
   &RigUnit_ModifyTransforms(
    ExecutePin=()
    ItemToModify="((Item=(Type=Bone,Name=\"base_sideBag_01\"),Transform=(Rotation=(X=-0.0594,Y=-0.3368,Z=-0.1632,W=0.9254),Translation=(X=0,Y=0,Z=0),Scale3D=(X=1,Y=1,Z=1))),(Item=(Type=Bone,Name=\"base_sideBag_02\"),Transform=(Rotation=(X=0.0842,Y=-0.2578,Z=0.0226,W=0.9623),Tra ...(truncated)"
    WeightMinimum=0
    Weight=@n10.Result
   )
   > DeltaRotation_5 = RigVMFunctionReferenceNode(
    Rotation Order=ZYX
    child=(Type=Bone,Name="spine_05")
    debugRotation=false
    onlyTwist=false
    outputMinMax=(X=0,Y=1)
    parent=(Type=Null,Name="spine_05_Spine01Delta_null")
    premult=1
    rotationMinMax=(X=0,Y=-24)
    triggerAxis=Z
    twistAxis=(X=1,Y=0,Z=0)
   )
   RigVMFunction_MathBoolToFloat = &RigVMFunction_MathBoolToFloat(Value=RigVMVariableNode() {variable=PlayingEmote}.Value)
   n11 = &RigVMFunction_AlphaInterp(InterpSpeedDecreasing=8, InterpSpeedIncreasing=8, Scale=1, bClampResult=False, Value=@RigVMFunction_MathBoolToFloat.Result)
   > &RigUnit_ModifyTransforms(
    ItemToModify="((Item=(Type=Bone,Name=\"base_sideBag_01\"),Transform=(Rotation=(X=0.0046,Y=0.0434,Z=-0.1044,W=0.9936),Translation=(X=0,Y=0,Z=0),Scale3D=(X=1,Y=1,Z=1))),(Item=(Type=Bone,Name=\"base_sideBag_02\"),Transform=(Rotation=(X=-0.0872,Y=0,Z=0,W=0.9962),Translation=(X= ...(truncated)"
    WeightMinimum=0
    Weight=&RigVMFunction_MathDoubleMul(B=@n11.Result, A=@DeltaRotation_5.outputWeight).Result
   )
   > DeltaRotation_6 = RigVMFunctionReferenceNode(
    Rotation Order=ZYX
    child=(Type=Bone,Name="spine_05")
    debugRotation=false
    onlyTwist=false
    outputMinMax=(X=0,Y=1)
    parent=(Type=Null,Name="spine_05_Spine01Delta_null")
    premult=1
    rotationMinMax=(X=0,Y=10)
    triggerAxis=Y
    twistAxis=(X=1,Y=0,Z=0)
   )
   > &RigUnit_ModifyTransforms(
    ItemToModify="((Item=(Type=Bone,Name=\"base_sideBag_01\"),Transform=(Rotation=(X=0.0057,Y=0.0432,Z=-0.1304,W=0.9905),Translation=(X=0,Y=0,Z=0),Scale3D=(X=1,Y=1,Z=1))),(Item=(Type=Bone,Name=\"base_sideBag_02\"),Transform=(Rotation=(X=-0.0872,Y=0,Z=0,W=0.9962),Translation=(X= ...(truncated)"
    WeightMinimum=0
    Weight=&RigVMFunction_MathDoubleMul(A=@DeltaRotation_6.outputWeight, B=@n11.Result).Result
   )
   &RigVMFunction_AlphaInterp(
    InterpSpeedDecreasing=8
    InterpSpeedIncreasing=8
    Scale=1
    bClampResult=False
    Value=&RigVMFunction_MathBoolToFloat(Value=RigVMVariableNode() {variable=PlayingEmote}.Value).Result
   )
  ]}
 CollapseNode_ContainedGraph {nodes [
   RigVMFunctionEntryNode()
   > DeltaRotation = RigVMFunctionReferenceNode(
    Rotation Order=ZYX
    child=(Type=Bone,Name="calf_r")
    outputMinMax=(X=0,Y=1)
    parent=(Type=Null,Name="calf_r_null")
    premult=1
    rotationMinMax=(X=0,Y=100)
    triggerAxis=Z
    twistAxis=(X=1,Y=0,Z=0)
   )
   > &RigUnit_ModifyTransforms(
    ItemToModify=((Item=(Type=Bone,Name="Kneepad_Helper"),Transform=(Rotation=(X=0.4226,Y=0,Z=0,W=0.9063),Translation=(X=0,Y=-3,Z=5.5),Scale3D=(X=1,Y=1,Z=1))))
    WeightMinimum=0
    Weight=@DeltaRotation.outputWeight
   )
   > RigVMFunctionReturnNode()
  ]}
 CollapseNode_ContainedGraph {nodes [
   RigVMFunctionEntryNode()
   > RigVMFunction_Sequence = RigVMAggregateNode()
   GetAlphaFromQuatDelta_3 = RigVMFunctionReferenceNode(
    Rotation Order=ZYX
    child=(Type=Bone,Name="upperarm_r")
    debugRotation=false
    outputMinMax=(X=0,Y=1)
    parent=(Type=Null,Name="upperarm_r_null")
    premult=1
    rotationMinMax=(X=0,Y=30)
    triggerAxis=Y
    twistAxis=(X=1,Y=0,Z=0)
    ExecuteContext=@RigVMFunction_Sequence.A
   )
   > &RigUnit_ModifyTransforms(
    ItemToModify=((Item=(Type=Bone,Name="base_Sleeves_r"),Transform=(Rotation=(X=0,Y=-0.1478,Z=0,W=0.989),Translation=(X=0,Y=0,Z=-0.5),Scale3D=(X=1,Y=1,Z=1))))
    WeightMinimum=0
    Weight=@GetAlphaFromQuatDelta_3.alpha
   )
   > n1 = &RigUnit_SphericalPoseReader(
    ActiveRegionScaleFactors=(PositiveWidth=0.5,NegativeWidth=0.5,PositiveHeight=0.5,NegativeHeight=0.5)
    ActiveRegionSize=0.15
    Debug=(bDrawDebug=false,bDraw2D=False,bDrawLocalAxes=False,DebugScale=25,DebugSegments=20,DebugThickness=0.25)
    DriverAxis=(X=-1,Y=0,Z=0)
    DriverItem=(Type=Bone,Name="upperarm_r")
    FalloffRegionScaleFactors=(PositiveWidth=1,NegativeWidth=1,PositiveHeight=1,NegativeHeight=1)
    FalloffSize=0.2
    FlipHeightScaling=False
    FlipWidthScaling=False
    OptionalParentItem=(Type=Bone,Name="clavicle_r")
    RotationOffset=(X=120,Y=0,Z=-145)
   )
   > &RigUnit_ModifyTransforms(
    ItemToModify="((Item=(Type=Bone,Name=\"base_Sleeves_r\"),Transform=(Rotation=(X=-0.0046,Y=0.0434,Z=0.1044,W=0.9936),Translation=(X=0,Y=0,Z=0),Scale3D=(X=1,Y=1,Z=1))),(Item=(Type=Bone,Name=\"base_backSleeve_r\"),Transform=(Rotation=(X=0,Y=-0.2588,Z=0,W=0.9659),Translation=(X ...(truncated)"
    WeightMinimum=0
    Weight=@n1.OutputParam
   )
   > GetAlphaFromQuatDelta_4 = RigVMFunctionReferenceNode(
    Rotation Order=ZYX
    child=(Type=Bone,Name="lowerarm_r")
    debugRotation=false
    outputMinMax=(X=0,Y=1)
    parent=(Type=Null,Name="lowerarm_r_null")
    premult=1
    rotationMinMax=(X=40,Y=90)
    triggerAxis=Z
    twistAxis=(X=1,Y=0,Z=0)
   )
   > &RigUnit_ModifyTransforms(
    ItemToModify="((Item=(Type=Bone,Name=\"base_frontSleeve_r\"),Transform=(Rotation=(X=0,Y=0,Z=0.4226,W=0.9063),Translation=(X=-20,Y=-5,Z=5),Scale3D=(X=1,Y=1,Z=1))),(Item=(Type=Bone,Name=\"base_TopSleeve_r\"),Transform=(Rotation=(X=0,Y=0.2588,Z=0,W=0.9659),Translation=(X=0,Y=0 ...(truncated)"
    WeightMinimum=0
    Weight=@GetAlphaFromQuatDelta_4.alpha
   )
   GetAlphaFromQuatDelta_2 = RigVMFunctionReferenceNode(
    Rotation Order=ZYX
    child=(Type=Bone,Name="upperarm_l")
    debugRotation=false
    outputMinMax=(X=0,Y=1)
    parent=(Type=Null,Name="upperarm_l_null")
    premult=1
    rotationMinMax=(X=0,Y=30)
    triggerAxis=Y
    twistAxis=(X=1,Y=0,Z=0)
    ExecuteContext=@RigVMFunction_Sequence.B
   )
   > &RigUnit_ModifyTransforms(
    ItemToModify=((Item=(Type=Bone,Name="base_sleeves_l"),Transform=(Rotation=(X=0,Y=-0.1219,Z=0,W=0.9925),Translation=(X=0,Y=0,Z=0.5),Scale3D=(X=1,Y=1,Z=1))))
    WeightMinimum=0
    Weight=@GetAlphaFromQuatDelta_2.alpha
   )
   > n2 = &RigUnit_SphericalPoseReader(
    ActiveRegionScaleFactors=(PositiveWidth=0.5,NegativeWidth=0.5,PositiveHeight=0.5,NegativeHeight=0.5)
    ActiveRegionSize=0.1
    Debug=(bDrawDebug=false,bDraw2D=False,bDrawLocalAxes=False,DebugScale=25,DebugSegments=20,DebugThickness=0.25)
    DriverAxis=(X=1,Y=0,Z=0)
    DriverItem=(Type=Bone,Name="upperarm_l")
    FalloffRegionScaleFactors=(PositiveWidth=1,NegativeWidth=1,PositiveHeight=1,NegativeHeight=1)
    FalloffSize=0.2
    FlipHeightScaling=False
    FlipWidthScaling=False
    OptionalParentItem=(Type=Bone,Name="clavicle_l")
    RotationOffset=(X=50,Y=-55,Z=0)
   )
   > &RigUnit_ModifyTransforms(
    ItemToModify=((Item=(Type=Bone,Name="base_sleeves_l"),Transform=(Rotation=(X=-0.1063,Y=-0.2495,Z=-0.1063,W=0.9566),Translation=(X=0,Y=-1,Z=-1.5),Scale3D=(X=1,Y=1,Z=1))))
    WeightMinimum=0
    Weight=@n2.OutputParam
   )
   > n3 = &RigUnit_SphericalPoseReader(
    ActiveRegionScaleFactors=(PositiveWidth=0.5,NegativeWidth=0.5,PositiveHeight=0.5,NegativeHeight=0.5)
    ActiveRegionSize=0.1
    Debug=(bDrawDebug=false,bDraw2D=False,bDrawLocalAxes=False,DebugScale=25,DebugSegments=20,DebugThickness=0.25)
    DriverAxis=(X=1,Y=0,Z=0)
    DriverItem=(Type=Bone,Name="upperarm_l")
    FalloffRegionScaleFactors=(PositiveWidth=1,NegativeWidth=1,PositiveHeight=1,NegativeHeight=1)
    FalloffSize=0.2
    FlipHeightScaling=False
    FlipWidthScaling=False
    OptionalParentItem=(Type=Bone,Name="clavicle_l")
    RotationOffset=(X=0,Y=-50,Z=12)
   )
   RigVMFunction_MathBoolToFloat = &RigVMFunction_MathBoolToFloat(Value=RigVMVariableNode() {variable=bigGunAiming}.Value)
   AlphaInterp = &RigVMFunction_AlphaInterp(InterpSpeedDecreasing=10, InterpSpeedIncreasing=10, Scale=1, bClampResult=False, Value=@RigVMFunction_MathBoolToFloat.Result)
   RigVMFunction_MathBoolToInteger = &RigVMFunction_MathBoolToInteger(Value=RigVMVariableNode() {variable=Crouching}.Value)
   n4 = &RigVMFunction_AlphaInterp(
    InterpSpeedDecreasing=10
    InterpSpeedIncreasing=10
    Scale=1
    bClampResult=False
    Value=RigVMDispatchNode(Values=(1,0.5), Index=@RigVMFunction_MathBoolToInteger.Result) {template="DISPATCH_RigVMDispatch_SelectInt32(in Index,in Values,out Result)"}.Result
   )
   Multiply = RigVMAggregateNode(B=@AlphaInterp.Result, C=@n4.Result, A=@n3.OutputParam)
   > &RigUnit_ModifyTransforms(
    ItemToModify=((Item=(Type=Bone,Name="base_sleeves_l"),Transform=(Rotation=(X=-0.2593,Y=0.0797,Z=0.0393,W=0.9617),Translation=(X=-1,Y=-1,Z=-1.5),Scale3D=(X=1,Y=1,Z=1))))
    WeightMinimum=0
    Weight=@Multiply.Result
   )
   > GetAlphaFromQuatDelta_5 = RigVMFunctionReferenceNode(
    Rotation Order=ZYX
    child=(Type=Bone,Name="lowerarm_l")
    debugRotation=false
    outputMinMax=(X=0,Y=1)
    parent=(Type=Null,Name="lowerarm_l_null")
    premult=1
    rotationMinMax=(X=40,Y=90)
    triggerAxis=Z
    twistAxis=(X=1,Y=0,Z=0)
   )
   > &RigUnit_ModifyTransforms(
    ItemToModify="((Item=(Type=Bone,Name=\"base_frontSleeve_l\"),Transform=(Rotation=(X=0,Y=0,Z=0.4226,W=0.9063),Translation=(X=-15,Y=-5,Z=5),Scale3D=(X=1,Y=1,Z=1))),(Item=(Type=Bone,Name=\"base_TopSleeve_l_01\"),Transform=(Rotation=(X=0,Y=0.342,Z=0,W=0.9397),Translation=(X=0,Y ...(truncated)"
    WeightMinimum=0
    Weight=@GetAlphaFromQuatDelta_5.alpha
   )
   RigVMFunctionReturnNode(ExecuteContext=@RigVMFunction_Sequence.C)
  ]}
 CollapseNode_ContainedGraph {nodes [
   n1 = RigVMFunctionEntryNode()
   n2 = &RigUnit_SphericalPoseReader(
    ActiveRegionScaleFactors=(PositiveWidth=1,NegativeWidth=1,PositiveHeight=1,NegativeHeight=1)
    ActiveRegionSize=0.05
    Debug=(bDrawDebug=false,bDraw2D=False,bDrawLocalAxes=False,DebugScale=25,DebugSegments=20,DebugThickness=0.25)
    DriverAxis=(X=-1,Y=0,Z=0)
    DriverItem=(Type=Bone,Name="upperarm_r")
    FalloffRegionScaleFactors=(PositiveWidth=1,NegativeWidth=1,PositiveHeight=1,NegativeHeight=1)
    FalloffSize=0.2
    FlipHeightScaling=False
    FlipWidthScaling=False
    OptionalParentItem=(Type=Bone,Name="None")
    RotationOffset=(X=0,Y=45,Z=20)
    ExecutePin=@n1.ExecutePin
   )
   > n3 = &RigUnit_ModifyTransforms(
    ItemToModify=((Item=(Type=Bone,Name="upperarm_twist_01_ShirtOutPush"),Transform=(Rotation=(X=0,Y=0,Z=0,W=1),Translation=(X=-2,Y=-4,Z=-1),Scale3D=(X=1,Y=1,Z=1))))
    WeightMinimum=0
    Weight=@n2.OutputParam
   )
   RigVMFunctionReturnNode(ExecutePin=@n3.ExecutePin)
  ]}
 CollapseNode_ContainedGraph {nodes [
   RigVMFunctionEntryNode()
   > GetAlphaFromQuatDelta = RigVMFunctionReferenceNode(
    Rotation Order=ZYX
    child=(Type=Bone,Name="thigh_l")
    debugRotation=false
    outputMinMax=(X=0,Y=1)
    parent=(Type=Null,Name="thigh_l_null")
    premult=1
    rotationMinMax=(X=30,Y=85)
    triggerAxis=Z
    twistAxis=(X=1,Y=0,Z=0)
   )
   > &RigUnit_ModifyTransforms(
    ItemToModify="((Item=(Type=Bone,Name=\"noDynamics_SideBag\"),Transform=(Rotation=(X=-0.2578,Y=-0.0842,Z=-0.0226,W=0.9623),Translation=(X=0,Y=-0.5,Z=3.5),Scale3D=(X=1.05,Y=1.05,Z=1.05))),(Item=(Type=Bone,Name=\"noDynamics_SideBagScale\"),Transform=(Rotation=(X=0,Y=0,Z=0,W=1) ...(truncated)"
    WeightMinimum=0
    Weight=@GetAlphaFromQuatDelta.alpha
   )
   > GetAlphaFromQuatDelta_1 = RigVMFunctionReferenceNode(
    Rotation Order=ZYX
    child=(Type=Bone,Name="thigh_l")
    debugRotation=false
    outputMinMax=(X=0,Y=1)
    parent=(Type=Null,Name="thigh_l_null")
    premult=1
    rotationMinMax=(X=-10,Y=-75)
    triggerAxis=Y
    twistAxis=(X=1,Y=0,Z=0)
   )
   > &RigUnit_ModifyTransforms(
    ItemToModify="((Item=(Type=Bone,Name=\"noDynamics_SideBag\"),Transform=(Rotation=(X=0,Y=0,Z=-0.5299,W=0.848),Translation=(X=1,Y=-2,Z=0),Scale3D=(X=1,Y=1,Z=1))),(Item=(Type=Bone,Name=\"noDynamics_SideBagScale\"),Transform=(Rotation=(X=0,Y=0,Z=0.1305,W=0.9914),Translation=(X= ...(truncated)"
    WeightMinimum=0
    Weight=@GetAlphaFromQuatDelta_1.alpha
   )
   > GetAlphaFromQuatDelta_2 = RigVMFunctionReferenceNode(
    Rotation Order=ZYX
    child=(Type=Bone,Name="thigh_l")
    debugRotation=false
    outputMinMax=(X=0,Y=1)
    parent=(Type=Null,Name="thigh_l_null")
    premult=1
    rotationMinMax=(X=-40,Y=-90)
    triggerAxis=Z
    twistAxis=(X=1,Y=0,Z=0)
   )
   > &RigUnit_ModifyTransforms(
    ItemToModify="((Item=(Type=Bone,Name=\"noDynamics_SideBag\"),Transform=(Rotation=(X=0.0436,Y=0,Z=0,W=0.999),Translation=(X=-1.5,Y=-1,Z=0),Scale3D=(X=1,Y=1,Z=1))),(Item=(Type=Bone,Name=\"noDynamics_SideBagScale\"),Transform=(Rotation=(X=0,Y=0,Z=0,W=1),Translation=(X=0,Y=0,Z= ...(truncated)"
    WeightMinimum=0
    Weight=@GetAlphaFromQuatDelta_2.alpha
   )
   > RigVMFunctionReturnNode()
  ]}
 Multiply_1_ContainedGraph {nodes [
   n1 = RigVMFunctionEntryNode()
   Multiply = &RigVMFunction_MathDoubleMul(A=@n1.A, B=@n1.B)
   RigVMFunctionReturnNode(Result=&RigVMFunction_MathDoubleMul(B=@n1.C, A=@Multiply.Result).Result)
  ]}
 Multiply_4_ContainedGraph {nodes [
   n1 = RigVMFunctionEntryNode()
   RigVMFunctionReturnNode(Result=&RigVMFunction_MathDoubleMul(B=@n1.C, A=&RigVMFunction_MathDoubleMul(A=@n1.A, B=@n1.B).Result).Result)
  ]}
 New Function_ContainedGraph {nodes [
   n1 = RigVMFunctionEntryNode()
   RigVMFunction_MathBoolToInteger = &RigVMFunction_MathBoolToInteger(Value=@n1.onlyTwist)
   RigVMFunction_MathQuaternionSwingTwist = &RigVMFunction_MathQuaternionSwingTwist(
    TwistAxis=@n1.twistAxis
    Input=&RigUnit_GetRelativeTransformForItem(bChildInitial=False, bParentInitial=False, Child=@n1.child, Parent=@n1.parent).RelativeTransform.Rotation
   )
   n2 = RigVMDispatchNode(
    Values=((),())
    Index=@RigVMFunction_MathBoolToInteger.Result
    Values.0=@RigVMFunction_MathQuaternionSwingTwist.Swing
    Values.1=@RigVMFunction_MathQuaternionSwingTwist.Twist
   ) {template="DISPATCH_RigVMDispatch_SelectInt32(in Index,in Values,out Result)"}
   RigVMFunction_MathQuaternionToEuler = &RigVMFunction_MathQuaternionToEuler(Value=@n2.Result, RotationOrder=@n1.Rotation Order)
   > DISPATCH_RigVMDispatch_Print = RigVMDispatchNode(ScreenColor=(R=1,G=1,B=1,A=1), ScreenDuration=1.5, Enabled=@n1.debugRotation, Value=@RigVMFunction_MathQuaternionToEuler.Result) {template="DISPATCH_RigVMDispatch_Print(in Prefix,in Value,in Enabled,in ScreenDuration,in ScreenColor)"}
   Cast = RigVMDispatchNode(Value=@n1.triggerAxis) {template="DISPATCH_RigVMDispatch_CastEnumToInt(in Value,out Result)"}
   DISPATCH_RigVMDispatch_SelectInt32 = RigVMDispatchNode(
    Values=(0,0,0)
    Index=@Cast.Result
    Values.0=@RigVMFunction_MathQuaternionToEuler.Result.X
    Values.1=@RigVMFunction_MathQuaternionToEuler.Result.Y
    Values.2=@RigVMFunction_MathQuaternionToEuler.Result.Z
   ) {template="DISPATCH_RigVMDispatch_SelectInt32(in Index,in Values,out Result)"}
   Remap = &RigVMFunction_MathDoubleRemap(
    bClamp=False
    Value=@DISPATCH_RigVMDispatch_SelectInt32.Result
    TargetMinimum=@n1.outputMinMax.X
    TargetMaximum=@n1.outputMinMax.Y
    SourceMinimum=@n1.rotationMinMax.X
    SourceMaximum=@n1.rotationMinMax.Y
   )
   Absolute = &RigVMFunction_MathDoubleAbs(Value=@Remap.Result)
   Multiply = &RigVMFunction_MathDoubleMul(
    A=&RigVMFunction_MathDoubleClamp(Value=RigVMDispatchNode(True=@Absolute.Result, Condition=@n1.absoluteValue, False=@Remap.Result) {template="DISPATCH_RigVMDispatch_If(in Condition,in True,in False,out Result)"}.Result).Result
    B=@n1.premult
   )
   > RigVMFunctionReturnNode(outputQuat=@n2.Result, outputWeight=@Multiply.Result)
  ]}
 New Function_ContainedGraph {nodes [
   RigVMFunctionEntryNode()
   RigVMFunction_MathBoolToInteger = &RigVMFunction_MathBoolToInteger(Value=RigVMVariableNode() {variable=onlyTwist}.Value)
   RigVMFunction_MathQuaternionSwingTwist = &RigVMFunction_MathQuaternionSwingTwist(
    Input=&RigUnit_GetRelativeTransformForItem(
     bChildInitial=False
     bParentInitial=False
     child=RigVMVariableNode() {variable=child}.Value
     parent=RigVMVariableNode() {variable=parent}.Value
    ).RelativeTransform.Rotation
    twistAxis=RigVMVariableNode() {variable=twistAxis}.Value
   )
   n1 = RigVMDispatchNode(
    Values=((),())
    Index=@RigVMFunction_MathBoolToInteger.Result
    Values.0=@RigVMFunction_MathQuaternionSwingTwist.Swing
    Values.1=@RigVMFunction_MathQuaternionSwingTwist.Twist
   ) {template="DISPATCH_RigVMDispatch_SelectInt32(in Index,in Values,out Result)"}
   > RigVMVariableNode(Value=@n1.Result) {variable=entry_outputQuat}
   RigVMFunction_MathQuaternionToEuler = &RigVMFunction_MathQuaternionToEuler(Value=@n1.Result, RotationOrder=RigVMVariableNode() {variable=Rotation Order}.Value)
   > DISPATCH_RigVMDispatch_Print = RigVMDispatchNode(
    ScreenColor=(R=1,G=1,B=1,A=1)
    ScreenDuration=1.5
    Value=@RigVMFunction_MathQuaternionToEuler.Result
    Enabled=RigVMVariableNode() {variable=debugRotation}.Value
   ) {template="DISPATCH_RigVMDispatch_Print(in Prefix,in Value,in Enabled,in ScreenDuration,in ScreenColor)"}
   Cast = RigVMDispatchNode(Value=RigVMVariableNode() {variable=triggerAxis}.Value) {template="DISPATCH_RigVMDispatch_CastEnumToInt(in Value,out Result)"}
   DISPATCH_RigVMDispatch_SelectInt32 = RigVMDispatchNode(
    Values=(0,0,0)
    Index=@Cast.Result
    Values.0=@RigVMFunction_MathQuaternionToEuler.Result.X
    Values.1=@RigVMFunction_MathQuaternionToEuler.Result.Y
    Values.2=@RigVMFunction_MathQuaternionToEuler.Result.Z
   ) {template="DISPATCH_RigVMDispatch_SelectInt32(in Index,in Values,out Result)"}
   n2 = RigVMVariableNode() {variable=rotationMinMax}
   n3 = RigVMVariableNode() {variable=outputMinMax}
   Remap = &RigVMFunction_MathDoubleRemap(
    bClamp=False
    Value=@DISPATCH_RigVMDispatch_SelectInt32.Result
    SourceMinimum=@n2.Value.X
    SourceMaximum=@n2.Value.Y
    TargetMinimum=@n3.Value.X
    TargetMaximum=@n3.Value.Y
   )
   Absolute = &RigVMFunction_MathDoubleAbs(Value=@Remap.Result)
   n4 = RigVMVariableNode() {variable=absoluteValue}
   Multiply = &RigVMFunction_MathDoubleMul(
    A=&RigVMFunction_MathDoubleClamp(Value=RigVMDispatchNode(True=@Absolute.Result, False=@Remap.Result, Condition=@n4.Value) {template="DISPATCH_RigVMDispatch_If(in Condition,in True,in False,out Result)"}.Result).Result
    B=RigVMVariableNode() {variable=premult}.Value
   )
   > RigVMFunctionReturnNode(alpha=@Multiply.Result, rawQuat=RigVMVariableNode() {variable=entry_outputQuat}.Value)
  ]}
 New Function_ContainedGraph {nodes [
   RigVMFunctionEntryNode()
   > DeltaRotation v4 = RigVMCollapseNode(
    Rotation Order=ZYX
    child=(Type=Bone,Name="hand_l")
    debugRotation=false
    onlyTwist=true
    outputMinMax=(X=0,Y=1)
    parent=(Type=Null,Name="hand_l_LowerArmDeltaRef_null")
    premult=1
    rotationMinMax=(X=70,Y=160)
    twistAxis=(X=1,Y=0,Z=0)
   )
   RigVMFunction_AnimEvalRichCurve = &RigVMFunction_AnimEvalRichCurve(
    Curve=(EditorCurveData=(Keys=((InterpMode=RCIM_Cubic,TangentMode=RCTM_User,ArriveTangent=1.6777,LeaveTangent=1.6777),(InterpMode=RCIM_Cubic,Time=0.5,Value=0.75,ArriveTangent=1,LeaveTangent=1),(InterpMode=RCIM_Cubic,Time=1,Value=1))))
    SourceMinimum=0
    TargetMinimum=0
    Value=@"DeltaRotation v4".outputWeight
   )
   > &RigUnit_ModifyTransforms(
    ItemToModify="((Item=(Type=Bone,Name=\"lowerarm_twist_01_l\"),Transform=(Rotation=(X=0,Y=0.0436,Z=0,W=0.999),Translation=(X=-0.7,Y=-0.17,Z=1),Scale3D=(X=1,Y=1,Z=1.4))),(Item=(Type=Bone,Name=\"lowerarm_twist_02_l\"),Transform=(Rotation=(X=0.3407,Y=0.0298,Z=0.0819,W=0.9361),T ...(truncated)"
    WeightMinimum=0
    Weight=@RigVMFunction_AnimEvalRichCurve.Result
   )
   > RigVMFunctionReturnNode()
  ]}
 New Function_ContainedGraph {nodes [
   n1 = RigVMFunctionEntryNode()
   RigVMFunction_MathBoolToInteger = &RigVMFunction_MathBoolToInteger(Value=@n1.onlyTwist)
   RigVMFunction_MathQuaternionSwingTwist = &RigVMFunction_MathQuaternionSwingTwist(
    TwistAxis=@n1.twistAxis
    Input=&RigUnit_GetRelativeTransformForItem(bChildInitial=False, bParentInitial=False, Child=@n1.child, Parent=@n1.parent).RelativeTransform.Rotation
   )
   n2 = RigVMDispatchNode(
    Values=((),())
    Index=@RigVMFunction_MathBoolToInteger.Result
    Values.0=@RigVMFunction_MathQuaternionSwingTwist.Swing
    Values.1=@RigVMFunction_MathQuaternionSwingTwist.Twist
   ) {template="DISPATCH_RigVMDispatch_SelectInt32(in Index,in Values,out Result)"}
   RigVMFunction_MathQuaternionToEuler = &RigVMFunction_MathQuaternionToEuler(Value=@n2.Result, RotationOrder=@n1.Rotation Order)
   > DISPATCH_RigVMDispatch_Print = RigVMDispatchNode(ScreenColor=(R=1,G=1,B=1,A=1), ScreenDuration=1.5, Enabled=@n1.debugRotation, Value=@RigVMFunction_MathQuaternionToEuler.Result) {template="DISPATCH_RigVMDispatch_Print(in Prefix,in Value,in Enabled,in ScreenDuration,in ScreenColor)"}
   Cast = RigVMDispatchNode(Value=@n1.triggerAxis) {template="DISPATCH_RigVMDispatch_CastEnumToInt(in Value,out Result)"}
   DISPATCH_RigVMDispatch_SelectInt32 = RigVMDispatchNode(
    Values=(0,0,0)
    Index=@Cast.Result
    Values.0=@RigVMFunction_MathQuaternionToEuler.Result.X
    Values.1=@RigVMFunction_MathQuaternionToEuler.Result.Y
    Values.2=@RigVMFunction_MathQuaternionToEuler.Result.Z
   ) {template="DISPATCH_RigVMDispatch_SelectInt32(in Index,in Values,out Result)"}
   Remap = &RigVMFunction_MathDoubleRemap(
    bClamp=False
    Value=@DISPATCH_RigVMDispatch_SelectInt32.Result
    TargetMinimum=@n1.outputMinMax.X
    TargetMaximum=@n1.outputMinMax.Y
    SourceMinimum=@n1.rotationMinMax.X
    SourceMaximum=@n1.rotationMinMax.Y
   )
   Absolute = &RigVMFunction_MathDoubleAbs(Value=@Remap.Result)
   Multiply = &RigVMFunction_MathDoubleMul(
    A=&RigVMFunction_MathDoubleClamp(Value=RigVMDispatchNode(True=@Absolute.Result, Condition=@n1.absoluteValue, False=@Remap.Result) {template="DISPATCH_RigVMDispatch_If(in Condition,in True,in False,out Result)"}.Result).Result
    B=@n1.premult
   )
   > RigVMFunctionReturnNode(outputQuat=@n2.Result, outputWeight=@Multiply.Result)
  ]}
 RigVMFunctionLibrary {nodes [
   DeltaRotation = RigVMCollapseNode(
    ExecuteContext=()
    Rotation Order=ZYX
    child=(Type=Bone,Name="None")
    outputMinMax=(X=0,Y=1)
    parent=(Type=Bone,Name="None")
    premult=1
    rotationMinMax=(X=0,Y=0)
    twistAxis=(X=1,Y=0,Z=0)
   )
   GetAlphaFromQuatDelta = RigVMCollapseNode(
    ExecuteContext=()
    Rotation Order=ZYX
    child=(Type=Bone,Name="None")
    outputMinMax=(X=0,Y=1)
    parent=(Type=Bone,Name="None")
    premult=1
    rotationMinMax=(X=0,Y=0)
    twistAxis=(X=1,Y=0,Z=0)
   )
   TwistExractionFix = RigVMCollapseNode(ExecuteContext=())
  ]}
 RigVMFunction_MathBoolAnd_2_ContainedGraph {nodes [
   n1 = RigVMFunctionEntryNode()
   RigVMFunction_MathBoolAnd = &RigVMFunction_MathBoolAnd(A=@n1.A, B=@n1.B)
   RigVMFunctionReturnNode(Result=&RigVMFunction_MathBoolAnd(B=@n1.C, A=@RigVMFunction_MathBoolAnd.Result).Result)
  ]}
 RigVMFunction_MathBoolAnd_2_ContainedGraph {nodes [
   n1 = RigVMFunctionEntryNode()
   RigVMFunctionReturnNode(Result=&RigVMFunction_MathBoolAnd(B=@n1.C, A=&RigVMFunction_MathBoolAnd(A=@n1.A, B=@n1.B).Result).Result)
  ]}
 RigVMFunction_Sequence_1_ContainedGraph {nodes [
   n1 = RigVMFunctionEntryNode()
   RigVMFunction_Sequence = &RigVMFunction_Sequence(ExecuteContext=@n1.ExecuteContext)
   n2 = &RigVMFunction_Sequence(ExecuteContext=@RigVMFunction_Sequence.B)
   RigVMFunctionReturnNode(A=@RigVMFunction_Sequence.A, B=@n2.A, C=@n2.B)
  ]}
 RigVMFunction_Sequence_1_ContainedGraph {nodes [
   n1 = RigVMFunctionEntryNode()
   RigVMFunction_Sequence = &RigVMFunction_Sequence(ExecuteContext=@n1.ExecuteContext)
   n2 = &RigVMFunction_Sequence(ExecuteContext=@RigVMFunction_Sequence.B)
   RigVMFunctionReturnNode(A=@RigVMFunction_Sequence.A, B=@n2.A, C=@n2.B)
  ]}
 RigVMFunction_Sequence_3_ContainedGraph {nodes [
   n1 = RigVMFunctionEntryNode()
   n2 = &RigVMFunction_Sequence(ExecuteContext=@n1.ExecuteContext)
   n3 = &RigVMFunction_Sequence(ExecuteContext=@n2.B)
   n4 = &RigVMFunction_Sequence(ExecuteContext=@n3.B)
   n5 = &RigVMFunction_Sequence(ExecuteContext=@n4.B)
   RigVMFunctionReturnNode(B=@n3.A, A=@n2.A, C=@n4.A, D=@n5.A, E=@n5.B)
  ]}
 RigVMModel {nodes [
   RigUnit_BeginExecution = &RigUnit_BeginExecution()
   FN_DeformRig_FNC = RigVMFunctionReferenceNode(
    DeformRig_ Curve="(deformRig_l_upperarm_up_alpha_72_#1=False,deformRig_r_upperarm_up_alpha_73_#2=False,deformRig_l_upperarm_fr_alpha_74_#3=False,deformRig_r_upperarm_fr_alpha_75_CB51B8 ...(truncated)"
    DeformRig_Bool="(upperarm_1_#4=True,elbow_3_#5=True,wrist_7_#6=True,groin_8_#7=True,glute_17_#8=True,knee_9_40C408324CA261 ...(truncated)"
    DeformRig_Debug=(upperarm_debug_18_#4=False,wrist_debug_20_#6=False,lat_debug_26_#9=False,groinGlute_debug_27_#7=False)
    DeformRig_Value="(upperarm_up_18_#10=(Rotation=(X=0.0,Y=-0.0,Z=0.0,W=1.0),Translation=(X=0.0,Y=0.0,Z=8.0),Scale3D=(X=0.70,Y=1.0,Z=1.0)),upperarm_fwd_fr_19_#11=(Rotation=(X=0.0,Y=-0.0,Z=0.0,W=1.0),Translation=(X=0.0, ...(truncated)"
    ExecuteContext=@RigUnit_BeginExecution.ExecutePin
   )
   > Deltas = RigVMCollapseNode()
   > TwistExractionFix = RigVMFunctionReferenceNode()
   Constraints = RigVMCollapseNode(ExecutePin=@TwistExractionFix.ExecuteContext)
   Thighs = RigVMCollapseNode(ExecuteContext=@Constraints.ExecutePin)
   Sliding = RigVMCollapseNode(ExecutePin=@Thighs.ExecuteContext)
   Shoes = RigVMCollapseNode(ExecuteContext=@Sliding.ExecutePin)
   > "wrist " = RigVMCollapseNode()
   > SleeveAdjustments = RigVMCollapseNode()
   ArmpitOutPushJoint = RigVMCollapseNode(ExecutePin=@SleeveAdjustments.ExecuteContext)
   KneePadAdjustment = RigVMCollapseNode(ExecuteContext=@ArmpitOutPushJoint.ExecutePin)
   NewBeltSetup = RigVMCollapseNode(ExecutePin=@KneePadAdjustment.ExecuteContext)
   ASW_AntiClippingSetup = RigVMCollapseNode(ExecuteContext=@NewBeltSetup.ExecutePin)
   Bag_Belt_Setup = RigVMCollapseNode(ExecutePin=())
   CrouchMove = RigVMCollapseNode(ExecuteContext=())
   &RigUnit_ModifyTransforms(
    ExecutePin=()
    ItemToModify="((Item=(Type=Bone,Name=\"dyn_tshirt_front_01\"),Transform=(Rotation=(X=-0.0274,Y=0.1834,Z=-0.1452,W=0.9719),Translation=(X=0,Y=0,Z=0),Scale3D=(X=1,Y=1,Z=1))),(Item=(Type=Bone,Name=\"dyn_tshirt_front_l_01\"),Transform=(Rotation=(X=0.0013,Y=0.0627,Z=0.1803,W=0.9 ...(truncated)"
    Weight=1
    WeightMinimum=0
   )
   Sleeves = RigVMCollapseNode(ExecuteContext=())
  ]}
]
hierarchy [
 root {loc=<0 0 0>, rot=<0 0 0 1>}
 pelvis {parent=root, loc=<0 -0.5301 98.0571>, rot=<0.0127 0.707 -0.0127 -0.707>}
 spine_01 {parent=pelvis, loc=<6.1208 -0.9328 0>, rot=<0 0 0.034 -0.9994>}
 spine_02 {parent=spine_01, loc=<6.5997 -0.2095 0>, rot=<0 0 0.0238 -0.9997>}
 spine_03 {parent=spine_02, loc=<6.4914 -0.1237 0>, rot=<0 0 0.0149 -0.9999>}
 spine_04 {parent=spine_03, loc=<6.9598 0.0794 0>, rot=<0 0 -0.0132 -0.9999>}
 spine_05 {parent=spine_04, loc=<6.9987 0 0>, rot=<0 0 -0.0015 -1>}
 neck_01 {parent=spine_05, loc=<11.3138 0.1529 0>, rot=<0 0 -0.2003 -0.9797>}
 neck_02 {parent=neck_01, loc=<4.5627 0 0>, rot=<0 0 0.0504 -0.9987>}
 head {parent=neck_02, loc=<5.3914 0 0>, rot=<0 0 0.115 -0.9934>}
 FACIAL_C_Jaw {parent=head, loc=<0.1977 0.9014 -0.0022>, rot=<-0.6219 0.3363 0.6221 0.3363>}
 FACIAL_C_Jawline {parent=FACIAL_C_Jaw, loc=<-0.0002 2.9947 7.3016>, rot=<0.2586 0 0 -0.966>}
 FACIAL_L_ChinSide {parent=FACIAL_C_Jaw, loc=<2.7182 0.9743 7.5815>, rot=<-0.2208 -0.3738 0.0499 -0.8995>}
 FACIAL_R_ChinSide {parent=FACIAL_C_Jaw, loc=<-2.7185 0.976 7.5838>, rot=<-0.2208 0.3738 -0.0499 -0.8995>}
 FACIAL_C_Chin {parent=FACIAL_C_Jaw, loc=<-0.0013 1.4888 7.8064>, rot=<-0.2591 0 0 -0.9659>}
 head_null {parent=neck_02, type=Null, loc=<5.3914 0 0>, rot=<0 0 0.115 -0.9934>}
 neck_02_null {parent=neck_01, type=Null, loc=<4.5627 0 0>, rot=<0 0 0.0504 -0.9987>}
 clavicle_l {parent=spine_05, loc=<5.4831 -0.1053 -3.2946>, rot=<0.0465 -0.7289 -0.009 -0.6829>}
 upperarm_l {parent=clavicle_l, loc=<9.8658 -1.2136 -0.4105>, rot=<0.0164 -0.3245 -0.0269 -0.9454>}
 lowerarm_l {parent=upperarm_l, loc=<26.465 -0.8345 0.259>, rot=<0.0036 0.0161 -0.2298 -0.9731>}
 hand_l {parent=lowerarm_l, loc=<22.7355 -0.3203 -0.6105>, rot=<-0.6784 -0.0223 0.0735 -0.7307>}
 thumb_01_l {parent=hand_l, loc=<1.3118 -0.3191 -2.459>, rot=<0.6546 -0.4042 0.0664 -0.6354>}
 thumb_02_l {parent=thumb_01_l, loc=<5.9336 -0.0448 0.0401>, rot=<-0.0037 0.0213 0.298 -0.9543>}
 thumb_03_l {parent=thumb_02_l, loc=<2.2223 0.3682 -0.0907>, rot=<0.0118 -0.013 0.0066 -0.9998>}
 index_metacarpal_l {parent=hand_l, loc=<1.6535 0.0183 -2.96>, rot=<0.0158 -0.0124 -0.0173 -0.9996>}
 index_01_l {parent=index_metacarpal_l, loc=<8.3008 0.0357 -0.0524>, rot=<-0.0044 0.0485 0.0924 -0.9945>}
 index_02_l {parent=index_01_l, loc=<3.3587 -0.3004 0.0452>, rot=<-0.002 0.0027 0.3225 -0.9466>}
 index_03_l {parent=index_02_l, loc=<2.5578 -0.123 -0.0252>, rot=<-0.002 0.0067 0.2558 -0.9667>}
 middle_metacarpal_l {parent=hand_l, loc=<2.1856 -0.1725 -0.5984>, rot=<0.006 0.0168 -0.0329 -0.9993>}
 middle_01_l {parent=middle_metacarpal_l, loc=<7.3275 -0.2447 0.0129>, rot=<-0.0208 0.0042 0.0858 -0.9961>}
 middle_02_l {parent=middle_01_l, loc=<4.0123 0.1401 0.1175>, rot=<-0.0218 0.0258 0.3512 -0.9357>}
 middle_03_l {parent=middle_02_l, loc=<2.6077 0.0534 0.1373>, rot=<0.0082 0.0188 0.1708 -0.9851>}
 ring_metacarpal_l {parent=hand_l, loc=<2.2074 -0.3625 1.3638>, rot=<-0.0284 0.0722 -0.0458 -0.9959>}
 ring_01_l {parent=ring_metacarpal_l, loc=<7.0295 -0.6197 -0.1446>, rot=<-0.0489 -0.0252 0.1892 -0.9804>}
 ring_02_l {parent=ring_01_l, loc=<3.6174 0.4647 -0.1027>, rot=<0 0 0.2806 -0.9598>}
 ring_03_l {parent=ring_02_l, loc=<2.4152 0.114 0.0199>, rot=<0 0 0.1189 -0.9929>}
 pinky_metacarpal_l {parent=hand_l, loc=<1.7616 -0.9753 3.3814>, rot=<-0.0397 0.0814 -0.069 -0.9935>}
 pinky_01_l {parent=pinky_metacarpal_l, loc=<7.1709 -0.4831 -0.1094>, rot=<-0.0654 -0.0035 0.2242 -0.9723>}
 pinky_02_l {parent=pinky_01_l, loc=<2.8274 0.0893 -0.1043>, rot=<0.0037 0.0011 0.2854 -0.9584>}
 pinky_03_l {parent=pinky_02_l, loc=<2.2046 0.0307 -0.0078>, rot=<-0.0002 0.0098 0.1445 -0.9895>}
 deform_wrist_dn_base_l {parent=hand_l, loc=<0 -1.5326 0>, rot=<0 0 0 -1>}
 deform_wrist_dn_l {parent=deform_wrist_dn_base_l, loc=<0 0 0>, rot=<0 0 0 1>}
 deform_wrist_up_base_l {parent=hand_l, loc=<0 2.3009 0>, rot=<0 0 0 -1>}
 deform_wrist_up_l {parent=deform_wrist_up_base_l, loc=<0 0 0>, rot=<0 0 0 1>}
 lowerarm_twist_02_l {parent=lowerarm_l, loc=<10.6707 0 0>, rot=<0 0 0 1>}
 lowerarm_twist_01_l {parent=lowerarm_l, loc=<18.316 0 0>, rot=<0 0 0 1>}
 deform_elbow_in_base_l {parent=lowerarm_l, loc=<0.1834 3.6062 -0.2572>, rot=<0 0 0 -1>}
 deform_elbow_in_l {parent=deform_elbow_in_base_l, loc=<0 0 0>, rot=<0 0 0 1>}
 deform_elbow_out_base_l {parent=lowerarm_l, loc=<0 -4 0>, rot=<0 0 0 1>}
 deform_elbow_out_l {parent=deform_elbow_out_base_l, loc=<0 0 0>, rot=<0 0 0 1>}
 hand_l_LowerArmDeltaRef_null {parent=lowerarm_l, type=Null, loc=<22.7355 -0.3203 -0.6105>, rot=<-0.6784 -0.0223 0.0735 -0.7307>}
 hand_l_null {parent=lowerarm_l, type=Null, loc=<22.7355 -0.3203 -0.6105>, rot=<-0.6784 -0.0223 0.0735 -0.7307>}
 upperarm_twist_01_l {parent=upperarm_l, loc=<7.2135 -0.1134 0>, rot=<0 0 0 1>}
 deform_upperarm_bk_base_l {parent=upperarm_twist_01_l, loc=<-6.4053 -4.3746 0.314>, rot=<0 0 0 -1>}
 deform_upperarm_bk_l {parent=deform_upperarm_bk_base_l, loc=<0 0 0>, rot=<0 0 0 1>}
 deform_upperarm_fr_base_l {parent=upperarm_twist_01_l, loc=<-5.6312 5.092 0.3354>, rot=<0 0 0 -1>}
 deform_upperarm_fr_l {parent=deform_upperarm_fr_base_l, loc=<0 0 0>, rot=<0 0 0 1>}
 deform_upperarm_up_base_l {parent=upperarm_twist_01_l, loc=<-8.891 0.4825 3.9517>, rot=<0 0 0 -1>}
 deform_upperarm_up_l {parent=deform_upperarm_up_base_l, loc=<0 0 0>, rot=<0 0 0 1>}
 base_sleeves_l {parent=upperarm_twist_01_l, loc=<4.5513 0 0.2513>, rot=<0 0 0 1>}
 base_TopSleeve_l_01 {parent=base_sleeves_l, loc=<4.8735 -0.0755 4.4362>, rot=<-0.0138 0.2036 0.0032 -0.979>}
 dyn_TopSleeve_l {parent=base_TopSleeve_l_01, loc=<0 0 0>, rot=<0 0 0 1>}
 base_underSleeve_l {parent=base_sleeves_l, loc=<5.285 1.4767 -7.8187>, rot=<0.9738 -0.0434 -0.0749 -0.2104>}
 dyn_underSleeve_l {parent=base_underSleeve_l, loc=<0 0 0>, rot=<0 0 0 1>}
 base_frontSleeve_l {parent=base_sleeves_l, loc=<4.8105 5.0057 -2.1102>, rot=<0.7803 -0.0006 -0.0499 -0.6235>}
 dyn_frontSleeve_l {parent=base_frontSleeve_l, loc=<0 0 0>, rot=<0 0 0 1>}
 base_backSleeve_l {parent=base_sleeves_l, loc=<4.8989 -5.1724 -2.6456>, rot=<-0.8029 0.0225 0.0642 -0.5922>}
 dyn_backSleeve_l {parent=base_backSleeve_l, loc=<0 0 0>, rot=<0 0 0 1>}
 upperarm_twist_02_l {parent=upperarm_l, loc=<19.7998 -0.1134 0>, rot=<0 0 0 1>}
 lowerarm_l_null {parent=upperarm_l, type=Null, loc=<26.465 -0.8345 0.259>, rot=<0.0036 0.0161 -0.2298 -0.9731>}
 upperarm_l_null {parent=clavicle_l, type=Null, loc=<9.8658 -1.2136 -0.4105>, rot=<0.0164 -0.3245 -0.0269 -0.9454>}
 clavicle_r {parent=spine_05, loc=<5.4831 -0.1053 3.2946>, rot=<0.7289 0.0465 -0.6829 0.009>}
 upperarm_r {parent=clavicle_r, loc=<-9.8658 1.2136 0.4105>, rot=<0.0164 -0.3245 -0.0269 -0.9454>}
 lowerarm_r {parent=upperarm_r, loc=<-26.465 0.8345 -0.259>, rot=<0.0036 0.0161 -0.2298 -0.9731>}
 hand_r {parent=lowerarm_r, loc=<-22.7355 0.3203 0.6106>, rot=<-0.6784 -0.0223 0.0735 -0.7307>}
 thumb_01_r {parent=hand_r, loc=<-1.3118 0.3191 2.459>, rot=<0.6546 -0.4042 0.0664 -0.6354>}
 thumb_02_r {parent=thumb_01_r, loc=<-5.9336 0.0448 -0.0401>, rot=<-0.0037 0.0213 0.298 -0.9543>}
 thumb_03_r {parent=thumb_02_r, loc=<-2.2223 -0.3682 0.0907>, rot=<0.0118 -0.013 0.0066 -0.9998>}
 index_metacarpal_r {parent=hand_r, loc=<-1.6535 -0.0183 2.96>, rot=<0.0158 -0.0124 -0.0173 -0.9996>}
 index_01_r {parent=index_metacarpal_r, loc=<-8.3008 -0.0357 0.0524>, rot=<-0.0044 0.0485 0.0924 -0.9945>}
 index_02_r {parent=index_01_r, loc=<-3.3587 0.3004 -0.0452>, rot=<-0.002 0.0027 0.3225 -0.9466>}
 index_03_r {parent=index_02_r, loc=<-2.5578 0.123 0.0252>, rot=<-0.002 0.0067 0.2558 -0.9667>}
 middle_metacarpal_r {parent=hand_r, loc=<-2.1856 0.1725 0.5984>, rot=<0.006 0.0168 -0.0329 -0.9993>}
 middle_01_r {parent=middle_metacarpal_r, loc=<-7.3275 0.2447 -0.0129>, rot=<-0.0208 0.0042 0.0858 -0.9961>}
 middle_02_r {parent=middle_01_r, loc=<-4.0123 -0.1401 -0.1175>, rot=<-0.0218 0.0258 0.3512 -0.9357>}
 middle_03_r {parent=middle_02_r, loc=<-2.6077 -0.0534 -0.1373>, rot=<0.0082 0.0188 0.1708 -0.9851>}
 ring_metacarpal_r {parent=hand_r, loc=<-2.2074 0.3625 -1.3638>, rot=<-0.0284 0.0722 -0.0458 -0.9959>}
 ring_01_r {parent=ring_metacarpal_r, loc=<-7.0295 0.6197 0.1446>, rot=<-0.0489 -0.0252 0.1892 -0.9804>}
 ring_02_r {parent=ring_01_r, loc=<-3.6174 -0.4647 0.1027>, rot=<0 0 0.2806 -0.9598>}
 ring_03_r {parent=ring_02_r, loc=<-2.4152 -0.114 -0.0199>, rot=<0 0 0.1189 -0.9929>}
 pinky_metacarpal_r {parent=hand_r, loc=<-1.7616 0.9753 -3.3814>, rot=<-0.0397 0.0814 -0.069 -0.9935>}
 pinky_01_r {parent=pinky_metacarpal_r, loc=<-7.1709 0.4831 0.1094>, rot=<-0.0654 -0.0035 0.2242 -0.9723>}
 pinky_02_r {parent=pinky_01_r, loc=<-2.8274 -0.0893 0.1043>, rot=<0.0037 0.0011 0.2854 -0.9584>}
 pinky_03_r {parent=pinky_02_r, loc=<-2.2046 -0.0307 0.0078>, rot=<-0.0002 0.0098 0.1445 -0.9895>}
 deform_wrist_dn_base_r {parent=hand_r, loc=<-0.0001 1.5328 0>, rot=<0 0 0 -1>}
 deform_wrist_dn_r {parent=deform_wrist_dn_base_r, loc=<0 0 0>, rot=<0 0 0 1>}
 deform_wrist_up_base_r {parent=hand_r, loc=<-0.0002 -2.3007 -0.0001>, rot=<0 0 0 -1>}
 deform_wrist_up_r {parent=deform_wrist_up_base_r, loc=<0 0 0>, rot=<0 0 0 1>}
 base_wristHelper_02 {parent=hand_r, loc=<0.4597 1.4744 3.4307>, rot=<-0.6276 -0.1309 -0.4183 -0.6434>}
 dyn_wristHelper_02 {parent=base_wristHelper_02, loc=<0 0 0>, rot=<0 0 0 1>}
 base_wristHelper_01 {parent=hand_r, loc=<-0.0298 0.4435 -4.2813>, rot=<-0.6276 -0.1309 -0.4183 -0.6434>}
 dyn_wristHelper_01 {parent=base_wristHelper_01, loc=<0 0 0>, rot=<0 0 0 1>}
 lowerarm_twist_02_r {parent=lowerarm_r, loc=<-10.6707 0 0>, rot=<0 0 0 1>}
 lowerarm_twist_01_r {parent=lowerarm_r, loc=<-18.316 0 0>, rot=<0 0 0 1>}
 deform_elbow_in_base_r {parent=lowerarm_r, loc=<-0.1834 -3.6062 0.2572>, rot=<0 0 0 -1>}
 deform_elbow_in_r {parent=deform_elbow_in_base_r, loc=<0 0 0>, rot=<0 0 0 1>}
 deform_elbow_out_base_r {parent=lowerarm_r, loc=<0 4 -0.0001>, rot=<0 0 0 -1>}
 deform_elbow_out_r {parent=deform_elbow_out_base_r, loc=<0 0 0>, rot=<0 0 0 1>}
 hand_r_null {parent=lowerarm_r, type=Null, loc=<-22.7355 0.3203 0.6105>, rot=<-0.6784 -0.0223 0.0735 -0.7307>}
 upperarm_twist_01_r {parent=upperarm_r, loc=<-7.2135 0.1134 0>, rot=<0 0 0 1>}
 deform_upperarm_bk_base_r {parent=upperarm_twist_01_r, loc=<6.4053 4.3746 -0.314>, rot=<0 0 0 -1>}
 deform_upperarm_bk_r {parent=deform_upperarm_bk_base_r, loc=<0 0 0>, rot=<0 0 0 1>}
 deform_upperarm_fr_base_r {parent=upperarm_twist_01_r, loc=<5.6312 -5.092 -0.3354>, rot=<0 0 0 -1>}
 deform_upperarm_fr_r {parent=deform_upperarm_fr_base_r, loc=<0 0 0>, rot=<0 0 0 1>}
 deform_upperarm_up_base_r {parent=upperarm_twist_01_r, loc=<8.891 -0.4825 -3.9517>, rot=<0 0 0 -1>}
 deform_upperarm_up_r {parent=deform_upperarm_up_base_r, loc=<0 0 0>, rot=<0 0 0 1>}
 base_Sleeves_r {parent=upperarm_twist_01_r, loc=<-4.584 0 0.0722>, rot=<0 0 0 1>}
 base_TopSleeve_r {parent=base_Sleeves_r, loc=<-4.3532 0.0772 -4.1626>, rot=<0.0013 -0.9955 0.085 -0.0423>}
 dyn_TopSleeve_r {parent=base_TopSleeve_r, loc=<0 0 0>, rot=<0 0 0 1>}
 base_frontSleeve_r {parent=base_Sleeves_r, loc=<-4.8106 -5.0057 2.1104>, rot=<0.0546 -0.683 0.7284 -0.0023>}
 dyn_frontSleeve_r {parent=base_frontSleeve_r, loc=<0 0 0>, rot=<0 0 0 1>}
 base_underSleeve_r {parent=base_Sleeves_r, loc=<-5.5458 -1.365 7.9208>, rot=<0.0885 -0.1242 0.9876 0.0365>}
 dyn_underSleeve_r {parent=base_underSleeve_r, loc=<0 0 0>, rot=<0 0 0 1>}
 base_backSleeve_r {parent=base_Sleeves_r, loc=<-4.899 5.1724 2.6457>, rot=<-0.0493 -0.7175 -0.6927 -0.0537>}
 dyn_backSleeve_r {parent=base_backSleeve_r, loc=<0 0 0>, rot=<0 0 0 1>}
 upperarm_twist_01_ShirtOutPush_base {parent=upperarm_twist_01_r, loc=<1.0236 -4.9801 0.3891>, rot=<0 0 0 1>}
 upperarm_twist_01_ShirtOutPush {parent=upperarm_twist_01_ShirtOutPush_base, loc=<-1 -4 -2>, rot=<0 0 0 1>}
 upperarm_twist_02_r {parent=upperarm_r, loc=<-19.7998 0.1134 0>, rot=<0 0 0 1>}
 lowerarm_r_null {parent=upperarm_r, type=Null, loc=<-26.465 0.8345 -0.259>, rot=<0.0036 0.0161 -0.2298 -0.9731>}
 base_bagStrapHelper {parent=clavicle_r, loc=<-5.7638 -0.3038 -5.6785>, rot=<0.9991 0.0252 0.0336 0.0008>}
 dyn_bagStrapHelper {parent=base_bagStrapHelper, loc=<0 0 0>, rot=<0 0 0 1>}
 upperarm_r_null {parent=clavicle_r, type=Null, loc=<-9.8658 1.2136 0.4105>, rot=<0.0164 -0.3245 -0.0269 -0.9454>}
 attach_backpack {parent=spine_05, loc=<-0.1046 -13.5922 0.0232>, rot=<-0.0085 -0.7063 0.009 -0.7078>}
 Backpack_BR {parent=attach_backpack, type=Null, loc=<0 2.5 0>, rot=<-0.0436 0 0 0.999>}
 attach_cape {parent=spine_05, loc=<9.2184 -11.7742 0.0447>, rot=<0.0625 -0.7036 -0.0619 -0.7051>}
 cape {parent=attach_cape, type=Null, loc=<0 3 0>, rot=<0 0 0 1>}
 base_pin_02 {parent=spine_05, loc=<0.3359 9.5217 3.339>, rot=<0.0284 -0.7065 -0.0284 -0.7065>}
 dyn_pin_02 {parent=base_pin_02, loc=<0 0 0>, rot=<0 0 0 1>}
 base_pin_01 {parent=spine_05, loc=<3.1289 7.9297 2.5742>, rot=<0.0284 -0.7065 -0.0284 -0.7065>}
 dyn_pin_01 {parent=base_pin_01, loc=<0 0 0>, rot=<0 0 0 1>}
 Belt_Spine05_Fr_base {parent=spine_05, loc=<1.2866 8.9524 2.7306>, rot=<0.0678 -0.6717 -0.6516 -0.3458>}
 Belt_Spine05_Fr {parent=Belt_Spine05_Fr_base, loc=<0 0 0>, rot=<0 0 0 1>}
 Belt_Spine05_Bk_base {parent=spine_05, loc=<2.4399 -9.0372 3.912>, rot=<-0.0965 -0.7203 0.6314 -0.2705>}
 Belt_Spine05_Bk {parent=Belt_Spine05_Bk_base, loc=<0 0 0>, rot=<0 0 0 1>}
 base_sideBag_01 {parent=spine_04, loc=<0.5862 0.7313 -1.3013>, rot=<0.2676 -0.7631 -0.5658 -0.1612>}
 dyn_sideBag_01 {parent=base_sideBag_01, loc=<0 0 0>, rot=<0 0 0 1>}
 base_bandolierFront_01 {parent=dyn_sideBag_01, loc=<-2.4228 1.9553 11.2888>, rot=<0.0773 0.0864 0.0191 -0.9931>}
 dyn_bandolierFront_01 {parent=base_bandolierFront_01, loc=<0 0 0>, rot=<0 0 0 1>}
 base_bandolierFront_02 {parent=dyn_bandolierFront_01, loc=<11.4729 -0.0263 -1.1843>, rot=<-0.2054 -0.1175 0.0127 -0.9715>}
 dyn_bandolierFront_02 {parent=base_bandolierFront_02, loc=<0 0 0>, rot=<0 0 0 1>}
 base_bandolierBack_01 {parent=dyn_sideBag_01, loc=<2.1969 -2.0126 -10.7507>, rot=<0.2508 0.0458 -0.0839 -0.9633>}
 dyn_bandolierBack_01 {parent=base_bandolierBack_01, loc=<0 0 0>, rot=<0 0 0 1>}
 base_bandolierBack_02 {parent=dyn_bandolierBack_01, loc=<11.4232 -0.7486 0.3181>, rot=<0.2139 -0.0114 0.0495 -0.9755>}
 dyn_bandolierBack_02 {parent=base_bandolierBack_02, loc=<0 0 0>, rot=<0 0 0 1>}
 base_sideBag_02 {parent=dyn_sideBag_01, loc=<23.8309 -0.0863 0.1392>, rot=<0 0 -0.1853 -0.9827>}
 dyn_sideBag_02 {parent=base_sideBag_02, loc=<0 0 0>, rot=<0 0 0 1>}
 dyn_sideBag_03 {parent=dyn_sideBag_02, loc=<18.9409 0.0005 0.0008>, rot=<0 0 0 1>}
 base_bagStrap_02 {parent=dyn_sideBag_03, loc=<5.4584 -1.678 -7.1809>, rot=<-0.7494 -0.0592 -0.0192 -0.6592>}
 dyn_bagStrap_02 {parent=base_bagStrap_02, loc=<-0.2858 0.1614 -0.1278>, rot=<0 0 0 1>}
 base_bagStrap_01 {parent=dyn_sideBag_03, loc=<6.0658 -2.4776 2.6996>, rot=<-0.7086 -0.0527 -0.0282 -0.7031>}
 dyn_bagStrap_01 {parent=base_bagStrap_01, loc=<-0.4514 0.1447 -0.1432>, rot=<0 0 0 1>}
 dyn_bandolierSideFront_01 {parent=dyn_sideBag_02, loc=<0.9114 1.1822 8.7562>, rot=<0 0 0 1>}
 dyn_bandolierSideBack_01 {parent=dyn_sideBag_02, loc=<0.6592 1.5302 -8.3943>, rot=<0 0 0.1853 -0.9827>}
 deform_pec_aim_base_l {parent=spine_04, loc=<3.2456 9.3309 -1.2237>, rot=<0.0603 -0.4033 0.3318 -0.8507>}
 deform_pec_aim_l {parent=deform_pec_aim_base_l, loc=<0 0 0>, rot=<0 0 0 1>}
 deform_pec_base_l {parent=deform_pec_aim_l, loc=<9.3665 0 0>, rot=<0 0 0 1>}
 deform_pec_l {parent=deform_pec_base_l, loc=<0 0 0>, rot=<0 0 0 1>}
 deform_pec_aim_base_r {parent=spine_04, loc=<3.2452 9.3309 1.2237>, rot=<-0.4033 -0.0603 0.8507 0.3318>}
 deform_pec_aim_r {parent=deform_pec_aim_base_r, loc=<0 0 0>, rot=<0 0 0 1>}
 deform_pec_base_r {parent=deform_pec_aim_r, loc=<-9.3667 -0.0003 -0.0004>, rot=<0 0 0 1>}
 deform_pec_r {parent=deform_pec_base_r, loc=<0 0 0>, rot=<0 0 0 1>}
 spine_05_null {parent=spine_04, type=Null, loc=<6.9987 0 0>, rot=<0 0 -0.0015 -1>}
 spine_04_null {parent=spine_03, type=Null, loc=<6.9598 0.0794 0>, rot=<0 0 -0.0132 -0.9999>}
 spine_03_null {parent=spine_02, type=Null, loc=<6.4914 -0.1237 0>, rot=<0 0 0.0149 -0.9999>}
 spine_04_Spine1Delta_null {parent=spine_01, type=Null, loc=<20.023 -1.0996 0>, rot=<0 0 0.0255 -0.9997>}
 spine_05_Spine01Delta_null {parent=spine_01, type=Null, loc=<27.0126 -1.456 0>, rot=<0 0 -0.024 0.9997>}
 spine_02_null {parent=spine_01, type=Null, loc=<6.5997 -0.2095 0>, rot=<0 0 0.0238 -0.9997>}
 thigh_l {parent=pelvis, loc=<-1.0855 -0.2109 -8.7053>, rot=<0.0009 0.0355 0.0155 -0.9993>}
 calf_l {parent=thigh_l, loc=<-42.5284 -0.0019 -0.7297>, rot=<0.0006 -0.0161 -0.0391 -0.9991>}
 foot_l {parent=calf_l, loc=<-46.7401 -2.5014 -0.592>, rot=<0.0427 -0.0238 0.0353 -0.9982>}
 ball_l {parent=foot_l, loc=<-3.5658 10.8066 -0.1544>, rot=<0.0001 -0.0001 -0.7063 -0.7079>}
 base_shoelace_l_03 {parent=foot_l, loc=<2.4745 7.9055 0.9154>, rot=<0.3763 -0.5165 0.7541 0.1512>}
 dyn_shoelace_l_03 {parent=base_shoelace_l_03, loc=<0.0327 -0.0051 0.0022>, rot=<0.0142 -0.0007 0.1105 -0.9938>}
 dyn_shoelace_l_04 {parent=dyn_shoelace_l_03, loc=<4.9295 0.3052 -0.5863>, rot=<0.012 0.0056 -0.0013 -0.9999>}
 base_shoelace_l_01 {parent=foot_l, loc=<2.4628 7.8397 -0.5986>, rot=<-0.3916 0.4171 0.7944 0.2039>}
 dyn_shoelace_l_01 {parent=base_shoelace_l_01, loc=<0.0094 0.0016 0.001>, rot=<-0.0292 0.0005 -0.001 -0.9996>}
 dyn_shoelace_l_02 {parent=dyn_shoelace_l_01, loc=<4.9564 0.9498 0.0018>, rot=<-0.0056 -0.0007 -0.158 -0.9874>}
 base_shoetongue_l_01 {parent=foot_l, loc=<3.2833 5.9416 0.2202>, rot=<0 0 0 1>}
 dyn_shoetongue_l_01 {parent=base_shoetongue_l_01, loc=<0 0 0>, rot=<0 0 0 1>}
 calf_twist_02_l {parent=calf_l, loc=<-19.1546 -1.7943 -0.2638>, rot=<0.0237 0.0095 -0.0035 -0.9997>}
 calf_twist_01_l {parent=calf_l, loc=<-35.8483 -1.9197 -0.5794>, rot=<0.0237 0.0095 -0.0035 -0.9997>}
 deform_knee_in_base_l {parent=calf_l, loc=<0.8031 -5.0231 0.2456>, rot=<0 0 0 -1>}
 deform_knee_in_l {parent=deform_knee_in_base_l, loc=<0 0 0>, rot=<0 0 0 1>}
 deform_knee_out_base_l {parent=calf_l, loc=<0 4 0>, rot=<0 0 0 -1>}
 deform_knee_out_l {parent=deform_knee_out_base_l, loc=<0 0 0>, rot=<0 0 0 1>}
 thigh_twist_01_l {parent=thigh_l, loc=<-27.952 0 0>, rot=<0.0013 0 0 -1>}
 deform_groin_base_l {parent=thigh_l, loc=<0.4801 5.6776 -0.2486>, rot=<0 0 0 -1>}
 deform_groin_l {parent=deform_groin_base_l, loc=<0 0 0>, rot=<0 0 0 1>}
 thigh_r {parent=pelvis, loc=<-1.0855 -0.2109 8.7053>, rot=<0.0355 -0.0009 0.9993 0.0155>}
 calf_r {parent=thigh_r, loc=<42.5284 0.0019 0.7297>, rot=<0.0006 -0.0161 -0.0391 -0.9991>}
 foot_r {parent=calf_r, loc=<46.7401 2.5014 0.592>, rot=<0.0427 -0.0238 0.0353 -0.9982>}
 ball_r {parent=foot_r, loc=<3.5658 -10.8066 0.1544>, rot=<0.0001 -0.0001 -0.7063 -0.7079>}
 base_shoelace_r_03 {parent=foot_r, loc=<-2.4644 -7.8824 -0.9214>, rot=<-0.5165 -0.3763 0.1512 -0.7541>}
 dyn_shoelace_r_03 {parent=base_shoelace_r_03, loc=<0.0088 0.0081 0.0019>, rot=<0.0847 -0.001 0.0317 -0.9959>}
 dyn_shoelace_r_04 {parent=dyn_shoelace_r_03, loc=<4.9135 0.304 0.584>, rot=<-0.1273 0.0569 -0.2111 -0.9675>}
 base_shoelace_r_01 {parent=foot_r, loc=<-2.4548 -7.8148 0.5921>, rot=<0.4171 0.3916 0.2039 -0.7944>}
 dyn_shoelace_r_01 {parent=base_shoelace_r_01, loc=<0.0301 0.0047 -0.0045>, rot=<-0.0564 -0.0003 0.0909 -0.9943>}
 dyn_shoelace_r_02 {parent=dyn_shoelace_r_01, loc=<4.9669 0.9523 0.0019>, rot=<-0.0066 0.038 -0.1391 -0.9895>}
 base_shoetongue_r_01 {parent=foot_r, loc=<-3.2807 -6.1657 -0.2655>, rot=<-0.0052 -0.999 -0.0453 -0.0026>}
 dyn_shoetongue_r_01 {parent=base_shoetongue_r_01, loc=<0 0 0>, rot=<0 0 0 1>}
 calf_twist_02_r {parent=calf_r, loc=<19.1546 1.7943 0.2638>, rot=<0.0237 0.0095 -0.0035 -0.9997>}
 calf_twist_01_r {parent=calf_r, loc=<35.8483 1.9197 0.5794>, rot=<0.0237 0.0095 -0.0035 -0.9997>}
 deform_knee_in_base_r {parent=calf_r, loc=<-0.8031 5.0231 -0.2457>, rot=<0 0 0 -1>}
 deform_knee_in_r {parent=deform_knee_in_base_r, loc=<0 0 0>, rot=<0 0 0 1>}
 deform_knee_out_base_r {parent=calf_r, loc=<0 -4 0>, rot=<0 0 0 -1>}
 deform_knee_out_r {parent=deform_knee_out_base_r, loc=<0 0 0>, rot=<0 0 0 1>}
 Base_Kneepad_Helper {parent=calf_r, loc=<1.411 -6.7544 -0.6851>, rot=<-0.6927 0.0276 0.7201 0.0309>}
 Kneepad_Helper {parent=Base_Kneepad_Helper, loc=<0 0 0>, rot=<0 0 0 1>}
 thigh_twist_01_r {parent=thigh_r, loc=<27.952 0 0>, rot=<0.0013 0 0 -1>}
 deform_groin_base_r {parent=thigh_r, loc=<-0.4801 -5.6776 0.2486>, rot=<0 0 0 -1>}
 deform_groin_r {parent=deform_groin_base_r, loc=<0 0 0>, rot=<0 0 0 1>}
 calf_r_null {parent=thigh_r, type=Null, loc=<42.5284 0.0019 0.7297>, rot=<0.0006 -0.0161 -0.0391 -0.9991>}
 deform_glute_base_l {parent=pelvis, loc=<-7.3466 -9.0213 -9.1245>, rot=<0.0172 0.031 -0.4477 -0.8935>}
 deform_glute_l {parent=deform_glute_base_l, loc=<0 0 0>, rot=<0 0 0 1>}
 deform_glute_base_r {parent=pelvis, loc=<-7.3464 -9.0213 9.1245>, rot=<0.031 -0.0172 0.8935 -0.4477>}
 deform_glute_r {parent=deform_glute_base_r, loc=<0 0 0>, rot=<0 0 0 1>}
 base_tshirt_front_r_01 {parent=pelvis, loc=<9.2116 8.0594 9.7377>, rot=<0.0229 -0.1953 0.9789 0.0565>}
 dyn_tshirt_front_r_01 {parent=base_tshirt_front_r_01, loc=<0 0.4611 0.4349>, rot=<0 0 0 1>}
 dyn_tshirt_front_r_02 {parent=dyn_tshirt_front_r_01, loc=<11.629 0 0>, rot=<0 0 0 1>}
 dyn_tshirt_front_r_03 {parent=dyn_tshirt_front_r_02, loc=<9.7785 0.001 0.0014>, rot=<0 0 0 -1>}
 base_tshirt_side_r_01 {parent=pelvis, loc=<8.105 -0.5811 15.0485>, rot=<0.0217 -0.7132 0.6984 0.0563>}
 dyn_tshirt_side_r_01 {parent=base_tshirt_side_r_01, loc=<0 1.2746 0>, rot=<0 0 0 1>}
 dyn_tshirt_side_r_02 {parent=dyn_tshirt_side_r_01, loc=<10.7475 0 0>, rot=<0 0 0 1>}
 dyn_tshirt_side_r_03 {parent=dyn_tshirt_side_r_02, loc=<9.7806 -0.001 0.0006>, rot=<0 0 0 -1>}
 base_tshirt_rear_r_01 {parent=pelvis, loc=<8.6492 -9.0943 11.308>, rot=<0 -0.9296 0.3686 0.0049>}
 dyn_tshirt_rear_r_01 {parent=base_tshirt_rear_r_01, loc=<0 -0.4845 0.2712>, rot=<0 0 0 1>}
 dyn_tshirt_rear_r_02 {parent=dyn_tshirt_rear_r_01, loc=<11.6182 0 0>, rot=<0 0 0 -1>}
 dyn_tshirt_rear_r_03 {parent=dyn_tshirt_rear_r_02, loc=<9.7773 -0.0003 -0.0009>, rot=<0 0 0 -1>}
 base_tshirt_rear_01 {parent=pelvis, loc=<8.298 -12.6899 0.013>, rot=<0.0008 -1 -0.0067 0>}
 dyn_tshirt_rear_01 {parent=base_tshirt_rear_01, loc=<0 1.0528 0>, rot=<0 0 0 1>}
 dyn_tshirt_rear_02 {parent=dyn_tshirt_rear_01, loc=<11.4465 0 0>, rot=<0 0 0 1>}
 dyn_tshirt_rear_03 {parent=dyn_tshirt_rear_02, loc=<9.7768 -0.0005 -0.0006>, rot=<0 0 0 -1>}
 base_tshirt_front_l_01 {parent=pelvis, loc=<9.2117 8.0594 -9.7377>, rot=<-0.0229 0.1953 0.9789 0.0565>}
 dyn_tshirt_front_l_01 {parent=base_tshirt_front_l_01, loc=<0 0.3939 -0.3695>, rot=<0 0 0 1>}
 dyn_tshirt_front_l_02 {parent=dyn_tshirt_front_l_01, loc=<11.7017 0 0>, rot=<0 0 0 1>}
 dyn_tshirt_front_l_03 {parent=dyn_tshirt_front_l_02, loc=<9.772 -0.0001 -0.0014>, rot=<0 0 0 -1>}
 base_tshirt_side_l_01 {parent=pelvis, loc=<8.1054 -0.5811 -15.0484>, rot=<0.0217 -0.7132 -0.6984 -0.0563>}
 dyn_tshirt_side_l_01 {parent=base_tshirt_side_l_01, loc=<0 0.9294 0>, rot=<0 0 0 1>}
 dyn_tshirt_side_l_02 {parent=dyn_tshirt_side_l_01, loc=<10.8211 0 0>, rot=<0 0 0 1>}
 dyn_tshirt_side_l_03 {parent=dyn_tshirt_side_l_02, loc=<9.7782 -0.0016 -0.0003>, rot=<0 0 0 -1>}
 base_tshirt_rear_l_01 {parent=pelvis, loc=<8.6186 -8.783 -11.0153>, rot=<0.0517 -0.9281 -0.3678 -0.0254>}
 dyn_tshirt_rear_l_01 {parent=base_tshirt_rear_l_01, loc=<0 0.0966 -0.6121>, rot=<0 0 0 1>}
 dyn_tshirt_rear_l_02 {parent=dyn_tshirt_rear_l_01, loc=<11.7139 0 0>, rot=<0 0 0 1>}
 dyn_tshirt_rear_l_03 {parent=dyn_tshirt_rear_l_02, loc=<9.774 0.0001 0.0003>, rot=<0 0 0 -1>}
 base_tshirt_front_01 {parent=pelvis, loc=<8.9541 9.5088 0.0006>, rot=<0 0.0004 0.9975 0.0709>}
 dyn_tshirt_front_01 {parent=base_tshirt_front_01, loc=<0 0.8122 0>, rot=<0 0 0 1>}
 dyn_tshirt_front_02 {parent=dyn_tshirt_front_01, loc=<11.3135 0 0>, rot=<0 0 0 1>}
 dyn_tshirt_front_03 {parent=dyn_tshirt_front_02, loc=<9.7762 0.001 -0.0001>, rot=<0 0 0 -1>}
 thigh_l_null {parent=pelvis, type=Null, loc=<-1.0855 -0.2109 -8.7053>, rot=<0.0009 0.0355 0.0155 -0.9993>}
 thigh_r_null {parent=pelvis, type=Null, loc=<-1.0855 -0.2109 8.7053>, rot=<0.0355 -0.0009 0.9993 0.0155>}
 noDynamics_SideBag_base {parent=pelvis, loc=<7.238 -4.8411 -11.8093>, rot=<0.0738 -0.8053 -0.5817 -0.0883>}
 noDynamics_SideBag {parent=noDynamics_SideBag_base, loc=<0 0 0>, rot=<0 0 0 1>}
 noDynamics_SideBagScale_base {parent=noDynamics_SideBag, loc=<-2.7601 0.6446 0>, rot=<0 0 0 1>}
 noDynamics_SideBagScale {parent=noDynamics_SideBagScale_base, loc=<0 0 0>, rot=<0 0 0 1>}
]
variables [
 ASWEnabled {type=bool, default=False}
 BaseballSliding {type=bool, default=False}
 Crouching {type=bool, default=False}
 CrouchingFloat {type=double, default=0}
 CrounchingWalk2D {type=double, default=0}
 DoubleHandedWeapon {type=bool, default=False}
 KneeSliding {type=bool, default=False}
 L lowerarm bend {type=double, default=0}
 PickaxeSwing {type=bool, default=False}
 PlayingEmote {type=bool, default=False}
 R lowerarm bend {type=double, default=0}
 bigGunAiming {type=bool, default=False}
 isDBNO {type=bool, default=False}
]
