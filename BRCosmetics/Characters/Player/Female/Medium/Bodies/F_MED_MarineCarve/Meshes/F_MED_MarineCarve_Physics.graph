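# Physics asset for the MarineCarve medium female player body, set up for secondary motion of the T-shirt around the pelvis.
# Eight front, side and rear chains each run from a small base capsule, free in rotation on the pelvis, through two dynamic capsules with tight swing limits.
# Rotation-free constraints cross-link neighbouring front and side tips in both directions, leaving the rear unlinked.
# Pelvis, spine and thigh capsules represent the body.
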PhysicsAsset /BRCosmetics/Characters/Player/Female/Medium/Bodies/F_MED_MarineCarve/Meshes/F_MED_MarineCarve_Physics
bodies {
 base_tshirt_front_01 [{shapeType=Capsule, length=0.505, radius=0.505, shapeName=base_tshirt_front_01_capsule, rotation=<rot 0 -90 0>}]
 base_tshirt_front_l_01 [{shapeType=Capsule, length=0.505, radius=0.505, shapeName=base_tshirt_front_l_01_capsule, rotation=<rot 0 -90 0>}]
 base_tshirt_front_r_01 [{shapeType=Capsule, length=0.505, radius=0.505, shapeName=base_tshirt_front_r_01_capsule, rotation=<rot 0 -90 0>}]
 base_tshirt_rear_01 [{shapeType=Capsule, length=0.505, radius=0.505, shapeName=base_tshirt_rear_01_capsule, rotation=<rot 0 -90 0>}]
 base_tshirt_rear_l_01 [{shapeType=Capsule, length=0.505, radius=0.505, shapeName=base_tshirt_rear_l_01_capsule, rotation=<rot 0 -90 0>}]
 base_tshirt_rear_r_01 [{shapeType=Capsule, length=0.505, radius=0.505, shapeName=base_tshirt_rear_r_01_capsule, rotation=<rot 0 -90 0>}]
 base_tshirt_side_l_01 [{shapeType=Capsule, length=0.505, radius=0.505, shapeName=base_tshirt_side_l_01_capsule, rotation=<rot 0 -90 0>}]
 base_tshirt_side_r_01 [{shapeType=Capsule, length=0.505, radius=0.505, shapeName=base_tshirt_side_r_01_capsule, rotation=<rot 0 -90 0>}]
 dyn_tshirt_front_01 [{shapeType=Capsule, length=15, radius=5, shapeName=dyn_tshirt_front_01_capsule, center=<9 -2 0>}]
 dyn_tshirt_front_02 [{shapeType=Capsule, length=5, radius=8, shapeName=dyn_tshirt_front_02_capsule, center=<9 -2 0>, rotation=<rot 90 0 0>}]
 dyn_tshirt_front_l_01 [{shapeType=Capsule, length=15, radius=5, shapeName=dyn_tshirt_front_l_01_capsule, center=<9 -2.5 0>}]
 dyn_tshirt_front_l_02 [{shapeType=Capsule, length=5, radius=8, shapeName=dyn_tshirt_front_l_02_capsule, center=<9 -2 0>, rotation=<rot 90 0 0>}]
 dyn_tshirt_front_r_01 [{shapeType=Capsule, length=15, radius=5, shapeName=dyn_tshirt_front_r_01_capsule, center=<9 -2 0>}]
 dyn_tshirt_front_r_02 [{shapeType=Capsule, length=5, radius=8, shapeName=dyn_tshirt_front_r_02_capsule, center=<9 -2 0>, rotation=<rot 90 0 0>}]
 dyn_tshirt_rear_01 [{shapeType=Capsule, length=15, radius=5, shapeName=dyn_tshirt_rear_01_capsule, center=<9 -2 0>}]
 dyn_tshirt_rear_02 [{shapeType=Capsule, length=5, radius=8, shapeName=dyn_tshirt_rear_02_capsule, center=<9 -2 0>, rotation=<rot 90 0 0>}]
 dyn_tshirt_rear_l_01 [{shapeType=Capsule, length=15, radius=5, shapeName=dyn_tshirt_rear_l_01_capsule, center=<9 -2 0>}]
 dyn_tshirt_rear_l_02 [{shapeType=Capsule, length=5, radius=8, shapeName=dyn_tshirt_rear_l_02_capsule, center=<9 -7 0>, rotation=<rot 90 0 0>}]
 dyn_tshirt_rear_r_01 [{shapeType=Capsule, length=15, radius=5, shapeName=dyn_tshirt_rear_r_01_capsule, center=<9 -2 0>}]
 dyn_tshirt_rear_r_02 [{shapeType=Capsule, length=5, radius=8, shapeName=dyn_tshirt_rear_r_02_capsule, center=<9 -7 0>, rotation=<rot 90 0 0>}]
 dyn_tshirt_side_l_01 [{shapeType=Capsule, length=15, radius=5, shapeName=dyn_tshirt_side_l_01_capsule, center=<9 -4 0>}]
 dyn_tshirt_side_l_02 [{shapeType=Capsule, length=5, radius=8, shapeName=dyn_tshirt_side_l_02_capsule, center=<9 -8 0>, rotation=<rot 90 0 0>}]
 dyn_tshirt_side_r_01 [{shapeType=Capsule, length=15, radius=5, shapeName=dyn_tshirt_side_r_01_capsule, center=<9 -2 0>}]
 dyn_tshirt_side_r_02 [{shapeType=Capsule, length=5, radius=8, shapeName=dyn_tshirt_side_r_02_capsule, center=<9 -8 0>, rotation=<rot 90 0 0>}]
 pelvis [{shapeType=Capsule, length=10.5716, radius=11.7888, shapeName=pelvis_capsule, center=<-6.807 0 0>}]
 spine_01 [{shapeType=Capsule, length=10.3665, radius=10.8641, shapeName=spine_01_capsule, center=<0 -0.591 0>, rotation=<rot 0 -90 0>}]
 spine_03 [{shapeType=Capsule, length=0.505, radius=0.505, shapeName=spine_03_capsule, rotation=<rot 0 -90 0>}]
 thigh_l [{shapeType=Capsule, length=27, radius=9, shapeName=thigh_l_capsule, center=<-19.9865 0 0>, rotation=<rot 90 0 0>}]
 thigh_r [{shapeType=Capsule, length=27, radius=9, shapeName=thigh_r_capsule, center=<19.9865 0 0>, rotation=<rot 90 180 0>}]
}
constraints [
 {
  bone1Name=spine_01
  bone2Name=pelvis
  swing1LimitDegrees=45
  swing1Motion=Free
  swing2LimitDegrees=45
  swing2Motion=Free
  twistLimitDegrees=45
  twistMotion=Free
 }
 {
  bone1Name=base_tshirt_front_01
  bone2Name=pelvis
  swing1LimitDegrees=45
  swing1Motion=Free
  swing2LimitDegrees=45
  swing2Motion=Free
  twistLimitDegrees=45
  twistMotion=Free
 }
 {bone1Name=dyn_tshirt_front_01, bone2Name=base_tshirt_front_01, swing1LimitDegrees=15, swing2LimitDegrees=5, twistLimitDegrees=0}
 {bone1Name=dyn_tshirt_front_02, bone2Name=dyn_tshirt_front_01, swing1LimitDegrees=15, swing2LimitDegrees=5, twistLimitDegrees=0}
 {
  bone1Name=base_tshirt_front_l_01
  bone2Name=pelvis
  swing1LimitDegrees=45
  swing1Motion=Free
  swing2LimitDegrees=45
  swing2Motion=Free
  twistLimitDegrees=45
  twistMotion=Free
 }
 {bone1Name=dyn_tshirt_front_l_01, bone2Name=base_tshirt_front_l_01, swing1LimitDegrees=15, swing2LimitDegrees=5, twistLimitDegrees=0}
 {bone1Name=dyn_tshirt_front_l_02, bone2Name=dyn_tshirt_front_l_01, swing1LimitDegrees=15, swing2LimitDegrees=5, twistLimitDegrees=0}
 {
  bone1Name=base_tshirt_front_r_01
  bone2Name=pelvis
  swing1LimitDegrees=45
  swing1Motion=Free
  swing2LimitDegrees=45
  swing2Motion=Free
  twistLimitDegrees=45
  twistMotion=Free
 }
 {bone1Name=dyn_tshirt_front_r_01, bone2Name=base_tshirt_front_r_01, swing1LimitDegrees=15, swing2LimitDegrees=5, twistLimitDegrees=0}
 {bone1Name=dyn_tshirt_front_r_02, bone2Name=dyn_tshirt_front_r_01, swing1LimitDegrees=15, swing2LimitDegrees=5, twistLimitDegrees=0}
 {
  bone1Name=base_tshirt_rear_01
  bone2Name=pelvis
  swing1LimitDegrees=45
  swing1Motion=Free
  swing2LimitDegrees=45
  swing2Motion=Free
  twistLimitDegrees=45
  twistMotion=Free
 }
 {bone1Name=dyn_tshirt_rear_01, bone2Name=base_tshirt_rear_01, swing1LimitDegrees=15, swing2LimitDegrees=5, twistLimitDegrees=0}
 {bone1Name=dyn_tshirt_rear_02, bone2Name=dyn_tshirt_rear_01, swing1LimitDegrees=15, swing2LimitDegrees=5, twistLimitDegrees=0}
 {bone1Name=dyn_tshirt_rear_l_01, bone2Name=base_tshirt_rear_l_01, swing1LimitDegrees=15, swing2LimitDegrees=5, twistLimitDegrees=0}
 {bone1Name=dyn_tshirt_rear_l_02, bone2Name=dyn_tshirt_rear_l_01, swing1LimitDegrees=15, swing2LimitDegrees=5, twistLimitDegrees=0}
 {
  bone1Name=base_tshirt_rear_r_01
  bone2Name=pelvis
  swing1LimitDegrees=45
  swing1Motion=Free
  swing2LimitDegrees=45
  swing2Motion=Free
  twistLimitDegrees=45
  twistMotion=Free
 }
 {bone1Name=dyn_tshirt_rear_r_01, bone2Name=base_tshirt_rear_r_01, swing1LimitDegrees=15, swing2LimitDegrees=5, twistLimitDegrees=0}
 {
  bone1Name=base_tshirt_rear_l_01
  bone2Name=pelvis
  swing1LimitDegrees=45
  swing1Motion=Free
  swing2LimitDegrees=45
  swing2Motion=Free
  twistLimitDegrees=45
  twistMotion=Free
 }
 {bone1Name=dyn_tshirt_rear_r_02, bone2Name=dyn_tshirt_rear_r_01, swing1LimitDegrees=15, swing2LimitDegrees=5, twistLimitDegrees=0}
 {
  bone1Name=base_tshirt_side_l_01
  bone2Name=pelvis
  swing1LimitDegrees=45
  swing1Motion=Free
  swing2LimitDegrees=45
  swing2Motion=Free
  twistLimitDegrees=45
  twistMotion=Free
 }
 {bone1Name=dyn_tshirt_side_l_01, bone2Name=base_tshirt_side_l_01, swing1LimitDegrees=15, swing2LimitDegrees=5, twistLimitDegrees=0}
 {bone1Name=dyn_tshirt_side_l_02, bone2Name=dyn_tshirt_side_l_01, swing1LimitDegrees=15, swing2LimitDegrees=5, twistLimitDegrees=0}
 {
  bone1Name=base_tshirt_side_r_01
  bone2Name=pelvis
  swing1LimitDegrees=45
  swing1Motion=Free
  swing2LimitDegrees=45
  swing2Motion=Free
  twistLimitDegrees=45
  twistMotion=Free
 }
 {bone1Name=dyn_tshirt_side_r_01, bone2Name=base_tshirt_side_r_01, swing1LimitDegrees=15, swing2LimitDegrees=5, twistLimitDegrees=0}
 {bone1Name=dyn_tshirt_side_r_02, bone2Name=dyn_tshirt_side_r_01, swing1LimitDegrees=15, swing2LimitDegrees=5, twistLimitDegrees=0}
 {
  bone1Name=thigh_r
  bone2Name=pelvis
  swing1LimitDegrees=45
  swing1Motion=Free
  swing2LimitDegrees=45
  swing2Motion=Free
  twistLimitDegrees=45
  twistMotion=Free
 }
 {
  bone1Name=thigh_l
  bone2Name=pelvis
  swing1LimitDegrees=45
  swing1Motion=Free
  swing2LimitDegrees=45
  swing2Motion=Free
  twistLimitDegrees=45
  twistMotion=Free
 }
 {
  bone1Name=dyn_tshirt_front_l_02
  bone2Name=dyn_tshirt_side_l_02
  swing1LimitDegrees=45
  swing1Motion=Free
  swing2LimitDegrees=45
  swing2Motion=Free
  twistLimitDegrees=45
  twistMotion=Free
 }
 {
  bone1Name=dyn_tshirt_front_r_02
  bone2Name=dyn_tshirt_side_r_02
  swing1LimitDegrees=45
  swing1Motion=Free
  swing2LimitDegrees=45
  swing2Motion=Free
  twistLimitDegrees=45
  twistMotion=Free
 }
 {
  bone1Name=dyn_tshirt_front_02
  bone2Name=dyn_tshirt_front_r_02
  swing1LimitDegrees=45
  swing1Motion=Free
  swing2LimitDegrees=45
  swing2Motion=Free
  twistLimitDegrees=45
  twistMotion=Free
 }
 {
  bone1Name=dyn_tshirt_front_02
  bone2Name=dyn_tshirt_front_l_02
  swing1LimitDegrees=45
  swing1Motion=Free
  swing2LimitDegrees=45
  swing2Motion=Free
  twistLimitDegrees=45
  twistMotion=Free
 }
 {
  bone1Name=dyn_tshirt_front_l_02
  bone2Name=dyn_tshirt_front_02
  swing1LimitDegrees=45
  swing1Motion=Free
  swing2LimitDegrees=45
  swing2Motion=Free
  twistLimitDegrees=45
  twistMotion=Free
 }
 {
  bone1Name=dyn_tshirt_front_r_02
  bone2Name=dyn_tshirt_front_02
  swing1LimitDegrees=45
  swing1Motion=Free
  swing2LimitDegrees=45
  swing2Motion=Free
  twistLimitDegrees=45
  twistMotion=Free
 }
 {
  bone1Name=dyn_tshirt_side_l_02
  bone2Name=dyn_tshirt_front_l_02
  swing1LimitDegrees=45
  swing1Motion=Free
  swing2LimitDegrees=45
  swing2Motion=Free
  twistLimitDegrees=45
  twistMotion=Free
 }
 {
  bone1Name=dyn_tshirt_side_r_02
  bone2Name=dyn_tshirt_front_r_02
  swing1LimitDegrees=45
  swing1Motion=Free
  swing2LimitDegrees=45
  swing2Motion=Free
  twistLimitDegrees=45
  twistMotion=Free
 }
 {
  bone1Name=spine_03
  bone2Name=spine_01
  swing1LimitDegrees=45
  swing1Motion=Free
  swing2LimitDegrees=45
  swing2Motion=Free
  twistLimitDegrees=45
  twistMotion=Free
 }
]
