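# Physics asset for the side bag and bandolier on the female medium MarineCarve body, giving their dyn bones jointed collision capsules.
# Bandolier front, back and side bag chains hang from spine_04 through tiny base_ bodies, with bandolier side pieces branching off the lower bag.
# Dyn joints mostly have tight swing and twist limits; anchor joints are mostly free, except the locked upper front bandolier.
# Pelvis, spine, left thigh and left T-shirt capsules supply body shapes.
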PhysicsAsset /BRCosmetics/Characters/Player/Female/Medium/Bodies/F_MED_MarineCarve/Meshes/F_MED_MarineCarve_SideBag_Physics
bodies {
 base_bandolierBack_01 [{shapeType=Capsule, length=0.505, radius=0.505, shapeName=base_bandolierBack_01_capsule, rotation=<rot 0 -90 0>}]
 base_bandolierBack_02 [{shapeType=Capsule, length=0.505, radius=0.505, shapeName=base_bandolierBack_02_capsule, rotation=<rot 0 -90 0>}]
 base_bandolierFront_01 [{shapeType=Capsule, length=0.505, radius=0.505, shapeName=base_bandolierFront_01_capsule, rotation=<rot 0 -90 0>}]
 base_bandolierFront_02 [{shapeType=Capsule, length=0.505, radius=0.505, shapeName=base_bandolierFront_02_capsule, rotation=<rot 0 -90 0>}]
 base_sideBag_01 [{shapeType=Capsule, length=0.505, radius=0.505, shapeName=base_sideBag_01_capsule, rotation=<rot 0 -90 0>}]
 base_sideBag_02 [{shapeType=Capsule, length=0.505, radius=0.505, shapeName=base_sideBag_02_capsule, rotation=<rot 0 -90 0>}]
 dyn_bandolierBack_01 [{shapeType=Capsule, length=10, radius=3, shapeName=dyn_bandolierBack_01_capsule, center=<0 0 -3.3>, rotation=<rot 90 0 0>}]
 dyn_bandolierBack_02 [{
   shapeType=Capsule
   length=10
   radius=3
   shapeName=dyn_bandolierBack_02_capsule
   center=<2.3416 0 -2.4178>
   rotation=<rot 81.419 -179.9994 -179.9993>
  }]
 dyn_bandolierFront_01 [{shapeType=Capsule, length=10, radius=3, shapeName=dyn_bandolierFront_01_capsule, center=<0 0 3>, rotation=<rot 79.88 -179.9999 -179.9999>}]
 dyn_bandolierFront_02 [{shapeType=Capsule, length=15, radius=3, shapeName=dyn_bandolierFront_02_capsule, center=<1.9677 0 2.2>, rotation=<rot 90 -0.0014 0>}]
 dyn_bandolierSideBack_01 [{
   shapeType=Capsule
   length=11.5821
   radius=3.2
   shapeName=dyn_bandolierSideBack_01_capsule
   center=<0.4352 -1.3945 -1.283>
   rotation=<rot 79.8475 -179.9998 171.7491>
  }]
 dyn_bandolierSideFront_01 [{
   shapeType=Capsule
   length=8
   radius=3
   shapeName=dyn_bandolierSideFront_01_capsule
   center=<-1.2055 -0.2477 1.3361>
   rotation=<rot 80.6748 -11.8891 -0.1903>
  }]
 dyn_sideBag_01 [{shapeType=Capsule, length=23, radius=5, shapeName=dyn_sideBag_01_capsule, center=<14 0 0>, rotation=<rot 0 -90 90>}]
 dyn_sideBag_02 [
  {shapeType=Capsule, length=16, radius=4, shapeName=dyn_sideBag_02_capsule, center=<10 -3.0121 -0.0151>, rotation=<rot 0.0002 0.0001 -0.8533>}
  {
   shapeType=Capsule
   length=16
   radius=4
   shapeName=dyn_sideBag_02_capsule
   center=<21.5474 -2.3849 -0.0057>
   rotation=<rot 0.0002 0.0001 -0.8533>
  }
  {
   shapeType=Capsule
   length=16.7763
   radius=3.2237
   shapeName=dyn_sideBag_02_capsule
   center=<14.891 0.9621 10.7574>
   rotation=<rot 85 -174.9999 -179.9999>
  }
  {
   shapeType=Capsule
   length=17.052
   radius=3.6057
   shapeName=dyn_sideBag_02_capsule
   center=<17 -2.1935 -0.0057>
   rotation=<rot 0.0002 0.0001 -0.8533>
  }
 ]
 dyn_tshirt_rear_l_02 [{
   shapeType=Capsule
   length=10.0172
   radius=0.9874
   shapeName=dyn_tshirt_rear_l_02_capsule
   center=<8.3902 1.1817 -1.131>
   rotation=<rot 4.4419 86.7712 0.4311>
  }]
 dyn_tshirt_side_l_02 [{
   shapeType=Capsule
   length=8.7752
   radius=0.9874
   shapeName=dyn_tshirt_rear_l_02_capsule
   center=<8.3108 0.011 0.5527>
   rotation=<rot 5.8349 86.7822 0.4358>
  }]
 pelvis [{
   shapeType=Capsule
   length=13.5647
   radius=11.0166
   shapeName=pelvis_capsule
   center=<-5.5905 0 -4.9939>
   rotation=<rot 82.4887 179.9999 -180>
  }]
 spine_01 [{shapeType=Capsule, length=10, radius=12, shapeName=spine_01_capsule, center=<0 0 -2.0509>, rotation=<rot 90 0 0>}]
 spine_02 [{shapeType=Capsule, length=9, radius=11, shapeName=spine_02_capsule, center=<0 0 -2.1978>, rotation=<rot 90 0 0>}]
 spine_03 [{shapeType=Capsule, length=10, radius=11, shapeName=spine_03_capsule, center=<0 0 -2.0759>, rotation=<rot 90 0 0>}]
 spine_04 [{shapeType=Capsule, length=15, radius=10, shapeName=spine_04_capsule, center=<-1.2504 1.288 -0.1636>, rotation=<rot -3.7107 -90 0.7655>}]
 thigh_l [
  {shapeType=Capsule, length=20, radius=10, shapeName=thigh_l_capsule, center=<-13.8514 -8.0206 5.451>, rotation=<rot 69.8668 96.8026 9.526>}
  {
   shapeType=Capsule
   length=19.9007
   radius=10.0993
   shapeName=thigh_l_capsule
   center=<-14.4259 0 2.8876>
   rotation=<rot 86.7321 179.9999 179.9999>
  }
 ]
}
constraints [
 {
  bone1Name=spine_03
  bone2Name=spine_02
  swing1LimitDegrees=45
  swing1Motion=Free
  swing2LimitDegrees=45
  swing2Motion=Free
  twistLimitDegrees=45
  twistMotion=Free
 }
 {
  bone1Name=spine_02
  bone2Name=spine_01
  swing1LimitDegrees=45
  swing1Motion=Free
  swing2LimitDegrees=45
  swing2Motion=Free
  twistLimitDegrees=45
  twistMotion=Free
 }
 {
  bone1Name=spine_01
  bone2Name=pelvis
  swing1LimitDegrees=45
  swing1Motion=Free
  swing2LimitDegrees=45
  swing2Motion=Free
  twistLimitDegrees=45
  twistMotion=Free
 }
 {
  bone1Name=thigh_l
  bone2Name=pelvis
  swing1LimitDegrees=45
  swing1Motion=Free
  swing2LimitDegrees=45
  swing2Motion=Free
  twistLimitDegrees=45
  twistMotion=Free
 }
 {
  bone1Name=spine_04
  bone2Name=spine_03
  swing1LimitDegrees=45
  swing1Motion=Free
  swing2LimitDegrees=45
  swing2Motion=Free
  twistLimitDegrees=45
  twistMotion=Free
 }
 {
  bone1Name=base_bandolierBack_01
  bone2Name=spine_04
  swing1LimitDegrees=45
  swing1Motion=Free
  swing2LimitDegrees=45
  swing2Motion=Free
  twistLimitDegrees=45
  twistMotion=Free
 }
 {bone1Name=dyn_bandolierBack_01, bone2Name=base_bandolierBack_01, swing1LimitDegrees=10, swing2LimitDegrees=10, twistLimitDegrees=2}
 {bone1Name=dyn_bandolierBack_02, bone2Name=dyn_bandolierBack_01, swing1LimitDegrees=10, swing2LimitDegrees=10, twistLimitDegrees=1}
 {
  bone1Name=base_bandolierFront_01
  bone2Name=spine_04
  swing1LimitDegrees=45
  swing1Motion=Locked
  swing2LimitDegrees=45
  swing2Motion=Locked
  twistLimitDegrees=45
  twistMotion=Locked
 }
 {bone1Name=dyn_bandolierFront_01, bone2Name=base_bandolierFront_01, swing1LimitDegrees=10, swing2LimitDegrees=10, twistLimitDegrees=2}
 {
  bone1Name=base_sideBag_01
  bone2Name=spine_04
  swing1LimitDegrees=45
  swing1Motion=Free
  swing2LimitDegrees=45
  swing2Motion=Free
  twistLimitDegrees=45
  twistMotion=Free
 }
 {bone1Name=dyn_bandolierFront_02, bone2Name=base_bandolierFront_02, swing1LimitDegrees=15, swing2LimitDegrees=15, twistLimitDegrees=0}
 {bone1Name=dyn_sideBag_01, bone2Name=base_sideBag_01, swing1LimitDegrees=40, swing2LimitDegrees=10, twistLimitDegrees=0}
 {bone1Name=dyn_sideBag_02, bone2Name=base_sideBag_02, swing1LimitDegrees=40, swing2LimitDegrees=30, twistLimitDegrees=40}
 {bone1Name=dyn_bandolierSideBack_01, bone2Name=dyn_sideBag_02, swing1LimitDegrees=20, swing2LimitDegrees=10, twistLimitDegrees=1}
 {bone1Name=dyn_bandolierSideFront_01, bone2Name=dyn_sideBag_02, swing1LimitDegrees=20, swing2LimitDegrees=10, twistLimitDegrees=1}
 {
  bone1Name=dyn_tshirt_rear_l_02
  bone2Name=pelvis
  swing1LimitDegrees=45
  swing1Motion=Free
  swing2LimitDegrees=45
  swing2Motion=Free
  twistLimitDegrees=45
  twistMotion=Free
 }
 {
  bone1Name=dyn_tshirt_side_l_02
  bone2Name=pelvis
  swing1LimitDegrees=45
  swing1Motion=Free
  swing2LimitDegrees=45
  swing2Motion=Free
  twistLimitDegrees=45
  twistMotion=Free
 }
 {
  bone1Name=base_sideBag_02
  bone2Name=dyn_sideBag_01
  swing1LimitDegrees=45
  swing1Motion=Free
  swing2LimitDegrees=45
  swing2Motion=Free
  twistLimitDegrees=45
  twistMotion=Free
 }
 {
  bone1Name=base_bandolierFront_02
  bone2Name=dyn_bandolierFront_01
  swing1LimitDegrees=45
  swing1Motion=Free
  swing2LimitDegrees=45
  swing2Motion=Free
  twistLimitDegrees=45
  twistMotion=Free
 }
 {
  bone1Name=dyn_bandolierBack_02
  bone2Name=base_bandolierBack_02
  swing1LimitDegrees=45
  swing1Motion=Free
  swing2LimitDegrees=45
  swing2Motion=Free
  twistLimitDegrees=45
  twistMotion=Free
 }
 {
  bone1Name=base_bandolierBack_02
  bone2Name=dyn_bandolierBack_01
  swing1LimitDegrees=45
  swing1Motion=Free
  swing2LimitDegrees=45
  swing2Motion=Free
  twistLimitDegrees=45
  twistMotion=Free
 }
]
