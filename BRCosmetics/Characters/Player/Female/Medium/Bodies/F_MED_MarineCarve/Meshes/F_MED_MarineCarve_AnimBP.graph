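# Animation Blueprint for the F_MED_MarineCarve female body cosmetic, adding secondary motion to clothing and gear.
# It copies the attached parent's pose into a Control Rig driven by movement, melee and emote states.
# Cached stages then apply kneepad offsets, AnimDynamics for T-shirt, shoelaces and straps, plus RigidBody and AnimDynamics for the side bag.
# Weighted pin-bone constraints and bone offsets join that chain, which a post-sim Control Rig finishes before output.
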AnimBlueprint /BRCosmetics/Characters/Player/Female/Medium/Bodies/F_MED_MarineCarve/Meshes/F_MED_MarineCarve_AnimBP
target_skeleton=&/F_MED_MarineCarve_Skeleton
graphs [
 AnimGraph {nodes [
   n1 = Anim Dynamics - Bone: dyn_tshirt_front_r_01!(ComponentPose, Alpha=1, GravityOverride:Vector="0,0,-980", ExternalForce="0,0,0") -> (Pose) {
    angularDampingOverride=0.9999
    angularSpringConstant=300
    bAngularSpring=true
    bOverrideAngularDamping=true
    bUseGravityOverride=true
    gravityScale=0.35
    simulationSpace=RootRelative
    boundBone {boneName=dyn_tshirt_front_r_01}
    componentAppliedLinearAccClamp=<2000 2000 2000>
    componentLinearAccScale=<0.5 0.5 0.5>
    componentLinearVelScale=<0.5 0.5 0.5>
    gravityOverride {z=-980}
    physicsBodyDefinitions [{
      sphereCollisionRadius=5
      boundBone {boneName=dyn_tshirt_front_r_01}
      boxExtents=<20 40 40>
      constraintSetup {angularLimitsMax=<0 0 0>, angularLimitsMin=<0 0 -10>, angularTarget=<1 0 0>}
      localJointOffset=<5 0 0>
     }]
   }
   Anim Dynamics - Bone: dyn_tshirt_front_r_02!(ComponentPose=@n1, Alpha=1, GravityOverride:Vector="0,0,-980", ExternalForce="0,0,0") -> (Pose) {
    angularDampingOverride=0.9999
    angularSpringConstant=300
    bAngularSpring=true
    bOverrideAngularDamping=true
    bUseGravityOverride=true
    gravityScale=0.35
    simulationSpace=RootRelative
    boundBone {boneName=dyn_tshirt_front_r_02}
    componentAppliedLinearAccClamp=<2000 2000 2000>
    componentLinearAccScale=<0.5 0.5 0.5>
    componentLinearVelScale=<0.5 0.5 0.5>
    gravityOverride {z=-980}
    physicsBodyDefinitions [{
      sphereCollisionRadius=5
      boundBone {boneName=dyn_tshirt_front_r_02}
      boxExtents=<20 40 40>
      constraintSetup {angularLimitsMax=<0 0 0>, angularLimitsMin=<0 0 -10>, angularTarget=<1 0 0>}
      localJointOffset=<5 0 0>
     }]
   }
   n2 = Anim Dynamics - Bone: dyn_tshirt_front_01!(ComponentPose, Alpha=1, GravityOverride:Vector="0,0,-980", ExternalForce="0,0,0") -> (Pose) {
    angularDampingOverride=0.9999
    angularSpringConstant=300
    bAngularSpring=true
    bOverrideAngularDamping=true
    bUseGravityOverride=true
    gravityScale=0.35
    simulationSpace=RootRelative
    boundBone {boneName=dyn_tshirt_front_01}
    componentAppliedLinearAccClamp=<2000 2000 2000>
    componentLinearAccScale=<0.5 0.5 0.5>
    componentLinearVelScale=<0.5 0.5 0.5>
    gravityOverride {z=-980}
    physicsBodyDefinitions [{
      sphereCollisionRadius=5
      boundBone {boneName=dyn_tshirt_front_01}
      boxExtents=<20 40 40>
      constraintSetup {angularLimitsMax=<0 0 0>, angularLimitsMin=<0 0 -10>, angularTarget=<1 0 0>}
      localJointOffset=<5 0 0>
     }]
   }
   n3 = Anim Dynamics - Bone: dyn_tshirt_front_02!(ComponentPose=@n2, Alpha=1, GravityOverride:Vector="0,0,-980", ExternalForce="0,0,0") -> (Pose) {
    angularDampingOverride=0.9999
    angularSpringConstant=300
    bAngularSpring=true
    bOverrideAngularDamping=true
    bUseGravityOverride=true
    gravityScale=0.35
    simulationSpace=RootRelative
    boundBone {boneName=dyn_tshirt_front_02}
    componentAppliedLinearAccClamp=<2000 2000 2000>
    componentLinearAccScale=<0.5 0.5 0.5>
    componentLinearVelScale=<0.5 0.5 0.5>
    gravityOverride {z=-980}
    physicsBodyDefinitions [{
      sphereCollisionRadius=5
      boundBone {boneName=dyn_tshirt_front_02}
      boxExtents=<20 40 40>
      constraintSetup {angularLimitsMax=<0 0 0>, angularLimitsMin=<0 0 -10>, angularTarget=<1 0 0>}
      localJointOffset=<5 0 0>
     }]
   }
   n4 = Anim Dynamics - Bone: dyn_tshirt_front_l_01!(ComponentPose=@n3, Alpha=1, GravityOverride:Vector="0,0,-980", ExternalForce="0,0,0") -> (Pose) {
    angularDampingOverride=0.9999
    angularSpringConstant=300
    bAngularSpring=true
    bOverrideAngularDamping=true
    bUseGravityOverride=true
    gravityScale=0.35
    simulationSpace=RootRelative
    boundBone {boneName=dyn_tshirt_front_l_01}
    componentAppliedLinearAccClamp=<2000 2000 2000>
    componentLinearAccScale=<0.5 0.5 0.5>
    componentLinearVelScale=<0.5 0.5 0.5>
    gravityOverride {z=-980}
    physicsBodyDefinitions [{
      sphereCollisionRadius=5
      boundBone {boneName=dyn_tshirt_front_l_01}
      boxExtents=<20 40 40>
      constraintSetup {angularLimitsMax=<0 0 0>, angularLimitsMin=<0 0 -10>, angularTarget=<1 0 0>}
      localJointOffset=<5 0 0>
     }]
   }
   Anim Dynamics - Bone: dyn_tshirt_front_l_02!(ComponentPose=@n4, Alpha=1, GravityOverride:Vector="0,0,-980", ExternalForce="0,0,0") -> (Pose) {
    angularDampingOverride=0.9999
    angularSpringConstant=300
    bAngularSpring=true
    bOverrideAngularDamping=true
    bUseGravityOverride=true
    gravityScale=0.35
    simulationSpace=RootRelative
    boundBone {boneName=dyn_tshirt_front_l_02}
    componentAppliedLinearAccClamp=<2000 2000 2000>
    componentLinearAccScale=<0.5 0.5 0.5>
    componentLinearVelScale=<0.5 0.5 0.5>
    gravityOverride {z=-980}
    physicsBodyDefinitions [{
      sphereCollisionRadius=5
      boundBone {boneName=dyn_tshirt_front_l_02}
      boxExtents=<20 40 40>
      constraintSetup {angularLimitsMax=<0 0 0>, angularLimitsMin=<0 0 -10>, angularTarget=<1 0 0>}
      localJointOffset=<5 0 0>
     }]
   }
   n5 = RigidBody!(ComponentPose=Local To Component(), Alpha, OverrideWorldGravity:Vector="0,0,0", ExternalForce) -> (Pose) {
    bOverrideWorldGravity=true
    overridePhysicsAsset=&/F_MED_MarineCarve_SideBag_Physics
    simulationSpace=BaseBoneSpace
    baseBoneRef {boneName=spine_01}
    simSpaceSettings {velocityScaleZ=0}
   }
   n6 = Anim Dynamics - Bone: dyn_sideBag_03!(ComponentPose=@n5, Alpha=0, GravityOverride:Vector="0,0,-980", ExternalForce="0,0,0") -> (Pose) {
    angularDampingOverride=1
    angularSpringConstant=8000
    bAngularSpring=true
    bOverrideAngularDamping=true
    bUseGravityOverride=true
    gravityScale=0.35
    simulationSpace=RootRelative
    boundBone {boneName=dyn_sideBag_03}
    componentAppliedLinearAccClamp=<5000 5000 5000>
    componentLinearAccScale=<1 1 1>
    componentLinearVelScale=<1 1 1>
    gravityOverride {z=-980}
    physicsBodyDefinitions [{
      boundBone {boneName=dyn_sideBag_03}
      boxExtents=<20 30 30>
      constraintSetup {angularLimitsMax=<0 20 0>, angularLimitsMin=<0 -20 -40>, angularTarget=<1 0 0>}
      localJointOffset=<6 0 0>
     }]
   }
   n7 = Anim Dynamics - Bone: dyn_sideBag_02!(ComponentPose=@n6, Alpha=0.8, GravityOverride:Vector="0,0,-980", ExternalForce="0,0,0") -> (Pose) {
    angularDampingOverride=1
    angularSpringConstant=8000
    bAngularSpring=true
    bOverrideAngularDamping=true
    bUseGravityOverride=true
    gravityScale=0.35
    simulationSpace=RootRelative
    boundBone {boneName=dyn_sideBag_02}
    componentAppliedLinearAccClamp=<5000 5000 5000>
    componentLinearAccScale=<1 1 1>
    componentLinearVelScale=<1 1 1>
    gravityOverride {z=-980}
    physicsBodyDefinitions [{
      boundBone {boneName=dyn_sideBag_02}
      boxExtents=<20 30 30>
      constraintSetup {angularLimitsMax=<0 5 0>, angularLimitsMin=<0 -5 -40>, angularTarget=<1 0 0>}
      localJointOffset=<4 0 0>
     }]
   }
   n8 = Anim Dynamics - Bone: dyn_bagStrap_01(ComponentPose=@n7, Alpha=0.5, ExternalForce="0,0,0") {
    angularDampingOverride=1
    angularSpringConstant=6000
    bAngularSpring=true
    bOverrideAngularDamping=true
    bUseGravityOverride=true
    gravityScale=0.35
    simulationSpace=RootRelative
    boundBone {boneName=dyn_bagStrap_01}
    componentAppliedLinearAccClamp=<3000 3000 3000>
    componentLinearAccScale=<1 1 1>
    componentLinearVelScale=<1 1 1>
    gravityOverride {z=-980}
    physicsBodyDefinitions [{
      boundBone {boneName=dyn_bagStrap_01}
      boxExtents=<12 30 12>
      constraintSetup {angularLimitsMax=<0 5 40>, angularLimitsMin=<-15 -5 -40>, angularTarget=<1 0 0>}
      localJointOffset=<5 0 0>
     }]
   }
   Anim Dynamics - Bone: dyn_bagStrap_02(ComponentPose=@n8, Alpha=0.6, ExternalForce="0,0,0") {
    angularDampingOverride=1
    angularSpringConstant=6000
    bAngularSpring=true
    bOverrideAngularDamping=true
    bUseGravityOverride=true
    gravityScale=0.35
    simulationSpace=RootRelative
    boundBone {boneName=dyn_bagStrap_02}
    componentAppliedLinearAccClamp=<3000 3000 3000>
    componentLinearAccScale=<1 1 1>
    componentLinearVelScale=<1 1 1>
    gravityOverride {z=-980}
    physicsBodyDefinitions [{
      boundBone {boneName=dyn_bagStrap_02}
      boxExtents=<12 30 12>
      constraintSetup {angularLimitsMax=<0 5 40>, angularLimitsMin=<-15 -5 -40>, angularTarget=<1 0 0>}
      localJointOffset=<5 0 0>
     }]
   }
   n9 = Anim Dynamics(Alpha=1, ExternalForce="0,0,0") {
    angularDampingOverride=0.9999
    angularSpringConstant=6000
    bAngularSpring=true
    bOverrideAngularDamping=true
    bUseGravityOverride=true
    bUsePlanarLimit=false
    bUseSphericalLimits=true
    gravityScale=0.35
    simulationSpace=RootRelative
    componentAppliedLinearAccClamp=<3000 3000 3000>
    componentLinearAccScale=<0.25 0.25 0.25>
    componentLinearVelScale=<0.25 0.25 0.25>
    gravityOverride {z=-980}
    physicsBodyDefinitions [{
      boundBone {boneName=None}
      boxExtents=<50 50 50>
      constraintSetup {angularLimitsMax=<0 0 0>, angularLimitsMin=<0 0 -10>, angularTarget=<1 0 0>}
      localJointOffset=<1 0 0>
     }]
    sphericalLimits [{limitRadius=13.5, drivingBone {boneName=thigh_twist_01_r}, sphereLocalOffset=<-7 -1 -2>}]
    boundBone=~
   }
   n10 = Anim Dynamics(ComponentPose=@n9, Alpha=1, ExternalForce="0,0,0") {
    angularDampingOverride=0.9999
    angularSpringConstant=6000
    bAngularSpring=true
    bOverrideAngularDamping=true
    bUseGravityOverride=true
    bUsePlanarLimit=false
    bUseSphericalLimits=true
    gravityScale=0.35
    simulationSpace=RootRelative
    componentAppliedLinearAccClamp=<3000 3000 3000>
    componentLinearAccScale=<0.25 0.25 0.25>
    componentLinearVelScale=<0.25 0.25 0.25>
    gravityOverride {z=-980}
    physicsBodyDefinitions [{
      boundBone {boneName=None}
      boxExtents=<50 50 50>
      constraintSetup {angularLimitsMax=<0 0 0>, angularLimitsMin=<0 0 -10>, angularTarget=<1 0 0>}
      localJointOffset=<1 0 0>
     }]
    sphericalLimits [{limitRadius=13.5, drivingBone {boneName=thigh_twist_01_r}, sphereLocalOffset=<-7 -1 -2>}]
    boundBone=~
   }
   n11 = Anim Dynamics(ComponentPose=@n10, Alpha=1, ExternalForce="0,0,0") {
    angularDampingOverride=0.9999
    angularSpringConstant=6000
    bAngularSpring=true
    bOverrideAngularDamping=true
    bUseGravityOverride=true
    bUsePlanarLimit=false
    bUseSphericalLimits=true
    gravityScale=0.35
    simulationSpace=RootRelative
    componentAppliedLinearAccClamp=<3000 3000 3000>
    componentLinearAccScale=<0.25 0.25 0.25>
    componentLinearVelScale=<0.25 0.25 0.25>
    gravityOverride {z=-980}
    physicsBodyDefinitions [{
      boundBone {boneName=None}
      boxExtents=<50 50 50>
      constraintSetup {angularLimitsMax=<0 0 0>, angularLimitsMin=<0 0 -10>, angularTarget=<1 0 0>}
      localJointOffset=<1 0 0>
     }]
    sphericalLimits [{limitRadius=11, drivingBone {boneName=thigh_twist_01_l}, sphereLocalOffset=<7 -1 -2>}]
    boundBone=~
   }
   n12 = Anim Dynamics(ComponentPose=@n11, Alpha=1, ExternalForce="0,0,0") {
    angularDampingOverride=0.9999
    angularSpringConstant=6000
    bAngularSpring=true
    bOverrideAngularDamping=true
    bUseGravityOverride=true
    bUsePlanarLimit=false
    bUseSphericalLimits=true
    gravityScale=0.35
    simulationSpace=RootRelative
    componentAppliedLinearAccClamp=<3000 3000 3000>
    componentLinearAccScale=<0.25 0.25 0.25>
    componentLinearVelScale=<0.25 0.25 0.25>
    gravityOverride {z=-980}
    physicsBodyDefinitions [{
      boundBone {boneName=None}
      boxExtents=<50 50 50>
      constraintSetup {angularLimitsMax=<0 0 0>, angularLimitsMin=<0 0 -10>, angularTarget=<1 0 0>}
      localJointOffset=<1 0 0>
     }]
    sphericalLimits [{limitRadius=11, drivingBone {boneName=thigh_twist_01_l}, sphereLocalOffset=<7 -1 -2>}]
    boundBone=~
   }
   n13 = Anim Dynamics(ComponentPose=@n12, Alpha=1, ExternalForce="0,0,0") {
    angularDampingOverride=0.9999
    angularSpringConstant=6000
    bAngularSpring=true
    bOverrideAngularDamping=true
    bUseGravityOverride=true
    bUsePlanarLimit=false
    bUseSphericalLimits=true
    gravityScale=0.35
    simulationSpace=RootRelative
    componentAppliedLinearAccClamp=<3000 3000 3000>
    componentLinearAccScale=<0.25 0.25 0.25>
    componentLinearVelScale=<0.25 0.25 0.25>
    gravityOverride {z=-980}
    physicsBodyDefinitions [{
      boundBone {boneName=None}
      boxExtents=<50 50 50>
      constraintSetup {angularLimitsMax=<0 0 0>, angularLimitsMin=<0 0 -10>, angularTarget=<1 0 0>}
      localJointOffset=<1 0 0>
     }]
    sphericalLimits [{limitRadius=13.5, drivingBone {boneName=thigh_twist_01_r}, sphereLocalOffset=<-7 -1 -2>}]
    boundBone=~
   }
   Anim Dynamics(ComponentPose=@n13, Alpha=1, ExternalForce="0,0,0") {
    angularDampingOverride=0.9999
    angularSpringConstant=6000
    bAngularSpring=true
    bOverrideAngularDamping=true
    bUseGravityOverride=true
    bUsePlanarLimit=false
    bUseSphericalLimits=true
    gravityScale=0.35
    simulationSpace=RootRelative
    componentAppliedLinearAccClamp=<3000 3000 3000>
    componentLinearAccScale=<0.25 0.25 0.25>
    componentLinearVelScale=<0.25 0.25 0.25>
    gravityOverride {z=-980}
    physicsBodyDefinitions [{
      boundBone {boneName=None}
      boxExtents=<50 50 50>
      constraintSetup {angularLimitsMax=<0 0 0>, angularLimitsMin=<0 0 -10>, angularTarget=<1 0 0>}
      localJointOffset=<1 0 0>
     }]
    sphericalLimits [{limitRadius=13.5, drivingBone {boneName=thigh_twist_01_r}, sphereLocalOffset=<-7 -1 -2>}]
    boundBone=~
   }
   Component To Local()
   n14 = "Constraint - base_pin_01 (PPPP)"!(
    ComponentPose
    Alpha=1
    ConstraintWeights_0=0.843
    ConstraintWeights_1:float=0.1489
    ConstraintWeights_2:float=0.0049
    ConstraintWeights_3:float=0.0032
   ) -> (Pose) {
    boneToModify {boneName=base_pin_01}
    constraintSetup [{targetBone {boneName=spine_05}}, {targetBone {boneName=spine_04}}, {targetBone {boneName=neck_01}}, {targetBone {boneName=deform_pec_r}}]
    constraintWeights [1,1,1,1]
   }
   "Constraint - base_pin_02 (PPPP)"!(
    ComponentPose=@n14
    Alpha=1
    ConstraintWeights_0=0.4512
    ConstraintWeights_1:float=0.4249
    ConstraintWeights_2:float=0.1238
    ConstraintWeights_3:float=0.0001
   ) -> (Pose) {
    boneToModify {boneName=base_pin_02}
    constraintSetup [
     {targetBone {boneName=spine_05}}
     {targetBone {boneName=spine_04}}
     {targetBone {boneName=deform_pec_r}}
     {targetBone {boneName=deform_pec_l}}
    ]
    constraintWeights [1,1,1,1]
   }
   RigidBody(OverrideWorldGravity="0,0,-980") {
    bOverrideWorldGravity=true
    overridePhysicsAsset=&/F_MED_MarineCarve_Physics
    simulationSpace=BaseBoneSpace
    baseBoneRef {boneName=spine_01}
    overrideWorldGravity {z=-980}
    simSpaceSettings {maxAngularAcceleration=2, maxAngularVelocity=30, maxLinearAcceleration=50, maxLinearVelocity=400, worldAlpha=1}
   }
   Output Pose(Result=Use cached pose 'CR_ExtraDyn'()) {layerGroup=None}
   n15 = Control Rig!(
    Source=Copy Pose From Mesh() {bUseAttachedParent=true}
    Alpha
    Crouching:bool=false
    CrouchingFloat:double=Get isCrouching float() -> (isCrouching float:double)
    KneeSliding:bool=Get bIsKneeSliding()
    BaseballSliding:bool=Get bIsBaseballSliding()
    PickaxeSwing:bool=Get bIsPlayingMeleeAnim()
    PlayingEmote:bool="Get PlayingEmoteMontagedouble "() -> ("PlayingEmoteMontagedouble ":bool)
    CrounchingWalk2D:double=Get CrouchMOV() -> (CrouchMOV:double)
    DoubleHandedWeapon:bool=false
    isDBNO:bool=false
    bigGunAiming:bool=false
   ) -> (Pose) {controlRigAssetReference {blueprintRigClass=&/F_MED_MarineCarve_CtrlRig.F_MED_MarineCarve_CtrlRig_C}}
   Save cached pose 'CR_Input'(Pose=@n15) {cacheName=CR_Input}
   n16 = Anim Dynamics - Bone: dyn_shoelace_l_01(ComponentPose=Local To Component(LocalPose=Use cached pose 'CR_TShirtDyn'()), Alpha=1, ExternalForce="0,0,0") {
    angularDampingOverride=1
    angularSpringConstant=1000
    bAngularSpring=true
    bOverrideAngularDamping=true
    bUseGravityOverride=true
    gravityScale=0.35
    simulationSpace=RootRelative
    boundBone {boneName=dyn_shoelace_l_01}
    componentAppliedLinearAccClamp=<3000 3000 3000>
    componentLinearAccScale=<5 5 5>
    componentLinearVelScale=<5 5 5>
    gravityOverride {z=-980}
    physicsBodyDefinitions [{
      boundBone {boneName=dyn_shoelace_l_01}
      boxExtents=<30 30 30>
      constraintSetup {angularLimitsMax=<0 0 0>, angularLimitsMin=<0 0 -30>, angularTarget=<1 0 0>}
      localJointOffset=<4 0 0>
     }]
   }
   n17 = Anim Dynamics - Bone: dyn_shoelace_l_02(ComponentPose=@n16, Alpha=1, ExternalForce="0,0,0") {
    angularDampingOverride=1
    angularSpringConstant=1000
    bAngularSpring=true
    bOverrideAngularDamping=true
    bUseGravityOverride=true
    gravityScale=0.35
    simulationSpace=RootRelative
    boundBone {boneName=dyn_shoelace_l_02}
    componentAppliedLinearAccClamp=<3000 3000 3000>
    componentLinearAccScale=<5 5 5>
    componentLinearVelScale=<5 5 5>
    gravityOverride {z=-980}
    physicsBodyDefinitions [{
      boundBone {boneName=dyn_shoelace_l_02}
      boxExtents=<30 30 30>
      constraintSetup {angularLimitsMax=<0 5 0>, angularLimitsMin=<0 -5 -50>, angularTarget=<1 0 0>}
      localJointOffset=<4 0 0>
     }]
   }
   n18 = Anim Dynamics - Bone: dyn_shoelace_l_03(ComponentPose=@n17, Alpha=1, ExternalForce="0,0,0") {
    angularDampingOverride=1
    angularSpringConstant=1000
    bAngularSpring=true
    bOverrideAngularDamping=true
    bUseGravityOverride=true
    gravityScale=0.35
    simulationSpace=RootRelative
    boundBone {boneName=dyn_shoelace_l_03}
    componentAppliedLinearAccClamp=<3000 3000 3000>
    componentLinearAccScale=<5 5 5>
    componentLinearVelScale=<5 5 5>
    gravityOverride {z=-980}
    physicsBodyDefinitions [{
      boundBone {boneName=dyn_shoelace_l_03}
      boxExtents=<30 30 30>
      constraintSetup {angularLimitsMax=<0 0 0>, angularLimitsMin=<0 0 -30>, angularTarget=<1 0 0>}
      localJointOffset=<4 0 0>
     }]
   }
   n19 = Anim Dynamics - Bone: dyn_shoelace_l_04(ComponentPose=@n18, Alpha=1, ExternalForce="0,0,0") {
    angularDampingOverride=1
    angularSpringConstant=1000
    bAngularSpring=true
    bOverrideAngularDamping=true
    bUseGravityOverride=true
    gravityScale=0.35
    simulationSpace=RootRelative
    boundBone {boneName=dyn_shoelace_l_04}
    componentAppliedLinearAccClamp=<3000 3000 3000>
    componentLinearAccScale=<5 5 5>
    componentLinearVelScale=<5 5 5>
    gravityOverride {z=-980}
    physicsBodyDefinitions [{
      boundBone {boneName=dyn_shoelace_l_04}
      boxExtents=<30 30 30>
      constraintSetup {angularLimitsMax=<0 5 0>, angularLimitsMin=<0 -5 -50>, angularTarget=<1 0 0>}
      localJointOffset=<4 0 0>
     }]
   }
   n20 = Anim Dynamics - Bone: dyn_shoelace_r_03(ComponentPose=@n19, Alpha=1, ExternalForce="0,0,0") {
    angularDampingOverride=1
    angularSpringConstant=1000
    bAngularSpring=true
    bOverrideAngularDamping=true
    bUseGravityOverride=true
    gravityScale=0.35
    simulationSpace=RootRelative
    boundBone {boneName=dyn_shoelace_r_03}
    componentAppliedLinearAccClamp=<3000 3000 3000>
    componentLinearAccScale=<5 5 5>
    componentLinearVelScale=<5 5 5>
    gravityOverride {z=-980}
    physicsBodyDefinitions [{
      boundBone {boneName=dyn_shoelace_r_03}
      boxExtents=<30 30 30>
      constraintSetup {angularLimitsMax=<0 0 0>, angularLimitsMin=<0 0 -30>, angularTarget=<1 0 0>}
      localJointOffset=<4 0 0>
     }]
   }
   n21 = Anim Dynamics - Bone: dyn_shoelace_r_04(ComponentPose=@n20, Alpha=1, ExternalForce="0,0,0") {
    angularDampingOverride=1
    angularSpringConstant=1000
    bAngularSpring=true
    bOverrideAngularDamping=true
    bUseGravityOverride=true
    gravityScale=0.35
    simulationSpace=RootRelative
    boundBone {boneName=dyn_shoelace_r_04}
    componentAppliedLinearAccClamp=<3000 3000 3000>
    componentLinearAccScale=<5 5 5>
    componentLinearVelScale=<5 5 5>
    gravityOverride {z=-980}
    physicsBodyDefinitions [{
      boundBone {boneName=dyn_shoelace_r_04}
      boxExtents=<30 30 30>
      constraintSetup {angularLimitsMax=<0 5 0>, angularLimitsMin=<0 -5 -50>, angularTarget=<1 0 0>}
      localJointOffset=<4 0 0>
     }]
   }
   n22 = Anim Dynamics - Bone: dyn_shoelace_r_01(ComponentPose=@n21, Alpha=1, ExternalForce="0,0,0") {
    angularDampingOverride=1
    angularSpringConstant=1000
    bAngularSpring=true
    bOverrideAngularDamping=true
    bUseGravityOverride=true
    gravityScale=0.35
    simulationSpace=RootRelative
    boundBone {boneName=dyn_shoelace_r_01}
    componentAppliedLinearAccClamp=<3000 3000 3000>
    componentLinearAccScale=<5 5 5>
    componentLinearVelScale=<5 5 5>
    gravityOverride {z=-980}
    physicsBodyDefinitions [{
      boundBone {boneName=dyn_shoelace_r_01}
      boxExtents=<30 30 30>
      constraintSetup {angularLimitsMax=<0 0 0>, angularLimitsMin=<0 0 -30>, angularTarget=<1 0 0>}
      localJointOffset=<4 0 0>
     }]
   }
   n23 = Anim Dynamics - Bone: dyn_shoelace_r_02(ComponentPose=@n22, Alpha=1, ExternalForce="0,0,0") {
    angularDampingOverride=1
    angularSpringConstant=1000
    bAngularSpring=true
    bOverrideAngularDamping=true
    bUseGravityOverride=true
    gravityScale=0.35
    simulationSpace=RootRelative
    boundBone {boneName=dyn_shoelace_r_02}
    componentAppliedLinearAccClamp=<3000 3000 3000>
    componentLinearAccScale=<5 5 5>
    componentLinearVelScale=<5 5 5>
    gravityOverride {z=-980}
    physicsBodyDefinitions [{
      boundBone {boneName=dyn_shoelace_r_02}
      boxExtents=<30 30 30>
      constraintSetup {angularLimitsMax=<0 5 0>, angularLimitsMin=<0 -5 -50>, angularTarget=<1 0 0>}
      localJointOffset=<4 0 0>
     }]
   }
   n24 = "Constraint - base_pin_01 (PPPPPPP)"!(
    ComponentPose=@n23
    Alpha=1
    ConstraintWeights_0=0.3896
    ConstraintWeights_1:float=0.2677
    ConstraintWeights_2:float=0.2512
    ConstraintWeights_3:float=0.0467
    ConstraintWeights_4:float=0.0394
    ConstraintWeights_5:float=0.0035
    ConstraintWeights_6:float=0.0019
   ) -> (Pose) {
    boneToModify {boneName=base_pin_01}
    constraintSetup [
     {targetBone {boneName=Belt_Spine05_Fr}}
     {targetBone {boneName=spine_05}}
     {targetBone {boneName=spine_04}}
     {targetBone {boneName=clavicle_r}}
     {targetBone {boneName=dyn_bandolierFront_01}}
     {targetBone {boneName=deform_pec_r}}
     {targetBone {boneName=neck_01}}
    ]
    constraintWeights [1,1,1,1,1,1,1]
   }
   n25 = "Constraint - base_pin_02 (PPPPPPP)"!(
    ComponentPose=@n24
    Alpha=1
    ConstraintWeights_0=0.3469
    ConstraintWeights_1:float=0.3312
    ConstraintWeights_2:float=0.1871
    ConstraintWeights_3:float=0.104
    ConstraintWeights_4:float=0.0199
    ConstraintWeights_5:float=0.0079
    ConstraintWeights_6:float=0.003
   ) -> (Pose) {
    boneToModify {boneName=base_pin_02}
    constraintSetup [
     {targetBone {boneName=Belt_Spine05_Fr}}
     {targetBone {boneName=spine_04}}
     {targetBone {boneName=deform_pec_r}}
     {targetBone {boneName=dyn_bandolierFront_01}}
     {targetBone {boneName=clavicle_r}}
     {targetBone {boneName=deform_pec_l}}
     {targetBone {boneName=spine_05}}
    ]
    constraintWeights [1,1,1,1,1,1,1]
   }
   n26 = "Transform (Modify) Bone - Bone: Belt_Spine05_Fr"!(ComponentPose=@n25, bAlphaBoolEnabled:bool=True, Translation="-0.001,1.941,-0.276", Rotation="8.298,1.961,13.071", Scale="1,1,1") -> (Pose) {
    alphaInputType=Bool
    rotationMode=BMM_Additive
    rotationSpace=BCS_BoneSpace
    translationMode=BMM_Additive
    translationSpace=BCS_BoneSpace
    boneToModify {boneName=Belt_Spine05_Fr}
    rotation=<rot 8.2981 1.961 13.0709>
    translation=<-0.001 1.941 -0.276>
   }
   n27 = "Transform (Modify) Bone - Bone: dyn_underSleeve_l"!(ComponentPose=@n26, bAlphaBoolEnabled:bool=True, Translation="1.110,-1.670,-7.015", Rotation="9.003,2.131,13.115", Scale="1,1,1") -> (Pose) {
    alphaInputType=Bool
    rotationMode=BMM_Additive
    rotationSpace=BCS_BoneSpace
    translationMode=BMM_Additive
    translationSpace=BCS_BoneSpace
    boneToModify {boneName=dyn_underSleeve_l}
    rotation=<rot 9.003 2.131 13.115>
    translation=<1.11 -1.67 -7.015>
   }
   n28 = "Transform (Modify) Bone - Bone: base_sideBag_01"!(ComponentPose=@n27, bAlphaBoolEnabled:bool=True, Rotation="10,-15,0") -> (Pose) {
    alphaInputType=Bool
    rotationMode=BMM_Additive
    rotationSpace=BCS_BoneSpace
    alphaBoolBlend {blendInTime=0.2, blendOutTime=0.5}
    boneToModify {boneName=base_sideBag_01}
    rotation {pitch=10, yaw=-15}
   }
   n29 = RigidBody(ComponentPose=@n28, OverrideWorldGravity="0,0,0") {
    bOverrideWorldGravity=true
    overridePhysicsAsset=&F_MED_MarineCarve_SideBag_Physics
    simulationSpace=BaseBoneSpace
    baseBoneRef {boneName=pelvis}
    simSpaceSettings {
     dampingAlpha=0
     maxAngularAcceleration=2
     maxAngularVelocity=3
     maxLinearAcceleration=1000
     maxLinearVelocity=100
     velocityScaleZ=0.2
     worldAlpha=1
     externalLinearDragV=<1 1 0.9>
    }
   }
   n30 = "Transform (Modify) Bone - Bone: base_sideBag_01"(ComponentPose=@n29, bAlphaBoolEnabled=True, Rotation="0,0,0") {
    alphaInputType=Bool
    rotationMode=BMM_Additive
    rotationSpace=BCS_BoneSpace
    alphaBoolBlend {blendInTime=0.2, blendOutTime=0.2}
    boneToModify {boneName=base_sideBag_01}
   }
   n31 = Anim Dynamics - Bone: dyn_sideBag_01!(ComponentPose=@n30, Alpha=0.7, GravityOverride:Vector="0,0,-980", ExternalForce="0,0,0") -> (Pose) {
    angularDampingOverride=1
    angularSpringConstant=8000
    bAngularSpring=true
    bOverrideAngularDamping=true
    bUseGravityOverride=true
    gravityScale=0.35
    simulationSpace=RootRelative
    boundBone {boneName=dyn_sideBag_01}
    componentAppliedLinearAccClamp=<5000 5000 5000>
    componentLinearAccScale=<1 1 1>
    componentLinearVelScale=<1 1 1>
    gravityOverride {z=-980}
    physicsBodyDefinitions [{
      boundBone {boneName=dyn_sideBag_01}
      boxExtents=<20 30 30>
      constraintSetup {angularLimitsMax=<0 5 0>, angularLimitsMin=<0 -5 -40>, angularTarget=<1 0 0>}
      localJointOffset=<1 0 0>
     }]
    simSpaceSettings {maxAngularAcceleration=25, maxAngularVelocity=5, simSpaceAngularAlpha=1}
   }
   n32 = Anim Dynamics - Bone: dyn_sideBag_03(ComponentPose=@n31, Alpha=0, GravityOverride="0,0,-980", ExternalForce="0,0,0") {
    angularDampingOverride=1
    angularSpringConstant=8000
    bAngularSpring=true
    bOverrideAngularDamping=true
    bUseGravityOverride=true
    gravityScale=0.35
    simulationSpace=RootRelative
    boundBone {boneName=dyn_sideBag_03}
    componentAppliedLinearAccClamp=<5000 5000 5000>
    componentLinearAccScale=<1 1 1>
    componentLinearVelScale=<1 1 1>
    gravityOverride {z=-980}
    physicsBodyDefinitions [{
      boundBone {boneName=dyn_sideBag_03}
      boxExtents=<20 30 30>
      constraintSetup {angularLimitsMax=<0 20 0>, angularLimitsMin=<0 -20 -40>, angularTarget=<1 0 0>}
      localJointOffset=<6 0 0>
     }]
   }
   n33 = Anim Dynamics - Bone: dyn_sideBag_02!(ComponentPose=@n32, Alpha=0.8, GravityOverride="0,0,-980") -> (Pose) {
    angularDampingOverride=1
    angularSpringConstant=8000
    bAngularSpring=true
    bOverrideAngularDamping=true
    bUseGravityOverride=true
    gravityScale=0.35
    simulationSpace=RootRelative
    boundBone {boneName=dyn_sideBag_02}
    componentAppliedLinearAccClamp=<5000 5000 5000>
    componentLinearAccScale=<1 1 1>
    componentLinearVelScale=<1 1 1>
    gravityOverride {z=-980}
    physicsBodyDefinitions [{
      boundBone {boneName=dyn_sideBag_02}
      boxExtents=<20 30 30>
      constraintSetup {angularLimitsMax=<0 5 0>, angularLimitsMin=<0 -5 -40>, angularTarget=<1 0 0>}
      localJointOffset=<4 0 0>
     }]
   }
   n34 = Anim Dynamics - Bone: dyn_bagStrap_01(ComponentPose=@n33, Alpha=0.5, ExternalForce="0,0,0") {
    angularDampingOverride=1
    angularSpringConstant=6000
    bAngularSpring=true
    bOverrideAngularDamping=true
    bUseGravityOverride=true
    gravityScale=0.35
    simulationSpace=RootRelative
    boundBone {boneName=dyn_bagStrap_01}
    componentAppliedLinearAccClamp=<3000 3000 3000>
    componentLinearAccScale=<1 1 1>
    componentLinearVelScale=<1 1 1>
    gravityOverride {z=-980}
    physicsBodyDefinitions [{
      boundBone {boneName=dyn_bagStrap_01}
      boxExtents=<12 30 12>
      constraintSetup {angularLimitsMax=<0 5 40>, angularLimitsMin=<-15 -5 -40>, angularTarget=<1 0 0>}
      localJointOffset=<5 0 0>
     }]
   }
   n35 = Anim Dynamics - Bone: dyn_bagStrap_02(ComponentPose=@n34, Alpha=0.6, ExternalForce="0,0,0") {
    angularDampingOverride=1
    angularSpringConstant=6000
    bAngularSpring=true
    bOverrideAngularDamping=true
    bUseGravityOverride=true
    gravityScale=0.35
    simulationSpace=RootRelative
    boundBone {boneName=dyn_bagStrap_02}
    componentAppliedLinearAccClamp=<3000 3000 3000>
    componentLinearAccScale=<1 1 1>
    componentLinearVelScale=<1 1 1>
    gravityOverride {z=-980}
    physicsBodyDefinitions [{
      boundBone {boneName=dyn_bagStrap_02}
      boxExtents=<12 30 12>
      constraintSetup {angularLimitsMax=<0 5 40>, angularLimitsMin=<-15 -5 -40>, angularTarget=<1 0 0>}
      localJointOffset=<5 0 0>
     }]
   }
   n36 = Control Rig!(
    Source=Component To Local(ComponentPose=@n35)
    bAlphaBoolEnabled:bool=True
    CrouchingFloat=Get isCrouching float()
    KneeSliding=Get bIsKneeSliding()
    BaseballSliding=Get bIsBaseballSliding()
    Jumping:bool=Get Isjumping() -> (Isjumping:bool)
    SkydiveFloat:bool=Get bIsSkydiveFloating()
    ParachuteOpened:bool=Get bIsParachuteOpened()
    PlayingMeleAnim:bool=Get bIsPlayingMeleeAnim()
    MoveLeft:double="Get Move Left "() -> ("Move Left ":double)
    MoveRight:double=Get Move Right() -> (Move Right:double)
    Playing Emote:bool="Get PlayingEmoteMontagedouble "()
    PlayingMeleCrounchingAnim:bool=Get Is Playing MeleeCrounching Anim() -> (Is Playing MeleeCrounching Anim:bool)
    RB_Enabled:bool=false
    CrounchingMoving2D:double=Get CrouchMOV()
    Crounchinggunout:double=Get isCrouchinggunOut() -> (isCrouchinggunOut:double)
    isTargeting:bool=false
    isSkyDiving:bool=false
    isDiveMode:bool=false
    CustomEmotePlaying:bool=Get isCustomEmotePlaying() -> (isCustomEmotePlaying:bool)
    FrontEndPlayingnoEmote:bool=Get IsFrontEndNOEmote() -> (IsFrontEndNOEmote:bool)
    NotInFrontEnd:bool=false
   ) -> (Pose) {
    alphaInputType=Bool
    alphaBoolBlend {blendInTime=0.2, blendOutTime=0.2}
    controlRigAssetReference {blueprintRigClass=&/F_MED_MarineCarve_PostSim_CtrlRig.F_MED_MarineCarve_PostSim_CtrlRig_C}
   }
   Save cached pose 'CR_ExtraDyn'(Pose=@n36) {cacheName=CR_ExtraDyn}
   n37 = "Transform (Modify) Bone - Bone: Kneepad_Helper"!(ComponentPose=Local To Component(LocalPose=Use cached pose 'CR_Input'()), bAlphaBoolEnabled:bool=True, Translation="0,4,0") -> (Pose) {
    alphaInputType=Bool
    rotationMode=BMM_Additive
    rotationSpace=BCS_BoneSpace
    translationMode=BMM_Additive
    translationSpace=BCS_BoneSpace
    alphaBoolBlend {blendInTime=0.25, blendOutTime=0.25}
    boneToModify {boneName=Kneepad_Helper}
    translation {y=4}
   }
   n38 = "Transform (Modify) Bone - Bone: Kneepad_Helper"(ComponentPose=@n37, bAlphaBoolEnabled=True, Translation="0,4,0") {
    alphaInputType=Bool
    rotationMode=BMM_Additive
    rotationSpace=BCS_BoneSpace
    translationMode=BMM_Additive
    translationSpace=BCS_BoneSpace
    alphaBoolBlend {blendInTime=0.25, blendOutTime=0.25}
    boneToModify {boneName=Kneepad_Helper}
    translation {y=4}
   }
   n39 = Anim Dynamics - Bone: dyn_tshirt_side_l_01!(
    ComponentPose=@n38
    Alpha=1
    GravityOverride:Vector="0,0,-980"
    SimSpaceSettings:Anim Phys Sim Space Settings Structure=(SimSpaceAngularAlpha=0,MaxAngularVelocity=10000,MaxAngularAcceleration=10000,ExternalAngularVelocity=(X=0,Y=0,Z=0))
   ) -> (Pose) {
    angularDampingOverride=0.9999
    angularSpringConstant=300
    bAngularSpring=true
    bOverrideAngularDamping=true
    bUseGravityOverride=true
    gravityScale=0.35
    simulationSpace=RootRelative
    boundBone {boneName=dyn_tshirt_side_l_01}
    componentAppliedLinearAccClamp=<2000 2000 2000>
    componentLinearAccScale=<0.5 0.5 0.5>
    componentLinearVelScale=<0.5 0.5 0.5>
    gravityOverride {z=-980}
    physicsBodyDefinitions [{
      sphereCollisionRadius=5
      boundBone {boneName=dyn_tshirt_side_l_01}
      boxExtents=<20 40 40>
      constraintSetup {angularLimitsMax=<0 0 0>, angularLimitsMin=<0 0 -5>, angularTarget=<1 0 0>}
      localJointOffset=<3 0 0>
     }]
   }
   n40 = Anim Dynamics - Bone: dyn_tshirt_side_l_02!(
    ComponentPose=@n39
    Alpha=1
    GravityOverride:Vector="0,0,-980"
    SimSpaceSettings:Anim Phys Sim Space Settings Structure=(SimSpaceAngularAlpha=0,MaxAngularVelocity=10000,MaxAngularAcceleration=10000,ExternalAngularVelocity=(X=0,Y=0,Z=0))
   ) -> (Pose) {
    angularDampingOverride=0.9999
    angularSpringConstant=300
    bAngularSpring=true
    bOverrideAngularDamping=true
    bUseGravityOverride=true
    gravityScale=0.35
    simulationSpace=RootRelative
    boundBone {boneName=dyn_tshirt_side_l_02}
    componentAppliedLinearAccClamp=<2000 2000 2000>
    componentLinearAccScale=<0.5 0.5 0.5>
    componentLinearVelScale=<0.5 0.5 0.5>
    gravityOverride {z=-980}
    physicsBodyDefinitions [{
      sphereCollisionRadius=5
      boundBone {boneName=dyn_tshirt_side_l_02}
      boxExtents=<20 40 40>
      constraintSetup {angularLimitsMax=<0 0 0>, angularLimitsMin=<0 0 -5>, angularTarget=<1 0 0>}
      localJointOffset=<3 0 0>
     }]
   }
   n41 = Anim Dynamics - Bone: dyn_tshirt_side_r_01!(
    ComponentPose=@n40
    Alpha=0.9
    GravityOverride:Vector="0,0,-980"
    SimSpaceSettings:Anim Phys Sim Space Settings Structure=(SimSpaceAngularAlpha=0,MaxAngularVelocity=10000,MaxAngularAcceleration=10000,ExternalAngularVelocity=(X=0,Y=0,Z=0))
   ) -> (Pose) {
    angularDampingOverride=0.9999
    angularSpringConstant=300
    bAngularSpring=true
    bOverrideAngularDamping=true
    bUseGravityOverride=true
    gravityScale=0.35
    simulationSpace=RootRelative
    boundBone {boneName=dyn_tshirt_side_r_01}
    componentAppliedLinearAccClamp=<2000 2000 2000>
    componentLinearAccScale=<0.5 0.5 0.5>
    componentLinearVelScale=<0.5 0.5 0.5>
    gravityOverride {z=-980}
    physicsBodyDefinitions [{
      sphereCollisionRadius=5
      boundBone {boneName=dyn_tshirt_side_r_01}
      boxExtents=<20 40 40>
      constraintSetup {angularLimitsMax=<0 0 0>, angularLimitsMin=<0 0 -6>, angularTarget=<1 0 0>}
      localJointOffset=<5 0 0>
     }]
   }
   n42 = Anim Dynamics - Bone: dyn_tshirt_side_r_02!(
    ComponentPose=@n41
    Alpha=0.9
    GravityOverride:Vector="0,0,-980"
    SimSpaceSettings:Anim Phys Sim Space Settings Structure=(SimSpaceAngularAlpha=0,MaxAngularVelocity=10000,MaxAngularAcceleration=10000,ExternalAngularVelocity=(X=0,Y=0,Z=0))
   ) -> (Pose) {
    angularDampingOverride=0.9999
    angularSpringConstant=300
    bAngularSpring=true
    bOverrideAngularDamping=true
    bUseGravityOverride=true
    gravityScale=0.35
    simulationSpace=RootRelative
    boundBone {boneName=dyn_tshirt_side_r_02}
    componentAppliedLinearAccClamp=<2000 2000 2000>
    componentLinearAccScale=<0.5 0.5 0.5>
    componentLinearVelScale=<0.5 0.5 0.5>
    gravityOverride {z=-980}
    physicsBodyDefinitions [{
      sphereCollisionRadius=5
      boundBone {boneName=dyn_tshirt_side_r_02}
      boxExtents=<20 40 40>
      constraintSetup {angularLimitsMax=<0 0 0>, angularLimitsMin=<0 0 -6>, angularTarget=<1 0 0>}
      localJointOffset=<5 0 0>
     }]
   }
   n43 = Anim Dynamics - Bone: dyn_tshirt_rear_l_01!(
    ComponentPose=@n42
    Alpha=1
    GravityOverride:Vector="0,0,-980"
    SimSpaceSettings:Anim Phys Sim Space Settings Structure=(SimSpaceAngularAlpha=0,MaxAngularVelocity=10000,MaxAngularAcceleration=10000,ExternalAngularVelocity=(X=0,Y=0,Z=0))
   ) -> (Pose) {
    angularDampingOverride=0.9999
    angularSpringConstant=800
    bAngularSpring=true
    bOverrideAngularDamping=true
    bUseGravityOverride=true
    gravityScale=0.35
    simulationSpace=RootRelative
    boundBone {boneName=dyn_tshirt_rear_l_01}
    componentAppliedLinearAccClamp=<2500 2500 2500>
    componentLinearAccScale=<0.5 0.5 0.5>
    componentLinearVelScale=<0.5 0.5 0.5>
    gravityOverride {z=-980}
    physicsBodyDefinitions [{
      collisionType=CustomSphere
      sphereCollisionRadius=2
      boundBone {boneName=dyn_tshirt_rear_l_01}
      boxExtents=<40 40 40>
      constraintSetup {angularLimitsMax=<0 0 0>, angularLimitsMin=<0 0 -6>, angularTarget=<0 0 0>}
      localJointOffset=<-4 0.2 -0.1>
     }]
   }
   n44 = Anim Dynamics - Bone: dyn_tshirt_rear_l_02!(
    ComponentPose=@n43
    Alpha=1
    GravityOverride:Vector="0,0,-980"
    SimSpaceSettings:Anim Phys Sim Space Settings Structure=(SimSpaceAngularAlpha=0,MaxAngularVelocity=10000,MaxAngularAcceleration=10000,ExternalAngularVelocity=(X=0,Y=0,Z=0))
   ) -> (Pose) {
    angularDampingOverride=0.9999
    angularSpringConstant=800
    bAngularSpring=true
    bOverrideAngularDamping=true
    bUseGravityOverride=true
    gravityScale=0.35
    simulationSpace=RootRelative
    boundBone {boneName=dyn_tshirt_rear_l_02}
    componentAppliedLinearAccClamp=<2500 2500 2500>
    componentLinearAccScale=<0.5 0.5 0.5>
    componentLinearVelScale=<0.5 0.5 0.5>
    gravityOverride {z=-980}
    physicsBodyDefinitions [{
      collisionType=CustomSphere
      sphereCollisionRadius=2
      boundBone {boneName=dyn_tshirt_rear_l_02}
      boxExtents=<20 20 20>
      constraintSetup {angularLimitsMax=<0 0 0>, angularLimitsMin=<0 0 -6>, angularTarget=<0 0 0>}
      localJointOffset=<-4 0.2 -0.1>
     }]
   }
   n45 = Anim Dynamics - Bone: dyn_tshirt_rear_r_01!(
    ComponentPose=@n44
    Alpha=1
    GravityOverride:Vector="0,0,-980"
    SimSpaceSettings:Anim Phys Sim Space Settings Structure=(SimSpaceAngularAlpha=0,MaxAngularVelocity=10000,MaxAngularAcceleration=10000,ExternalAngularVelocity=(X=0,Y=0,Z=0))
   ) -> (Pose) {
    angularDampingOverride=0.9999
    angularSpringConstant=800
    bAngularSpring=true
    bOverrideAngularDamping=true
    bUseGravityOverride=true
    gravityScale=0.35
    simulationSpace=RootRelative
    boundBone {boneName=dyn_tshirt_rear_r_01}
    componentAppliedLinearAccClamp=<2500 2500 2500>
    componentLinearAccScale=<0.5 0.5 0.5>
    componentLinearVelScale=<0.5 0.5 0.5>
    gravityOverride {z=-980}
    physicsBodyDefinitions [{
      collisionType=CustomSphere
      sphereCollisionRadius=2
      boundBone {boneName=dyn_tshirt_rear_r_01}
      boxExtents=<33 33 33>
      constraintSetup {angularLimitsMax=<0 0 0>, angularLimitsMin=<0 0 -6>, angularTarget=<0 0 0>}
      localJointOffset=<-4 0.2 -0.1>
     }]
   }
   n46 = Anim Dynamics - Bone: dyn_tshirt_rear_r_02!(
    ComponentPose=@n45
    Alpha=1
    GravityOverride:Vector="0,0,-980"
    SimSpaceSettings:Anim Phys Sim Space Settings Structure=(SimSpaceAngularAlpha=0,MaxAngularVelocity=10000,MaxAngularAcceleration=10000,ExternalAngularVelocity=(X=0,Y=0,Z=0))
   ) -> (Pose) {
    angularDampingOverride=0.9999
    angularSpringConstant=800
    bAngularSpring=true
    bOverrideAngularDamping=true
    bUseGravityOverride=true
    gravityScale=0.35
    simulationSpace=RootRelative
    boundBone {boneName=dyn_tshirt_rear_r_02}
    componentAppliedLinearAccClamp=<2500 2500 2500>
    componentLinearAccScale=<0.5 0.5 0.5>
    componentLinearVelScale=<0.5 0.5 0.5>
    gravityOverride {z=-980}
    physicsBodyDefinitions [{
      collisionType=CustomSphere
      sphereCollisionRadius=2
      boundBone {boneName=dyn_tshirt_rear_r_02}
      boxExtents=<16 16 16>
      constraintSetup {angularLimitsMax=<0 0 0>, angularLimitsMin=<0 0 -6>, angularTarget=<0 0 0>}
      localJointOffset=<-4 0.2 -0.1>
     }]
   }
   n47 = Anim Dynamics - Bone: dyn_tshirt_rear_01!(
    ComponentPose=@n46
    Alpha=1
    GravityOverride:Vector="0,0,-980"
    SimSpaceSettings:Anim Phys Sim Space Settings Structure=(SimSpaceAngularAlpha=0,MaxAngularVelocity=10000,MaxAngularAcceleration=10000,ExternalAngularVelocity=(X=0,Y=0,Z=0))
   ) -> (Pose) {
    angularDampingOverride=0.9999
    angularSpringConstant=800
    bAngularSpring=true
    bOverrideAngularDamping=true
    bUseGravityOverride=true
    gravityScale=0.35
    simulationSpace=RootRelative
    boundBone {boneName=dyn_tshirt_rear_01}
    componentAppliedLinearAccClamp=<2500 2500 2500>
    componentLinearAccScale=<0.5 0.5 0.5>
    componentLinearVelScale=<0.5 0.5 0.5>
    gravityOverride {z=-980}
    physicsBodyDefinitions [{
      collisionType=CustomSphere
      sphereCollisionRadius=2
      boundBone {boneName=dyn_tshirt_rear_01}
      boxExtents=<45 45 45>
      constraintSetup {angularLimitsMax=<0 5 0>, angularLimitsMin=<0 -5 -15>, angularTarget=<0 0 0>}
      localJointOffset=<-4 0.2 -0.1>
     }]
   }
   n48 = Anim Dynamics - Bone: dyn_tshirt_rear_02!(
    ComponentPose=@n47
    Alpha=1
    GravityOverride:Vector="0,0,-980"
    SimSpaceSettings:Anim Phys Sim Space Settings Structure=(SimSpaceAngularAlpha=0,MaxAngularVelocity=10000,MaxAngularAcceleration=10000,ExternalAngularVelocity=(X=0,Y=0,Z=0))
   ) -> (Pose) {
    angularDampingOverride=0.9999
    angularSpringConstant=800
    bAngularSpring=true
    bOverrideAngularDamping=true
    bUseGravityOverride=true
    gravityScale=0.35
    simulationSpace=RootRelative
    boundBone {boneName=dyn_tshirt_rear_02}
    componentAppliedLinearAccClamp=<2500 2500 2500>
    componentLinearAccScale=<0.5 0.5 0.5>
    componentLinearVelScale=<0.5 0.5 0.5>
    gravityOverride {z=-980}
    physicsBodyDefinitions [{
      collisionType=CustomSphere
      sphereCollisionRadius=2
      boundBone {boneName=dyn_tshirt_rear_02}
      boxExtents=<25 25 25>
      constraintSetup {angularLimitsMax=<0 5 0>, angularLimitsMin=<0 -5 -15>, angularTarget=<0 0 0>}
      localJointOffset=<-4 0.2 -0.1>
     }]
   }
   n49 = "Transform (Modify) Bone - Bone: dyn_tshirt_side_r_01"!(ComponentPose=@n48, bAlphaBoolEnabled:bool=True, Translation="0,-0.6,0") -> (Pose) {
    alphaInputType=Bool
    translationMode=BMM_Additive
    translationSpace=BCS_BoneSpace
    alphaBoolBlend {blendInTime=0.5, blendOutTime=0.5}
    boneToModify {boneName=dyn_tshirt_side_r_01}
    translation {y=-0.6}
   }
   n50 = "Transform (Modify) Bone - Bone: dyn_tshirt_side_r_02"!(ComponentPose=@n49, bAlphaBoolEnabled:bool=True, Translation="0,-0.4,0") -> (Pose) {
    alphaInputType=Bool
    translationMode=BMM_Additive
    translationSpace=BCS_BoneSpace
    alphaBoolBlend {blendInTime=0.5, blendOutTime=0.5}
    boneToModify {boneName=dyn_tshirt_side_r_02}
    translation {y=-0.4}
   }
   n51 = Anim Dynamics - Bone: dyn_TopSleeve_l(ComponentPose=@n50, Alpha=0.8, ExternalForce="0,0,0") {
    angularDampingOverride=1
    angularSpringConstant=6000
    bAngularSpring=true
    bOverrideAngularDamping=true
    bUseGravityOverride=true
    bUsePlanarLimit=false
    gravityScale=0.35
    simulationSpace=RootRelative
    boundBone {boneName=dyn_TopSleeve_l}
    componentAppliedLinearAccClamp=<10000 10000 10000>
    componentLinearAccScale=<1 1 1>
    componentLinearVelScale=<1 1 1>
    gravityOverride {z=-980}
    physicsBodyDefinitions [{
      boundBone {boneName=dyn_TopSleeve_l}
      boxExtents=<50 50 50>
      constraintSetup {angularLimitsMax=<0 3 0>, angularLimitsMin=<0 -3 -5>, angularTarget=<1 0 0>}
      localJointOffset=<5 0 0>
     }]
    relativeSpaceBone {boneName=pelvis}
   }
   n52 = Anim Dynamics - Bone: dyn_underSleeve_l(ComponentPose=@n51, Alpha=0.7, ExternalForce="0,0,0") {
    angularDampingOverride=1
    angularSpringConstant=6000
    bAngularSpring=true
    bOverrideAngularDamping=true
    bUseGravityOverride=true
    bUsePlanarLimit=false
    gravityScale=0.35
    simulationSpace=RootRelative
    boundBone {boneName=dyn_underSleeve_l}
    componentAppliedLinearAccClamp=<10000 10000 10000>
    componentLinearAccScale=<1 1 1>
    componentLinearVelScale=<1 1 1>
    gravityOverride {z=-980}
    physicsBodyDefinitions [{
      boundBone {boneName=dyn_underSleeve_l}
      boxExtents=<50 50 50>
      constraintSetup {angularLimitsMax=<0 3 0>, angularLimitsMin=<0 -3 -5>, angularTarget=<1 0 0>}
      localJointOffset=<5 0 0>
     }]
    relativeSpaceBone {boneName=pelvis}
   }
   n53 = Anim Dynamics - Bone: dyn_frontSleeve_l(ComponentPose=@n52, Alpha=0.7, ExternalForce="0,0,0") {
    angularDampingOverride=1
    angularSpringConstant=6000
    bAngularSpring=true
    bOverrideAngularDamping=true
    bUseGravityOverride=true
    bUsePlanarLimit=false
    gravityScale=0.35
    simulationSpace=RootRelative
    boundBone {boneName=dyn_frontSleeve_l}
    componentAppliedLinearAccClamp=<10000 10000 10000>
    componentLinearAccScale=<1 1 1>
    componentLinearVelScale=<1 1 1>
    gravityOverride {z=-980}
    physicsBodyDefinitions [{
      boundBone {boneName=dyn_frontSleeve_l}
      boxExtents=<50 50 50>
      constraintSetup {angularLimitsMax=<0 3 0>, angularLimitsMin=<0 -3 -5>, angularTarget=<1 0 0>}
      localJointOffset=<5 0 0>
     }]
    relativeSpaceBone {boneName=pelvis}
   }
   n54 = Anim Dynamics - Bone: dyn_backSleeve_l(ComponentPose=@n53, Alpha=0.8, ExternalForce="0,0,0") {
    angularDampingOverride=1
    angularSpringConstant=6000
    bAngularSpring=true
    bOverrideAngularDamping=true
    bUseGravityOverride=true
    bUsePlanarLimit=false
    gravityScale=0.35
    simulationSpace=RootRelative
    boundBone {boneName=dyn_backSleeve_l}
    componentAppliedLinearAccClamp=<10000 10000 10000>
    componentLinearAccScale=<1 1 1>
    componentLinearVelScale=<1 1 1>
    gravityOverride {z=-980}
    physicsBodyDefinitions [{
      boundBone {boneName=dyn_backSleeve_l}
      boxExtents=<50 50 50>
      constraintSetup {angularLimitsMax=<0 3 0>, angularLimitsMin=<0 -3 -5>, angularTarget=<1 0 0>}
      localJointOffset=<5 0 0>
     }]
    relativeSpaceBone {boneName=pelvis}
   }
   n55 = Anim Dynamics - Bone: dyn_TopSleeve_r(ComponentPose=@n54, Alpha=0.8, ExternalForce="0,0,0") {
    angularDampingOverride=1
    angularSpringConstant=6000
    bAngularSpring=true
    bOverrideAngularDamping=true
    bUseGravityOverride=true
    bUsePlanarLimit=false
    gravityScale=0.35
    simulationSpace=RootRelative
    boundBone {boneName=dyn_TopSleeve_r}
    componentAppliedLinearAccClamp=<10000 10000 10000>
    componentLinearAccScale=<1 1 1>
    componentLinearVelScale=<1 1 1>
    gravityOverride {z=-980}
    physicsBodyDefinitions [{
      boundBone {boneName=dyn_TopSleeve_r}
      boxExtents=<50 50 50>
      constraintSetup {angularLimitsMax=<0 3 0>, angularLimitsMin=<0 -3 -5>, angularTarget=<1 0 0>}
      localJointOffset=<5 0 0>
     }]
    relativeSpaceBone {boneName=pelvis}
    sphericalLimits [{limitRadius=8, drivingBone {boneName=upperarm_twist_01_r}, sphereLocalOffset=<-15 0 0>}]
   }
   n56 = Anim Dynamics - Bone: dyn_frontSleeve_r(ComponentPose=@n55, Alpha=0.6, ExternalForce="0,0,0") {
    angularDampingOverride=1
    angularSpringConstant=6000
    bAngularSpring=true
    bOverrideAngularDamping=true
    bUseGravityOverride=true
    bUsePlanarLimit=false
    gravityScale=0.35
    simulationSpace=RootRelative
    boundBone {boneName=dyn_frontSleeve_r}
    componentAppliedLinearAccClamp=<10000 10000 10000>
    componentLinearAccScale=<1 1 1>
    componentLinearVelScale=<1 1 1>
    gravityOverride {z=-980}
    physicsBodyDefinitions [{
      boundBone {boneName=dyn_frontSleeve_r}
      boxExtents=<50 50 50>
      constraintSetup {angularLimitsMax=<0 3 0>, angularLimitsMin=<0 -3 -5>, angularTarget=<1 0 0>}
      localJointOffset=<5 0 0>
     }]
    relativeSpaceBone {boneName=pelvis}
    sphericalLimits [{limitRadius=8, drivingBone {boneName=upperarm_twist_01_r}, sphereLocalOffset=<-15 0 0>}]
   }
   n57 = Anim Dynamics - Bone: dyn_underSleeve_r(ComponentPose=@n56, Alpha=0.7, ExternalForce="0,0,0") {
    angularDampingOverride=1
    angularSpringConstant=6000
    bAngularSpring=true
    bOverrideAngularDamping=true
    bUseGravityOverride=true
    bUsePlanarLimit=false
    gravityScale=0.35
    simulationSpace=RootRelative
    boundBone {boneName=dyn_underSleeve_r}
    componentAppliedLinearAccClamp=<10000 10000 10000>
    componentLinearAccScale=<1 1 1>
    componentLinearVelScale=<1 1 1>
    gravityOverride {z=-980}
    physicsBodyDefinitions [{
      boundBone {boneName=dyn_underSleeve_r}
      boxExtents=<50 50 50>
      constraintSetup {angularLimitsMax=<0 3 0>, angularLimitsMin=<0 -3 -5>, angularTarget=<1 0 0>}
      localJointOffset=<5 0 0>
     }]
    relativeSpaceBone {boneName=pelvis}
    sphericalLimits [{limitRadius=8, drivingBone {boneName=upperarm_twist_01_r}, sphereLocalOffset=<-15 0 0>}]
   }
   n58 = Anim Dynamics - Bone: dyn_backSleeve_r(ComponentPose=@n57, Alpha=0.8, ExternalForce="0,0,0") {
    angularDampingOverride=1
    angularSpringConstant=6000
    bAngularSpring=true
    bOverrideAngularDamping=true
    bUseGravityOverride=true
    bUsePlanarLimit=false
    gravityScale=0.35
    simulationSpace=RootRelative
    boundBone {boneName=dyn_backSleeve_r}
    componentAppliedLinearAccClamp=<10000 10000 10000>
    componentLinearAccScale=<1 1 1>
    componentLinearVelScale=<1 1 1>
    gravityOverride {z=-980}
    physicsBodyDefinitions [{
      boundBone {boneName=dyn_backSleeve_r}
      boxExtents=<50 50 50>
      constraintSetup {angularLimitsMax=<0 3 0>, angularLimitsMin=<0 -3 -5>, angularTarget=<1 0 0>}
      localJointOffset=<5 0 0>
     }]
    relativeSpaceBone {boneName=pelvis}
    sphericalLimits [{limitRadius=8, drivingBone {boneName=upperarm_twist_01_r}, sphereLocalOffset=<-15 0 0>}]
   }
   Save cached pose 'CR_TShirtDyn'(Pose=Component To Local(ComponentPose=@n58)) {cacheName=CR_TShirtDyn}
   n59 = Anim Dynamics - Bone: dyn_aboveSleeve_l_01!(SimulationSpace:AnimPhysSimSpaceType Enum, ComponentPose, Alpha=1, GravityOverride:Vector="0,0,-980", ExternalForce="0,0,0") -> (Pose) {
    angularDampingOverride=0.9999
    angularSpringConstant=800
    bAngularSpring=true
    bOverrideAngularDamping=true
    bUseGravityOverride=true
    bUseSphericalLimits=true
    gravityScale=0.35
    simulationSpace=RootRelative
    boundBone {boneName=dyn_aboveSleeve_l_01}
    componentAppliedLinearAccClamp=<4000 4000 4000>
    componentLinearAccScale=<0.5 0.5 0.5>
    componentLinearVelScale=<0.5 0.5 0.5>
    gravityOverride {z=-980}
    physicsBodyDefinitions [{
      boundBone {boneName=dyn_aboveSleeve_l_01}
      boxExtents=<20 25 25>
      constraintSetup {angularLimitsMax=<0 4 0>, angularLimitsMin=<0 -4 -5>, angularTarget=<1 0 0>}
      localJointOffset=<4 0 0>
     }]
    relativeSpaceBone {boneName=upperarm_l}
   }
   n60 = Anim Dynamics - Bone: dyn_sleeveScoot_l_01!(SimulationSpace:AnimPhysSimSpaceType Enum, ComponentPose=@n59, Alpha=1, GravityOverride:Vector="0,0,-980", ExternalForce="0,0,0") -> (Pose) {
    angularDampingOverride=0.9999
    angularSpringConstant=800
    bAngularSpring=true
    bOverrideAngularDamping=true
    bUseGravityOverride=true
    bUseSphericalLimits=true
    gravityScale=0.35
    simulationSpace=RootRelative
    boundBone {boneName=dyn_sleeveScoot_l_01}
    componentAppliedLinearAccClamp=<4000 4000 4000>
    componentLinearAccScale=<0.5 0.5 0.5>
    componentLinearVelScale=<0.5 0.5 0.5>
    gravityOverride {z=-980}
    physicsBodyDefinitions [{
      boundBone {boneName=dyn_sleeveScoot_l_01}
      boxExtents=<20 25 25>
      constraintSetup {angularLimitsMax=<0 5 5>, angularLimitsMin=<0 -5 -5>, angularTarget=<1 0 0>}
      localJointOffset=<4 0 0>
     }]
    relativeSpaceBone {boneName=upperarm_l}
   }
   n61 = Anim Dynamics - Bone: dyn_backSleeve_l_02!(SimulationSpace:AnimPhysSimSpaceType Enum, ComponentPose=@n60, Alpha=1, GravityOverride:Vector="0,0,-980", ExternalForce="0,0,0") -> (Pose) {
    angularDampingOverride=0.9999
    angularSpringConstant=800
    bAngularSpring=true
    bOverrideAngularDamping=true
    bUseGravityOverride=true
    bUseSphericalLimits=true
    gravityScale=0.35
    simulationSpace=RootRelative
    boundBone {boneName=dyn_backSleeve_l_02}
    componentAppliedLinearAccClamp=<4000 4000 4000>
    componentLinearAccScale=<0.5 0.5 0.5>
    componentLinearVelScale=<0.5 0.5 0.5>
    gravityOverride {z=-980}
    physicsBodyDefinitions [{
      boundBone {boneName=dyn_backSleeve_l_02}
      boxExtents=<20 25 25>
      constraintSetup {angularLimitsMax=<0 5 4>, angularLimitsMin=<0 0 -4>, angularTarget=<1 0 0>}
      localJointOffset=<4 0 0>
     }]
    relativeSpaceBone {boneName=upperarm_l}
   }
   n62 = Anim Dynamics - Bone: dyn_backSleeve_l_01!(SimulationSpace:AnimPhysSimSpaceType Enum, ComponentPose=@n61, Alpha=1, GravityOverride:Vector="0,0,-980", ExternalForce="0,0,0") -> (Pose) {
    angularDampingOverride=0.9999
    angularSpringConstant=800
    bAngularSpring=true
    bOverrideAngularDamping=true
    bUseGravityOverride=true
    bUseSphericalLimits=true
    gravityScale=0.35
    simulationSpace=RootRelative
    boundBone {boneName=dyn_backSleeve_l_01}
    componentAppliedLinearAccClamp=<4000 4000 4000>
    componentLinearAccScale=<0.5 0.5 0.5>
    componentLinearVelScale=<0.5 0.5 0.5>
    gravityOverride {z=-980}
    physicsBodyDefinitions [{
      boundBone {boneName=dyn_backSleeve_l_01}
      boxExtents=<20 25 25>
      constraintSetup {angularLimitsMax=<0 5 4>, angularLimitsMin=<0 0 -4>, angularTarget=<1 0 0>}
      localJointOffset=<4 0 0>
     }]
    relativeSpaceBone {boneName=upperarm_l}
   }
   n63 = Anim Dynamics - Bone: dyn_underSleeve_l_02!(SimulationSpace:AnimPhysSimSpaceType Enum, ComponentPose=@n62, Alpha=1, GravityOverride:Vector="0,0,-980", ExternalForce="0,0,0") -> (Pose) {
    angularDampingOverride=0.9999
    angularSpringConstant=800
    bAngularSpring=true
    bOverrideAngularDamping=true
    bUseGravityOverride=true
    bUseSphericalLimits=true
    gravityScale=0.35
    simulationSpace=RootRelative
    boundBone {boneName=dyn_underSleeve_l_02}
    componentAppliedLinearAccClamp=<4000 4000 4000>
    componentLinearAccScale=<0.5 0.5 0.5>
    componentLinearVelScale=<0.5 0.5 0.5>
    gravityOverride {z=-980}
    physicsBodyDefinitions [{
      boundBone {boneName=dyn_underSleeve_l_02}
      boxExtents=<20 25 25>
      constraintSetup {angularLimitsMax=<0 5 5>, angularLimitsMin=<0 -5 -5>, angularTarget=<1 0 0>}
      localJointOffset=<4 0 0>
     }]
    relativeSpaceBone {boneName=upperarm_l}
   }
   n64 = Anim Dynamics - Bone: dyn_underSleeve_l_01!(SimulationSpace:AnimPhysSimSpaceType Enum, ComponentPose=@n63, Alpha=1, GravityOverride:Vector="0,0,-980", ExternalForce="0,0,0") -> (Pose) {
    angularDampingOverride=0.9999
    angularSpringConstant=800
    bAngularSpring=true
    bOverrideAngularDamping=true
    bUseGravityOverride=true
    bUseSphericalLimits=true
    gravityScale=0.35
    simulationSpace=RootRelative
    boundBone {boneName=dyn_underSleeve_l_01}
    componentAppliedLinearAccClamp=<4000 4000 4000>
    componentLinearAccScale=<0.5 0.5 0.5>
    componentLinearVelScale=<0.5 0.5 0.5>
    gravityOverride {z=-980}
    physicsBodyDefinitions [{
      boundBone {boneName=dyn_underSleeve_l_01}
      boxExtents=<20 25 25>
      constraintSetup {angularLimitsMax=<0 5 5>, angularLimitsMin=<0 -5 -5>, angularTarget=<1 0 0>}
      localJointOffset=<4 0 0>
     }]
    relativeSpaceBone {boneName=upperarm_l}
   }
   n65 = Anim Dynamics - Bone: dyn_aboveSleeve_l_02!(SimulationSpace:AnimPhysSimSpaceType Enum, ComponentPose=@n64, Alpha=1, GravityOverride:Vector="0,0,-980", ExternalForce="0,0,0") -> (Pose) {
    angularDampingOverride=0.9999
    angularSpringConstant=800
    bAngularSpring=true
    bOverrideAngularDamping=true
    bUseGravityOverride=true
    bUseSphericalLimits=true
    gravityScale=0.35
    simulationSpace=RootRelative
    boundBone {boneName=dyn_aboveSleeve_l_02}
    componentAppliedLinearAccClamp=<4000 4000 4000>
    componentLinearAccScale=<0.5 0.5 0.5>
    componentLinearVelScale=<0.5 0.5 0.5>
    gravityOverride {z=-980}
    physicsBodyDefinitions [{
      boundBone {boneName=dyn_aboveSleeve_l_02}
      boxExtents=<20 25 25>
      constraintSetup {angularLimitsMax=<0 4 0>, angularLimitsMin=<0 -4 -5>, angularTarget=<1 0 0>}
      localJointOffset=<4 0 0>
     }]
    relativeSpaceBone {boneName=upperarm_l}
   }
   n66 = Anim Dynamics - Bone: dyn_aboveSleeve_r_01!(SimulationSpace:AnimPhysSimSpaceType Enum, ComponentPose=@n65, Alpha=1, GravityOverride:Vector="0,0,-980", ExternalForce="0,0,0") -> (Pose) {
    angularDampingOverride=0.9999
    angularSpringConstant=800
    bAngularSpring=true
    bOverrideAngularDamping=true
    bUseGravityOverride=true
    gravityScale=0.35
    simulationSpace=RootRelative
    boundBone {boneName=dyn_aboveSleeve_r_01}
    componentAppliedLinearAccClamp=<4000 4000 4000>
    componentLinearAccScale=<0.5 0.5 0.5>
    componentLinearVelScale=<0.5 0.5 0.5>
    gravityOverride {z=-980}
    physicsBodyDefinitions [{
      boundBone {boneName=dyn_aboveSleeve_r_01}
      boxExtents=<20 25 25>
      constraintSetup {angularLimitsMax=<0 4 0>, angularLimitsMin=<0 -4 -5>, angularTarget=<1 0 0>}
      localJointOffset=<2 0 0>
     }]
    relativeSpaceBone {boneName=upperarm_r}
   }
   n67 = Anim Dynamics - Bone: dyn_sleeveScoot_r_01!(SimulationSpace:AnimPhysSimSpaceType Enum, ComponentPose=@n66, Alpha=1, GravityOverride:Vector="0,0,-980", ExternalForce="0,0,0") -> (Pose) {
    angularDampingOverride=0.9999
    angularSpringConstant=800
    bAngularSpring=true
    bOverrideAngularDamping=true
    bUseGravityOverride=true
    gravityScale=0.35
    simulationSpace=RootRelative
    boundBone {boneName=dyn_sleeveScoot_r_01}
    componentAppliedLinearAccClamp=<4000 4000 4000>
    componentLinearAccScale=<0.5 0.5 0.5>
    componentLinearVelScale=<0.5 0.5 0.5>
    gravityOverride {z=-980}
    physicsBodyDefinitions [{
      boundBone {boneName=dyn_sleeveScoot_r_01}
      boxExtents=<20 25 25>
      constraintSetup {angularLimitsMax=<0 5 5>, angularLimitsMin=<0 -5 -5>, angularTarget=<1 0 0>}
      localJointOffset=<4 0 0>
     }]
    relativeSpaceBone {boneName=upperarm_r}
   }
   n68 = Anim Dynamics - Bone: dyn_backSleeve_r_02!(SimulationSpace:AnimPhysSimSpaceType Enum, ComponentPose=@n67, Alpha=1, GravityOverride:Vector="0,0,-980", ExternalForce="0,0,0") -> (Pose) {
    angularDampingOverride=0.9999
    angularSpringConstant=800
    bAngularSpring=true
    bOverrideAngularDamping=true
    bUseGravityOverride=true
    gravityScale=0.35
    simulationSpace=RootRelative
    boundBone {boneName=dyn_backSleeve_r_02}
    componentAppliedLinearAccClamp=<4000 4000 4000>
    componentLinearAccScale=<0.5 0.5 0.5>
    componentLinearVelScale=<0.5 0.5 0.5>
    gravityOverride {z=-980}
    physicsBodyDefinitions [{
      boundBone {boneName=dyn_backSleeve_r_02}
      boxExtents=<20 25 25>
      constraintSetup {angularLimitsMax=<0 0 4>, angularLimitsMin=<0 -5 -4>, angularTarget=<1 0 0>}
      localJointOffset=<4 0 0>
     }]
    relativeSpaceBone {boneName=upperarm_r}
   }
   n69 = Anim Dynamics - Bone: dyn_backSleeve_r_01!(SimulationSpace:AnimPhysSimSpaceType Enum, ComponentPose=@n68, Alpha=1, GravityOverride:Vector="0,0,-980", ExternalForce="0,0,0") -> (Pose) {
    angularDampingOverride=0.9999
    angularSpringConstant=800
    bAngularSpring=true
    bOverrideAngularDamping=true
    bUseGravityOverride=true
    gravityScale=0.35
    simulationSpace=RootRelative
    boundBone {boneName=dyn_backSleeve_r_01}
    componentAppliedLinearAccClamp=<4000 4000 4000>
    componentLinearAccScale=<0.5 0.5 0.5>
    componentLinearVelScale=<0.5 0.5 0.5>
    gravityOverride {z=-980}
    physicsBodyDefinitions [{
      boundBone {boneName=dyn_backSleeve_r_01}
      boxExtents=<20 25 25>
      constraintSetup {angularLimitsMax=<0 0 4>, angularLimitsMin=<0 -5 -4>, angularTarget=<1 0 0>}
      localJointOffset=<4 0 0>
     }]
    relativeSpaceBone {boneName=upperarm_r}
   }
   n70 = Anim Dynamics - Bone: dyn_underSleeve_r_02!(SimulationSpace:AnimPhysSimSpaceType Enum, ComponentPose=@n69, Alpha=1, GravityOverride:Vector="0,0,-980", ExternalForce="0,0,0") -> (Pose) {
    angularDampingOverride=0.9999
    angularSpringConstant=800
    bAngularSpring=true
    bOverrideAngularDamping=true
    bUseGravityOverride=true
    gravityScale=0.35
    simulationSpace=RootRelative
    boundBone {boneName=dyn_underSleeve_r_02}
    componentAppliedLinearAccClamp=<4000 4000 4000>
    componentLinearAccScale=<0.5 0.5 0.5>
    componentLinearVelScale=<0.5 0.5 0.5>
    gravityOverride {z=-980}
    physicsBodyDefinitions [{
      boundBone {boneName=dyn_underSleeve_r_02}
      boxExtents=<20 25 25>
      constraintSetup {angularLimitsMax=<0 5 5>, angularLimitsMin=<0 -5 -5>, angularTarget=<1 0 0>}
      localJointOffset=<4 0 0>
     }]
    relativeSpaceBone {boneName=upperarm_r}
   }
   n71 = Anim Dynamics - Bone: dyn_underSleeve_r_01!(SimulationSpace:AnimPhysSimSpaceType Enum, ComponentPose=@n70, Alpha=1, GravityOverride:Vector="0,0,-980", ExternalForce="0,0,0") -> (Pose) {
    angularDampingOverride=0.9999
    angularSpringConstant=800
    bAngularSpring=true
    bOverrideAngularDamping=true
    bUseGravityOverride=true
    gravityScale=0.35
    simulationSpace=RootRelative
    boundBone {boneName=dyn_underSleeve_r_01}
    componentAppliedLinearAccClamp=<4000 4000 4000>
    componentLinearAccScale=<0.5 0.5 0.5>
    componentLinearVelScale=<0.5 0.5 0.5>
    gravityOverride {z=-980}
    physicsBodyDefinitions [{
      boundBone {boneName=dyn_underSleeve_r_01}
      boxExtents=<20 25 25>
      constraintSetup {angularLimitsMax=<0 5 5>, angularLimitsMin=<0 -5 -5>, angularTarget=<1 0 0>}
      localJointOffset=<4 0 0>
     }]
    relativeSpaceBone {boneName=upperarm_r}
   }
   n72 = Anim Dynamics - Bone: dyn_aboveSleeve_r_02!(SimulationSpace:AnimPhysSimSpaceType Enum, ComponentPose=@n71, Alpha=1, GravityOverride:Vector="0,0,-980", ExternalForce="0,0,0") -> (Pose) {
    angularDampingOverride=0.9999
    angularSpringConstant=800
    bAngularSpring=true
    bOverrideAngularDamping=true
    bUseGravityOverride=true
    gravityScale=0.35
    simulationSpace=RootRelative
    boundBone {boneName=dyn_aboveSleeve_r_02}
    componentAppliedLinearAccClamp=<4000 4000 4000>
    componentLinearAccScale=<0.5 0.5 0.5>
    componentLinearVelScale=<0.5 0.5 0.5>
    gravityOverride {z=-980}
    physicsBodyDefinitions [{
      boundBone {boneName=dyn_aboveSleeve_r_02}
      boxExtents=<20 25 25>
      constraintSetup {angularLimitsMax=<0 4 0>, angularLimitsMin=<0 -4 -5>, angularTarget=<1 0 0>}
      localJointOffset=<4 0 0>
     }]
    relativeSpaceBone {boneName=upperarm_r}
   }
   Reroute Node(InputPin:Component Space Pose Link Structure=@n72) -> (OutputPin:Component Space Pose Link Structure)
  ]}
 EventGraph {nodes [
   n1 = Event BlueprintUpdateAnimation()
   > Set Delta Time X(Delta Time X=@n1.DeltaTimeX)
   > n2 = Sequence!(execute) -> (then_0, then_1, then_2:Exec, then_3:Exec, then_4:Exec)
   n3 = GetOwnerSkeletalMeshComponent()
   n4 = BreakRotator(InRot=BreakTransform(InTransform=GetDeltaTransformFromRefPose(self=@n3, BoneName=spine_03, BaseName=pelvis)).Rotation)
   Set spine_rot_r(execute=@n2.then_1, spine_rot_r:double="To Float (Boolean)"(InBool="float <= float"(A=@n4, B=-10) -> (ReturnValue=~))) -> (Output_Get:double)
   n5 = GetOwnerSkeletalMeshComponent()
   n6 = BreakRotator(InRot=BreakTransform(InTransform=GetDeltaTransformFromRefPose(self=@n5, BoneName=spine_03, BaseName=pelvis)).Rotation)
   > Set spine_rot_L(spine_rot_L:double="To Float (Boolean)"(InBool="float >= float"(A=@n6, B=-10))) -> (Output_Get:double)
   n7 = GetAnimInstance(self=@n3)
   > Is Valid(InputObject=@n7)
   > Set isCustomEmotePlaying(isCustomEmotePlaying:bool=Montage_IsPlaying(self=@n7, Montage=/BRCosmetics/Animation/Game/MainPlayer/Emotes/YouBoreMe/Emote_You_Bore_Me_CMF_M)) -> (Output_Get:bool)
   n8 = Montage_IsPlaying(self=@n7, Montage=/BRCosmetics/Animation/Game/MainPlayer/Emotes/AprilBevie/Emote_AprilBevie_CMF_Player2_M)
   n9 = Montage_IsPlaying(self=@n7, Montage=/BRCosmetics/Animation/Game/MainPlayer/Emotes/AprilBevie/Emote_AprilBevie_CMF_Player1_M)
   n10 = Get bIsPlayingEmote()
   > "Set PlayingEmoteMontagedouble "("PlayingEmoteMontagedouble ":bool=OR Boolean!(A=@n9, B=@n10, C:bool=@n8) -> (ReturnValue)) -> (Output_Get:bool)
   > Set noPostFixes(noPostFixes:bool=NOT Boolean(A=@n10)) -> (Output_Get:bool)
   > Set EmoteAprilFix(EmoteAprilFix:bool=OR Boolean!(A=@n9, B=@n8) -> (ReturnValue)) -> (Output_Get:bool)
   n11 = GetAnimInstance(self=GetOwnerSkeletalMeshComponent())
   > Is Valid(InputObject=@n11)
   > Set IsplayingIdelFronEnd(IsplayingIdelFronEnd:bool=Montage_IsPlaying(self=@n11, Montage=/BRCosmetics/Animation/Game/MainPlayer/CharIntros/s39/MarineCarve/CharIntro_MarineCarve_M)) -> (Output_Get:bool)
   n12 = GetCurrentSkinWeightProfileName!(self:Skinned Mesh Component Object Reference=GetPartSkeletalMeshComponent(), InLayer:ESkinWeightProfileLayer Enum=Primary) -> (ReturnValue:Name=None)
   Set ASWOn(execute=@n2.then_4, ASWOn:bool="Not Equal (Name)"(A=@n12) -> (ReturnValue=~)) -> (Output_Get:bool)
   > Set KneeFix(KneeFix:bool=OR Boolean(A=Get bIsSkydiveDiveMode(), B=Get bIsSkydiveFloating())) -> (Output_Get:bool)
   n13 = PlayerMovement(execute=@n2.then_2, AnimInstance=Self-Reference(), XDotProductCoefficient=1, YDotProductCoefficient=1, ZDotProductCoefficient=1)
   n14 = BreakVector(InVec=MakeVector(X=@n13, Y=@n13.YMovement))
   > "Set Move Left "("Move Left ":double="Absolute (Float)"(A="Clamp (Float)"(Value=@n14, Min=-1, Max=0))) -> (Output_Get:double)
   > Set Move Right(Move Right:double="Clamp (Float)"(Value=@n14, Min=0, Max=1)) -> (Output_Get:double)
   > Set moveForward(moveForward:double="Clamp (Float)"(Value=@n14.Y, Min=0, Max=1)) -> (Output_Get:double)
   > Set moveBackward(moveBackward:double="Absolute (Float)"(A="Clamp (Float)"(Value=@n14.Y, Min=-1, Max=0))) -> (Output_Get:double)
   Sequence!(execute=@n2.then_3) -> (then_0, then_1)
   n15 = Get bRigidBodyEnabled()
   n16 = Get CurrentLOD()
   > n17 = Branch(Condition=OR Boolean(A=NOT Boolean(A=@n15), B="integer > integer"(A=@n16, B=1)))
   > n18 = Sequence()
   > Do Once(Reset=@n20.then_1)
   n19 = GetPartSkeletalMeshComponent()
   > SetSkinWeightProfile!(execute, self:Skinned Mesh Component Object Reference=@n19, InProfileName:Name=Alternative, InLayer:ESkinWeightProfileLayer Enum=Secondary) -> (then, ReturnValue:bool)
   Do Once(execute=@n20.then_0, Reset=@n18.then_1)
   > ClearSkinWeightProfile!(execute, self:Skinned Mesh Component Object Reference=@n19, InLayer:ESkinWeightProfileLayer Enum=Secondary) -> (then)
   n20 = Sequence(execute=@n17.else)
   n21 = "Gravity Override State Function (Parameter Asset) (Nativized)"(
    execute=@n2.then_0
    CharacterPartAnimInstance=Self-Reference()
    GameTimeInSeconds=GetGameTimeInSeconds()
    Parameters=Get Parameters() -> (Parameters:Fort Character Dynamics Parameters Object Reference)
   )
   > Set GravityOverride(GravityOverride=@n21)
   n22 = Get bIsCrouching()
   n23 = Get Delta Time X()
   > Set isCrouching float(isCrouching float:double=FInterpTo(Current=Get isCrouching float(), Target="To Float (Boolean)"(InBool=@n22), DeltaTime=@n23, InterpSpeed=15)) -> (Output_Get:double)
   n24 = Get bIsTargeting()
   > Set isTargeting float(isTargeting float:double=FInterpTo(Current=Get isTargeting float() -> (isTargeting float:double), Target="To Float (Boolean)"(InBool=@n24), DeltaTime=@n23, InterpSpeed=15)) -> (Output_Get:double)
   n25 = NOT Boolean(A=Get bIsSliding())
   n26 = NOT Boolean(A=Get bIsKneeSliding())
   n27 = Get bIsCrouching()
   n28 = Get bIsMoving2D()
   > Set CrouchMOV(CrouchMOV:double="To Float (Boolean)"(InBool=AND Boolean!(A=@n27, B=@n28, C:bool=@n25, D:bool=@n26) -> (ReturnValue))) -> (Output_Get:double)
   n29 = NOT Boolean(A=Get isGunOut() -> (isGunOut:bool))
   n30 = Get bIsCrouching()
   > Set isCrouchinggunOut(isCrouchinggunOut:double="To Float (Boolean)"(InBool=AND Boolean!(A=@n30, B=@n29) -> (ReturnValue))) -> (Output_Get:double)
   n31 = Get bIsPlayingEmote()
   > Set IsFrontEndNOEmote(IsFrontEndNOEmote:bool=AND Boolean(A=Get bIsInFrontEnd(), B=NOT Boolean(A=@n31))) -> (Output_Get:bool)
   > n32 = Branch(Condition=Get bigGunEquipped() -> (bigGunEquipped:bool))
   > Set bigGunAiming(bigGunAiming:bool=Get bIsTargeting()) -> (Output_Get:bool)
   > Set IsNotInFrontEnd(IsNotInFrontEnd:bool=NOT Boolean(A=Get bIsInFrontEnd())) -> (Output_Get:bool)
   Set bigGunAiming(execute=@n32.else, bigGunAiming=false)
   Event OnBeginJumping()
   > Set Isjumping(Isjumping:bool=1) -> (Output_Get:bool)
   Event OnEndSkydiving()
   > Set startingSkydive(startingSkydive:bool=false) -> (Output_Get:bool)
   Event OnEndFalling()
   > Set Isjumping(Isjumping=false)
   Event OnWeaponChanged()
   > Set isGunOut(isGunOut:bool=1) -> (Output_Get:bool)
   > "Set hasTwoHandedWeapon "("hasTwoHandedWeapon ":bool=false) -> (Output_Get:bool)
   > Set bigGunEquipped(bigGunEquipped:bool=false) -> (Output_Get:bool)
   > n33 = Switch on EFortWeaponCoreAnimation!(execute, Selection:EFortWeaponCoreAnimation Enum=Get WeaponCoreAnim!() -> (WeaponCoreAnim)) -> (
    Melee:Exec
    Pistol:Exec
    Shotgun:Exec
    PaperBlueprint:Exec
    Rifle:Exec
    MeleeOneHand:Exec
    MachinePistol:Exec
    RocketLauncher:Exec
    GrenadeLauncher:Exec
    GoingCommando:Exec
    AssaultRifle:Exec
    TacticalShotgun:Exec
    SniperRifle:Exec
    TrapPlacement:Exec
    ShoulderLauncher:Exec
    AbilityDecoTool:Exec
    Crossbow:Exec
    C4:Exec
    RemoteControl:Exec
    DualWield:Exec
    AR_BullPup:Exec
    AR_ForwardGrip:Exec
    MedPackPaddles:Exec
    SMG_P90:Exec
    AR_DrumGun:Exec
    Consumable_Small:Exec
    Consumable_Large:Exec
    Balloon:Exec
    MountedTurret:Exec
    CreativeTool:Exec
    ExplosiveBow:Exec
    AshtonIndigo:Exec
    AshtonChicago:Exec
    MeleeDualWield:Exec
    Unarmed:Exec
   )
   Set bigGunEquipped(execute=[@n33.AssaultRifle, @n33.Shotgun, @n33.TacticalShotgun], bigGunEquipped=true)
   n34 = "Set hasTwoHandedWeapon "(execute=@n33.Melee, "hasTwoHandedWeapon "=true)
   Set isGunOut(execute=[@n33.MeleeDualWield, @n33.Rifle, @n33.Unarmed, @n34.then], isGunOut=false)
   Event OnBeginSkydiving()
   > Set startingSkydive(startingSkydive=true)
   > Delay(Duration=2.5)
   > Set startingSkydive(startingSkydive=false)
   Event OnBeginCrouching()
   > Set Is Playing MeleeCrounching Anim(Is Playing MeleeCrounching Anim:bool=Get bIsPlayingMeleeAnim()) -> (Output_Get:bool)
   Event OnEndCrouching()
   > Set Is Playing MeleeCrounching Anim(Is Playing MeleeCrounching Anim=false)
   Event OnCharacterCustomizationLoaded()
   > n35 = Sequence()
   > Set bagStartOffset(bagStartOffset:bool=true) -> (Output_Get:bool)
   > Delay(Duration=0.2)
   > Set bagStartOffset(bagStartOffset=false)
   > n36 = Branch(Condition=Get bIsInFrontEnd())
   > Set Sim Space Settings A(Sim Space Settings A:Anim Phys Sim Space Settings Structure=Get Sim Space Settings_FE() -> (Sim Space Settings_FE:Anim Phys Sim Space Settings Structure)) -> (Output_Get:Anim Phys Sim Space Settings Structure)
   > Set TshirtDynAlpha(TshirtDynAlpha:double=0.2) -> (Output_Get:double)
   Set Sim Space Settings A(execute=@n36.else, Sim Space Settings A=Get Sim Space Settings_BR() -> (Sim Space Settings_BR:Anim Phys Sim Space Settings Structure))
   > Set TshirtDynAlpha(TshirtDynAlpha=1)
   Reroute Node(InputPin:Exec=@n35.then_1) -> (OutputPin:Exec)
   "To String (Name)"(InName=@n12)
  ]}
]
variables [
 ASWOn {type=boolean}
 Bag Dynamics Alpha {type=number}
 bagStartOffset {type=boolean, default=true}
 bigGunAiming {type=boolean}
 bigGunEquipped {type=boolean}
 CrouchMOV {type=number}
 CrouchWalk {type=boolean}
 Delta Time X {type=number}
 EmoteAprilFix {type=boolean}
 ForceASW {type=boolean}
 Gravity Override Params {
  type=FortGravityOverrideParameterSet
  default=(Skydive_GoingUp=(GravityOverride=()),Skydive_FloatingDown=(GravityOverride=()),Skydive_DivingDown=(GravityOverride=()),Skydive_Parachuting=(GravityOverride=()),OnGround_Standing=(GravityOverride=()),OnGround_Moving=(GravityOverride=()),OnGround_Moving_HighRotationRate=(GravityOverride=()),OnGround_Crouch_Moving=(GravityOverride=()),Emote_or_Melee=(GravityOverride=()),Swimming=(GravityOverride=()),Front_End=(GravityOverride=()),DBNO=(GravityOverride=()),DBNO_Carried=(GravityOverride=()),OnGround_Falling=(GravityOverride=()),Ziplining=(GravityOverride=()),Skydive_Surfing=(GravityOverride=()),Riding_Standing=(GravityOverride=()),Riding_Moving=(GravityOverride=()),Riding_Moving_Targeting=(GravityOverride=()),Tactical_Sprint=(GravityOverride=()),Rail_Grinding=(GravityOverride=()),)
 }
 GravityOverride {type=Vector}
 "hasTwoHandedWeapon " {type=boolean}
 Is ParachuteOpened {type=number}
 Is Playing Melee Anim {type=number}
 Is Playing MeleeCrounching Anim {type=boolean}
 Is Skydive Floating {type=number}
 isBaseballSlidling float {type=number}
 isCrouching float {type=number}
 isCrouchinggunOut {type=number}
 IsCrouchtargeting {type=number}
 isCustomEmotePlaying {type=boolean}
 IsFrontEndNOEmote {type=boolean}
 isGunOut {type=boolean}
 Isjumping {type=boolean}
 isKneeSlidling float {type=number}
 IsNotInFrontEnd {type=boolean}
 IsplayingIdelFronEnd {type=boolean}
 isTargeting float {type=number}
 KneeFix {type=boolean}
 "Move Left " {type=number}
 Move Right {type=number}
 moveBackward {type=number}
 moveForward {type=number}
 noPostFixes {type=boolean}
 Parameters {type=FortCharacterDynamicsParameters, default=null}
 Playing Emote {type=number}
 "PlayingEmoteMontagedouble " {type=boolean}
 Sim Space Settings A {
  type=AnimPhysSimSpaceSettings
  default {maxAngularAcceleration=10000, maxAngularVelocity=10000, simSpaceAngularAlpha=0, externalAngularVelocity=<0 0 0>}
 }
 Sim Space Settings_BR {
  type=AnimPhysSimSpaceSettings
  default {maxAngularAcceleration=10000, maxAngularVelocity=10000, simSpaceAngularAlpha=0, externalAngularVelocity=<0 0 0>}
 }
 Sim Space Settings_FE {
  type=AnimPhysSimSpaceSettings
  default {maxAngularAcceleration=700, maxAngularVelocity=700, simSpaceAngularAlpha=0.6, externalAngularVelocity=<0.1 0.1 0.1>}
 }
 Sim Space Tshirt {type=AnimPhysSimSpaceType}
 spine_rot_L {type=number}
 spine_rot_r {type=number}
 startingSkydive {type=boolean}
 startingSkydivePareachutenOpen {type=boolean}
 TshirtDynAlpha {type=number, default=1}
]
