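# Control Rig for the MarineCarve female face-accessory part that offsets base hair bones as a corrective pose.
# DeltaRotation v4 splits head and right clavicle rotation relative to reference nulls into swing/twist, then remaps one Euler axis to a clamped weight.
# Both weights multiply with isCrouching to drive a Modify Transforms node, so the correction applies only while crouching.
# A second hair offset weighted by isSkydive exists but is not wired into the execution chain.
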ControlRigBlueprint /BRCosmetics/Characters/Player/Female/Medium/Bodies/F_MED_MarineCarve/Meshes/Parts/F_MED_MarineCarve_FaceAcc_CtrlRig
graphs [
 CollapseNode_ContainedGraph {nodes [
   RigVMFunctionEntryNode()
   > DeltaRotation v4 = RigVMFunctionReferenceNode(
    Rotation Order=ZYX
    child=(Type=Bone,Name="head")
    debugRotation=false
    outputMinMax=(X=0,Y=1)
    parent=(Type=Null,Name="head_null")
    premult=1
    rotationMinMax=(X=0,Y=-34)
    triggerAxis=Z
    twistAxis=(X=1,Y=0,Z=0)
   )
   > DeltaRotation v4_1 = RigVMFunctionReferenceNode(
    Rotation Order=ZYX
    child=(Type=Bone,Name="clavicle_r")
    debugRotation=false
    outputMinMax=(X=0,Y=1)
    parent=(Type=Null,Name="clavicle_r_null")
    premult=1
    rotationMinMax=(X=0,Y=-24)
    triggerAxis=Y
    twistAxis=(X=1,Y=0,Z=0)
   )
   RigVMFunction_MathBoolToFloat = &RigVMFunction_MathBoolToFloat(Value=RigVMVariableNode() {variable=pub isCrouching}.Value)
   Multiply = RigVMAggregateNode(B=@"DeltaRotation v4".outputWeight, A=@"DeltaRotation v4_1".outputWeight, C=@RigVMFunction_MathBoolToFloat.Result)
   > &RigUnit_ModifyTransforms(
    ItemToModify="((Item=(Type=Bone,Name=\"base_hairBackSide_r_00\"),Transform=(Rotation=(X=0,Y=0,Z=0,W=1),Translation=(X=0,Y=0,Z=0),Scale3D=(X=1,Y=1,Z=1))),(Item=(Type=Bone,Name=\"base_hairSide_r_01\"),Transform=(Rotation=(X=0,Y=0.1736,Z=0,W=0.9848),Translation=(X=0,Y=0,Z=0),S ...(truncated)"
    WeightMinimum=0
    Weight=@Multiply.Result
   )
   > RigVMFunctionReturnNode()
   &RigUnit_ModifyTransforms(
    ExecutePin=()
    ItemToModify="((Item=(Type=Bone,Name=\"base_hairBackSide_r_00\"),Transform=(Rotation=(X=0,Y=-0.2588,Z=0,W=0.9659),Translation=(X=0,Y=0,Z=0),Scale3D=(X=1,Y=1,Z=1))),(Item=(Type=Bone,Name=\"base_hairBackSide_l_00\"),Transform=(Rotation=(X=0,Y=-0.2588,Z=0,W=0.9659),Translation ...(truncated)"
    WeightMinimum=0
    Weight=&RigVMFunction_MathBoolToFloat(Value=RigVMVariableNode() {variable=pub isSkydive}.Value).Result
   )
  ]}
 Multiply_1_ContainedGraph {nodes [
   n1 = RigVMFunctionEntryNode()
   Multiply = &RigVMFunction_MathDoubleMul(A=@n1.A, B=@n1.B)
   RigVMFunctionReturnNode(Result=&RigVMFunction_MathDoubleMul(B=@n1.C, A=@Multiply.Result).Result)
  ]}
 New Function_ContainedGraph {nodes [
   n1 = RigVMFunctionEntryNode()
   RigVMFunction_MathBoolToInteger = &RigVMFunction_MathBoolToInteger(Value=@n1.onlyTwist)
   RigVMFunction_MathQuaternionSwingTwist = &RigVMFunction_MathQuaternionSwingTwist(
    TwistAxis=@n1.twistAxis
    Input=&RigUnit_GetRelativeTransformForItem(bChildInitial=False, bParentInitial=False, Child=@n1.child, Parent=@n1.parent).RelativeTransform.Rotation
   )
   n2 = RigVMDispatchNode(
    Values=((),())
    Index=@RigVMFunction_MathBoolToInteger.Result
    Values.0=@RigVMFunction_MathQuaternionSwingTwist.Swing
    Values.1=@RigVMFunction_MathQuaternionSwingTwist.Twist
   ) {template="DISPATCH_RigVMDispatch_SelectInt32(in Index,in Values,out Result)"}
   RigVMFunction_MathQuaternionToEuler = &RigVMFunction_MathQuaternionToEuler(Value=@n2.Result, RotationOrder=@n1.Rotation Order)
   > DISPATCH_RigVMDispatch_Print = RigVMDispatchNode(ScreenColor=(R=1,G=1,B=1,A=1), ScreenDuration=1.5, Enabled=@n1.debugRotation, Value=@RigVMFunction_MathQuaternionToEuler.Result) {template="DISPATCH_RigVMDispatch_Print(in Prefix,in Value,in Enabled,in ScreenDuration,in ScreenColor)"}
   Cast = RigVMDispatchNode(Value=@n1.triggerAxis) {template="DISPATCH_RigVMDispatch_CastEnumToInt(in Value,out Result)"}
   DISPATCH_RigVMDispatch_SelectInt32 = RigVMDispatchNode(
    Values=(0,0,0)
    Index=@Cast.Result
    Values.0=@RigVMFunction_MathQuaternionToEuler.Result.X
    Values.1=@RigVMFunction_MathQuaternionToEuler.Result.Y
    Values.2=@RigVMFunction_MathQuaternionToEuler.Result.Z
   ) {template="DISPATCH_RigVMDispatch_SelectInt32(in Index,in Values,out Result)"}
   Remap = &RigVMFunction_MathDoubleRemap(
    bClamp=False
    Value=@DISPATCH_RigVMDispatch_SelectInt32.Result
    TargetMinimum=@n1.outputMinMax.X
    TargetMaximum=@n1.outputMinMax.Y
    SourceMinimum=@n1.rotationMinMax.X
    SourceMaximum=@n1.rotationMinMax.Y
   )
   Absolute = &RigVMFunction_MathDoubleAbs(Value=@Remap.Result)
   Multiply = &RigVMFunction_MathDoubleMul(
    A=&RigVMFunction_MathDoubleClamp(Value=RigVMDispatchNode(True=@Absolute.Result, Condition=@n1.absoluteValue, False=@Remap.Result) {template="DISPATCH_RigVMDispatch_If(in Condition,in True,in False,out Result)"}.Result).Result
    B=@n1.premult
   )
   > RigVMFunctionReturnNode(outputQuat=@n2.Result, outputWeight=@Multiply.Result)
  ]}
 New Function_ContainedGraph {nodes [
   n1 = RigVMFunctionEntryNode()
   RigVMFunction_MathBoolToInteger = &RigVMFunction_MathBoolToInteger(Value=@n1.onlyTwist)
   RigVMFunction_MathQuaternionSwingTwist = &RigVMFunction_MathQuaternionSwingTwist(
    TwistAxis=@n1.twistAxis
    Input=&RigUnit_GetRelativeTransformForItem(bChildInitial=False, bParentInitial=False, Child=@n1.child, Parent=@n1.parent).RelativeTransform.Rotation
   )
   n2 = RigVMDispatchNode(
    Values=((),())
    Index=@RigVMFunction_MathBoolToInteger.Result
    Values.0=@RigVMFunction_MathQuaternionSwingTwist.Swing
    Values.1=@RigVMFunction_MathQuaternionSwingTwist.Twist
   ) {template="DISPATCH_RigVMDispatch_SelectInt32(in Index,in Values,out Result)"}
   RigVMFunction_MathQuaternionToEuler = &RigVMFunction_MathQuaternionToEuler(Value=@n2.Result, RotationOrder=@n1.Rotation Order)
   > DISPATCH_RigVMDispatch_Print = RigVMDispatchNode(ScreenColor=(R=1,G=1,B=1,A=1), ScreenDuration=1.5, Enabled=@n1.debugRotation, Value=@RigVMFunction_MathQuaternionToEuler.Result) {template="DISPATCH_RigVMDispatch_Print(in Prefix,in Value,in Enabled,in ScreenDuration,in ScreenColor)"}
   Cast = RigVMDispatchNode(Value=@n1.triggerAxis) {template="DISPATCH_RigVMDispatch_CastEnumToInt(in Value,out Result)"}
   DISPATCH_RigVMDispatch_SelectInt32 = RigVMDispatchNode(
    Values=(0,0,0)
    Index=@Cast.Result
    Values.0=@RigVMFunction_MathQuaternionToEuler.Result.X
    Values.1=@RigVMFunction_MathQuaternionToEuler.Result.Y
    Values.2=@RigVMFunction_MathQuaternionToEuler.Result.Z
   ) {template="DISPATCH_RigVMDispatch_SelectInt32(in Index,in Values,out Result)"}
   Remap = &RigVMFunction_MathDoubleRemap(
    bClamp=False
    Value=@DISPATCH_RigVMDispatch_SelectInt32.Result
    TargetMinimum=@n1.outputMinMax.X
    TargetMaximum=@n1.outputMinMax.Y
    SourceMinimum=@n1.rotationMinMax.X
    SourceMaximum=@n1.rotationMinMax.Y
   )
   Absolute = &RigVMFunction_MathDoubleAbs(Value=@Remap.Result)
   Multiply = &RigVMFunction_MathDoubleMul(
    A=&RigVMFunction_MathDoubleClamp(Value=RigVMDispatchNode(True=@Absolute.Result, Condition=@n1.absoluteValue, False=@Remap.Result) {template="DISPATCH_RigVMDispatch_If(in Condition,in True,in False,out Result)"}.Result).Result
    B=@n1.premult
   )
   > RigVMFunctionReturnNode(outputQuat=@n2.Result, outputWeight=@Multiply.Result)
  ]}
 RigVMFunctionLibrary {nodes [
   DeltaRotation = RigVMCollapseNode(
    ExecuteContext=()
    Rotation Order=ZYX
    child=(Type=Bone,Name="None")
    outputMinMax=(X=0,Y=1)
    parent=(Type=Bone,Name="None")
    premult=1
    rotationMinMax=(X=0,Y=0)
    twistAxis=(X=1,Y=0,Z=0)
   )
   DeltaRotation v4 = RigVMCollapseNode(
    ExecuteContext=()
    Rotation Order=ZYX
    child=(Type=Bone,Name="None")
    outputMinMax=(X=0,Y=1)
    parent=(Type=Bone,Name="None")
    premult=1
    rotationMinMax=(X=0,Y=0)
    twistAxis=(X=1,Y=0,Z=0)
   )
  ]}
 RigVMModel {nodes [
   RigUnit_BeginExecution = &RigUnit_BeginExecution()
   RigVMCollapseNode(ExecuteContext=@RigUnit_BeginExecution.ExecutePin)
  ]}
]
hierarchy [
 root {loc=<0 0 0>, rot=<0 0 0 1>}
 pelvis {parent=root, loc=<0 -0.5301 98.0571>, rot=<0.0127 0.707 -0.0127 -0.707>}
 spine_01 {parent=pelvis, loc=<6.1208 -0.9328 0>, rot=<0 0 0.034 -0.9994>}
 spine_02 {parent=spine_01, loc=<6.5997 -0.2095 0>, rot=<0 0 0.0238 -0.9997>}
 spine_03 {parent=spine_02, loc=<6.4914 -0.1237 0>, rot=<0 0 0.0149 -0.9999>}
 spine_04 {parent=spine_03, loc=<6.9598 0.0794 0>, rot=<0 0 -0.0132 -0.9999>}
 spine_05 {parent=spine_04, loc=<6.9987 0 0>, rot=<0 0 -0.0015 -1>}
 neck_01 {parent=spine_05, loc=<11.3138 0.1529 0>, rot=<0 0 -0.2003 -0.9797>}
 neck_02 {parent=neck_01, loc=<4.5627 0 0>, rot=<0 0 0.0504 -0.9987>}
 head {parent=neck_02, loc=<5.3914 0 0>, rot=<0 0 0.115 -0.9934>}
 base_hairFront_r_01 {parent=head, loc=<8.6635 5.4221 6.1414>, rot=<0.1686 0.035 0.9806 -0.0938>}
 dyn_hairFront_r_01 {parent=base_hairFront_r_01, loc=<0 0 0>, rot=<0 0 0 1>}
 dyn_hairFront_r_02 {parent=dyn_hairFront_r_01, loc=<4.3493 0 0>, rot=<-0.0223 -0.1899 0.1007 -0.9764>}
 dyn_hairFront_r_03 {parent=dyn_hairFront_r_02, loc=<4.781 0 0>, rot=<-0.1305 0 0 -0.9914>}
 base_hairFront_l_01 {parent=head, loc=<7.8127 7.5597 -3.3986>, rot=<0.0493 -0.9437 -0.2335 -0.2292>}
 dyn_hairFront_l_01 {parent=base_hairFront_l_01, loc=<0 0 0>, rot=<0 0 0 1>}
 dyn_hairFront_l_02 {parent=dyn_hairFront_l_01, loc=<4.3447 0 0>, rot=<0.0075 -0.2204 -0.0697 -0.9729>}
 dyn_hairFront_l_03 {parent=dyn_hairFront_l_02, loc=<6.0715 0 0>, rot=<0.0436 0 0 -0.999>}
 base_hairFrontSide_l_01 {parent=head, loc=<5.9397 6.1506 -7.7977>, rot=<0.0572 -0.9826 -0.125 -0.1252>}
 dyn_hairFrontSide_l_01 {parent=base_hairFrontSide_l_01, loc=<0 0 0>, rot=<0 0 0 1>}
 dyn_hairFrontSide_l_02 {parent=dyn_hairFrontSide_l_01, loc=<5.1021 0 0>, rot=<-0.0373 -0.1457 -0.0694 -0.9862>}
 dyn_hairFrontSide_l_03 {parent=dyn_hairFrontSide_l_02, loc=<4.8045 0 0>, rot=<0 0 0 1>}
 base_hairSide_l_01 {parent=head, loc=<5.1916 -0.4767 -10.9752>, rot=<0.0041 -1 0 0>}
 dyn_hairSide_l_01 {parent=base_hairSide_l_01, loc=<0 0 0>, rot=<0 0 0 1>}
 dyn_hairSide_l_02 {parent=dyn_hairSide_l_01, loc=<6.8494 0 0>, rot=<0 0 0 1>}
 base_hairSide_r_01 {parent=head, loc=<5.4349 -0.4747 10.9752>, rot=<0 0 1 -0.0041>}
 dyn_hairSide_r_01 {parent=base_hairSide_r_01, loc=<0 0 0>, rot=<0 0 0 1>}
 dyn_hairSide_r_02 {parent=dyn_hairSide_r_01, loc=<7.093 0 0>, rot=<0 0 0 1>}
 base_scarfBowKnot {parent=head, loc=<9.9185 7.9335 -1.0928>, rot=<0.0029 -0.7071 -0.0029 -0.7071>}
 base_scarfBow_01 {parent=base_scarfBowKnot, loc=<0.3046 -0.5295 1.6753>, rot=<0.4578 0.429 0.3256 -0.7073>}
 dyn_scarfBow_01 {parent=base_scarfBow_01, loc=<0 0 0>, rot=<0 0 0 1>}
 base_scarfBow_02 {parent=base_scarfBowKnot, loc=<0.885 0.2899 0.7768>, rot=<0.6532 0.2996 0.0258 -0.6949>}
 dyn_scarfBow_02 {parent=base_scarfBow_02, loc=<0 0 0>, rot=<0 0 0 1>}
 base_hairBackSide_r_00 {parent=head, loc=<10.1237 -8.0065 4.6302>, rot=<0.1893 -0.5463 0.8054 -0.1303>}
 dyn_hairBackSide_r_00 {parent=base_hairBackSide_r_00, loc=<0 0 0>, rot=<0 0 0 1>}
 dyn_hairBackSide_r_01 {parent=dyn_hairBackSide_r_00, loc=<5.2396 0 0.0547>, rot=<0 -0.2631 0 -0.9648>}
 dyn_hairBackSide_r_02 {parent=dyn_hairBackSide_r_01, loc=<6.8716 0 0>, rot=<0 0 0 1>}
 base_hairBackSide_l_00 {parent=head, loc=<9.9972 -8.1475 -4.3129>, rot=<0.1175 -0.8418 0.4879 -0.1986>}
 dyn_hairBackSide_l_00 {parent=base_hairBackSide_l_00, loc=<0 0 0>, rot=<0 0 0 1>}
 dyn_hairBackSide_l_01 {parent=dyn_hairBackSide_l_00, loc=<5.2396 0 0.0547>, rot=<0 -0.2631 0 -0.9648>}
 dyn_hairBackSide_l_02 {parent=dyn_hairBackSide_l_01, loc=<7.0999 0 0>, rot=<0 0 0 1>}
 head_null {parent=neck_02, type=Null, loc=<5.3914 0 0>, rot=<0 0 0.115 -0.9934>}
 clavicle_l {parent=spine_05, loc=<5.4831 -0.1053 -3.2946>, rot=<0.0465 -0.7289 -0.009 -0.6829>}
 upperarm_l {parent=clavicle_l, loc=<9.8658 -1.2136 -0.4105>, rot=<0.0164 -0.3245 -0.0269 -0.9454>}
 clavicle_r {parent=spine_05, loc=<5.4831 -0.1053 3.2946>, rot=<0.7289 0.0465 -0.6829 0.009>}
 upperarm_r {parent=clavicle_r, loc=<-9.8658 1.2136 0.4105>, rot=<0.0164 -0.3245 -0.0269 -0.9454>}
 head_null_sp {parent=spine_05, type=Null, loc=<20.6573 3.5489 0>, rot=<0 0 -0.036 -0.9994>}
 clavicle_l_null {parent=spine_05, type=Null, loc=<5.483 -0.1053 -3.2946>, rot=<0.0465 -0.7289 -0.009 -0.6829>}
 clavicle_r_null {parent=spine_05, type=Null, loc=<5.483 -0.1053 3.2946>, rot=<0.7289 0.0465 -0.6829 0.009>}
]
variables [
 pub isCrouching {type=bool, default=False}
 pub isSkydive {type=bool, default=False}
]
