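# Post-pose AnimBP for the MarineCarve female face accessory, adding secondary motion to hair and scarf bow bones.
# Copies the customization base mesh pose through a Control Rig fed crouch and skydive states.
# A serial chain of root-relative, angular-spring AnimDynamics nodes then simulates each bone within angular limits.
# The event graph computes gravity, rigid-body and sim-space variables the AnimGraph never reads; OLDSETUP is disconnected.
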AnimBlueprint /BRCosmetics/Characters/Player/Female/Medium/Bodies/F_MED_MarineCarve/Meshes/Parts/F_MED_MarineCarve_FaceAcc_AnimBP
target_skeleton=&/F_MED_MarineCarve_FaceAcc_Skeleton
graphs [
 AnimGraph {nodes [
   Copy Pose From Mesh() {bCopyCurves=true, bUseAttachedParent=true}
   Output Pose(Result=Use cached pose 'CR_Cache_Dynamics'()) {layerGroup=None}
   n1 = Anim Dynamics - Bone: dyn_hairSide_r_01!(
    ComponentPose=Local To Component(LocalPose=Use cached pose 'CRCache_Input'())
    Alpha=1
    GravityOverride:Vector="0,0,-980"
    ExternalForce="0,0,0"
   ) -> (Pose) {
    angularDampingOverride=1
    angularSpringConstant=20000
    bAngularSpring=true
    bOverrideAngularDamping=true
    bUseGravityOverride=true
    gravityScale=0.35
    simulationSpace=RootRelative
    boundBone {boneName=dyn_hairSide_r_01}
    componentAppliedLinearAccClamp=<10000 10000 10000>
    componentLinearAccScale=<1 1 1>
    componentLinearVelScale=<1 1 1>
    gravityOverride {z=-980}
    physicsBodyDefinitions [{
      boundBone {boneName=dyn_hairSide_r_01}
      boxExtents=<30 50 50>
      constraintSetup {angularLimitsMax=<5 5 10>, angularLimitsMin=<-5 -15 -10>, angularTarget=<1 0 0>}
      localJointOffset=<4 0 0>
     }]
    simSpaceSettings {maxAngularAcceleration=25, maxAngularVelocity=5, simSpaceAngularAlpha=1}
   }
   n2 = Anim Dynamics - Bone: dyn_hairSide_r_02!(ComponentPose=@n1, Alpha=1, GravityOverride:Vector="0,0,-980", ExternalForce="0,0,0") -> (Pose) {
    angularDampingOverride=1
    angularSpringConstant=20000
    bAngularSpring=true
    bOverrideAngularDamping=true
    bUseGravityOverride=true
    gravityScale=0.35
    simulationSpace=RootRelative
    boundBone {boneName=dyn_hairSide_r_02}
    componentAppliedLinearAccClamp=<10000 10000 10000>
    componentLinearAccScale=<1 1 1>
    componentLinearVelScale=<1 1 1>
    gravityOverride {z=-980}
    physicsBodyDefinitions [{
      boundBone {boneName=dyn_hairSide_r_02}
      boxExtents=<30 50 50>
      constraintSetup {angularLimitsMax=<5 5 10>, angularLimitsMin=<-5 -15 -10>, angularTarget=<1 0 0>}
      localJointOffset=<4 0 0>
     }]
    simSpaceSettings {maxAngularAcceleration=25, maxAngularVelocity=5, simSpaceAngularAlpha=1}
   }
   n3 = Anim Dynamics - Bone: dyn_hairSide_l_01!(ComponentPose=@n2, Alpha=1, GravityOverride:Vector="0,0,-980", ExternalForce="0,0,0") -> (Pose) {
    angularDampingOverride=1
    angularSpringConstant=20000
    bAngularSpring=true
    bOverrideAngularDamping=true
    bUseGravityOverride=true
    gravityScale=0.35
    simulationSpace=RootRelative
    boundBone {boneName=dyn_hairSide_l_01}
    componentAppliedLinearAccClamp=<10000 10000 10000>
    componentLinearAccScale=<1 1 1>
    componentLinearVelScale=<1 1 1>
    gravityOverride {z=-980}
    physicsBodyDefinitions [{
      boundBone {boneName=dyn_hairSide_l_01}
      boxExtents=<30 50 50>
      constraintSetup {angularLimitsMax=<5 5 10>, angularLimitsMin=<-5 -15 -10>, angularTarget=<1 0 0>}
      localJointOffset=<4 0 0>
     }]
    simSpaceSettings {maxAngularAcceleration=25, maxAngularVelocity=5, simSpaceAngularAlpha=1}
   }
   n4 = Anim Dynamics - Bone: dyn_hairSide_l_02!(ComponentPose=@n3, Alpha=1, GravityOverride:Vector="0,0,-980", ExternalForce="0,0,0") -> (Pose) {
    angularDampingOverride=1
    angularSpringConstant=20000
    bAngularSpring=true
    bOverrideAngularDamping=true
    bUseGravityOverride=true
    gravityScale=0.35
    simulationSpace=RootRelative
    boundBone {boneName=dyn_hairSide_l_02}
    componentAppliedLinearAccClamp=<10000 10000 10000>
    componentLinearAccScale=<1 1 1>
    componentLinearVelScale=<1 1 1>
    gravityOverride {z=-980}
    physicsBodyDefinitions [{
      boundBone {boneName=dyn_hairSide_l_02}
      boxExtents=<30 50 50>
      constraintSetup {angularLimitsMax=<5 5 10>, angularLimitsMin=<-5 -15 -10>, angularTarget=<1 0 0>}
      localJointOffset=<4 0 0>
     }]
    simSpaceSettings {maxAngularAcceleration=25, maxAngularVelocity=5, simSpaceAngularAlpha=1}
   }
   n5 = Anim Dynamics - Bone: dyn_hairFront_l_01!(ComponentPose=@n4, Alpha=1, GravityOverride:Vector="0,0,-980", ExternalForce="0,0,0") -> (Pose) {
    angularDampingOverride=1
    angularSpringConstant=20000
    bAngularSpring=true
    bOverrideAngularDamping=true
    bUseGravityOverride=true
    gravityScale=0.35
    simulationSpace=RootRelative
    boundBone {boneName=dyn_hairFront_l_01}
    componentAppliedLinearAccClamp=<10000 10000 10000>
    componentLinearAccScale=<1 1 1>
    componentLinearVelScale=<1 1 1>
    gravityOverride {z=-980}
    physicsBodyDefinitions [{
      boundBone {boneName=dyn_hairFront_l_01}
      boxExtents=<30 50 50>
      constraintSetup {angularLimitsMax=<5 2 15>, angularLimitsMin=<-5 -10 -5>, angularTarget=<1 0 0>}
      localJointOffset=<4 0 0>
     }]
    simSpaceSettings {maxAngularAcceleration=25, maxAngularVelocity=5, simSpaceAngularAlpha=1}
   }
   n6 = Anim Dynamics - Bone: dyn_hairFront_l_02!(ComponentPose=@n5, Alpha=1, GravityOverride:Vector="0,0,-980", ExternalForce="0,0,0") -> (Pose) {
    angularDampingOverride=1
    angularSpringConstant=20000
    bAngularSpring=true
    bOverrideAngularDamping=true
    bUseGravityOverride=true
    gravityScale=0.35
    simulationSpace=RootRelative
    boundBone {boneName=dyn_hairFront_l_02}
    componentAppliedLinearAccClamp=<10000 10000 10000>
    componentLinearAccScale=<1 1 1>
    componentLinearVelScale=<1 1 1>
    gravityOverride {z=-980}
    physicsBodyDefinitions [{
      boundBone {boneName=dyn_hairFront_l_02}
      boxExtents=<30 50 50>
      constraintSetup {angularLimitsMax=<5 2 15>, angularLimitsMin=<-5 -10 0>, angularTarget=<1 0 0>}
      localJointOffset=<4 0 0>
     }]
    simSpaceSettings {maxAngularAcceleration=25, maxAngularVelocity=5, simSpaceAngularAlpha=1}
   }
   n7 = Anim Dynamics - Bone: dyn_hairFront_l_03!(ComponentPose=@n6, Alpha=1, GravityOverride:Vector="0,0,-980", ExternalForce="0,0,0") -> (Pose) {
    angularDampingOverride=1
    angularSpringConstant=20000
    bAngularSpring=true
    bOverrideAngularDamping=true
    bUseGravityOverride=true
    gravityScale=0.35
    simulationSpace=RootRelative
    boundBone {boneName=dyn_hairFront_l_03}
    componentAppliedLinearAccClamp=<10000 10000 10000>
    componentLinearAccScale=<1 1 1>
    componentLinearVelScale=<1 1 1>
    gravityOverride {z=-980}
    physicsBodyDefinitions [{
      boundBone {boneName=dyn_hairFront_l_03}
      boxExtents=<30 50 50>
      constraintSetup {angularLimitsMax=<5 2 15>, angularLimitsMin=<-5 -10 0>, angularTarget=<1 0 0>}
      localJointOffset=<4 0 0>
     }]
    simSpaceSettings {maxAngularAcceleration=25, maxAngularVelocity=5, simSpaceAngularAlpha=1}
   }
   n8 = Anim Dynamics - Bone: dyn_hairFront_r_01!(ComponentPose=@n7, Alpha=1, GravityOverride:Vector="0,0,-980", ExternalForce="0,0,0") -> (Pose) {
    angularDampingOverride=1
    angularSpringConstant=20000
    bAngularSpring=true
    bOverrideAngularDamping=true
    bUseGravityOverride=true
    gravityScale=0.35
    simulationSpace=RootRelative
    boundBone {boneName=dyn_hairFront_r_01}
    componentAppliedLinearAccClamp=<10000 10000 10000>
    componentLinearAccScale=<1 1 1>
    componentLinearVelScale=<1 1 1>
    gravityOverride {z=-980}
    physicsBodyDefinitions [{
      boundBone {boneName=dyn_hairFront_r_01}
      boxExtents=<30 50 50>
      constraintSetup {angularLimitsMax=<5 2 15>, angularLimitsMin=<-5 -10 -5>, angularTarget=<1 0 0>}
      localJointOffset=<4 0 0>
     }]
    simSpaceSettings {maxAngularAcceleration=25, maxAngularVelocity=5, simSpaceAngularAlpha=1}
   }
   n9 = Anim Dynamics - Bone: dyn_hairFront_r_02!(ComponentPose=@n8, Alpha=1, GravityOverride:Vector="0,0,-980", ExternalForce="0,0,0") -> (Pose) {
    angularDampingOverride=1
    angularSpringConstant=20000
    bAngularSpring=true
    bOverrideAngularDamping=true
    bUseGravityOverride=true
    gravityScale=0.35
    simulationSpace=RootRelative
    boundBone {boneName=dyn_hairFront_r_02}
    componentAppliedLinearAccClamp=<10000 10000 10000>
    componentLinearAccScale=<1 1 1>
    componentLinearVelScale=<1 1 1>
    gravityOverride {z=-980}
    physicsBodyDefinitions [{
      boundBone {boneName=dyn_hairFront_r_02}
      boxExtents=<30 50 50>
      constraintSetup {angularLimitsMax=<5 2 15>, angularLimitsMin=<-5 -10 0>, angularTarget=<1 0 0>}
      localJointOffset=<4 0 0>
     }]
    simSpaceSettings {maxAngularAcceleration=25, maxAngularVelocity=5, simSpaceAngularAlpha=1}
   }
   n10 = Anim Dynamics - Bone: dyn_hairFront_r_03!(ComponentPose=@n9, Alpha=1, GravityOverride:Vector="0,0,-980", ExternalForce="0,0,0") -> (Pose) {
    angularDampingOverride=1
    angularSpringConstant=20000
    bAngularSpring=true
    bOverrideAngularDamping=true
    bUseGravityOverride=true
    gravityScale=0.35
    simulationSpace=RootRelative
    boundBone {boneName=dyn_hairFront_r_03}
    componentAppliedLinearAccClamp=<10000 10000 10000>
    componentLinearAccScale=<1 1 1>
    componentLinearVelScale=<1 1 1>
    gravityOverride {z=-980}
    physicsBodyDefinitions [{
      boundBone {boneName=dyn_hairFront_r_03}
      boxExtents=<30 50 50>
      constraintSetup {angularLimitsMax=<5 2 15>, angularLimitsMin=<-5 -10 0>, angularTarget=<1 0 0>}
      localJointOffset=<4 0 0>
     }]
    simSpaceSettings {maxAngularAcceleration=25, maxAngularVelocity=5, simSpaceAngularAlpha=1}
   }
   n11 = Anim Dynamics - Bone: dyn_hairFrontSide_l_01!(ComponentPose=@n10, Alpha=1, GravityOverride:Vector="0,0,-980", ExternalForce="0,0,0") -> (Pose) {
    angularDampingOverride=1
    angularSpringConstant=20000
    bAngularSpring=true
    bOverrideAngularDamping=true
    bUseGravityOverride=true
    gravityScale=0.35
    simulationSpace=RootRelative
    boundBone {boneName=dyn_hairFrontSide_l_01}
    componentAppliedLinearAccClamp=<10000 10000 10000>
    componentLinearAccScale=<1 1 1>
    componentLinearVelScale=<1 1 1>
    gravityOverride {z=-980}
    physicsBodyDefinitions [{
      boundBone {boneName=dyn_hairFrontSide_l_01}
      boxExtents=<30 50 50>
      constraintSetup {angularLimitsMax=<5 2 15>, angularLimitsMin=<-5 -15 -3>, angularTarget=<1 0 0>}
      localJointOffset=<4 0 0>
     }]
    simSpaceSettings {maxAngularAcceleration=25, maxAngularVelocity=5, simSpaceAngularAlpha=1}
   }
   n12 = Anim Dynamics - Bone: dyn_hairFrontSide_l_02!(ComponentPose=@n11, Alpha=1, GravityOverride:Vector="0,0,-980", ExternalForce="0,0,0") -> (Pose) {
    angularDampingOverride=1
    angularSpringConstant=20000
    bAngularSpring=true
    bOverrideAngularDamping=true
    bUseGravityOverride=true
    gravityScale=0.35
    simulationSpace=RootRelative
    boundBone {boneName=dyn_hairFrontSide_l_02}
    componentAppliedLinearAccClamp=<10000 10000 10000>
    componentLinearAccScale=<1 1 1>
    componentLinearVelScale=<1 1 1>
    gravityOverride {z=-980}
    physicsBodyDefinitions [{
      boundBone {boneName=dyn_hairFrontSide_l_02}
      boxExtents=<30 50 50>
      constraintSetup {angularLimitsMax=<5 2 15>, angularLimitsMin=<-5 -15 0>, angularTarget=<1 0 0>}
      localJointOffset=<4 0 0>
     }]
    simSpaceSettings {maxAngularAcceleration=25, maxAngularVelocity=5, simSpaceAngularAlpha=1}
   }
   n13 = Anim Dynamics - Bone: dyn_hairFrontSide_l_03!(ComponentPose=@n12, Alpha=1, GravityOverride:Vector="0,0,-980", ExternalForce="0,0,0") -> (Pose) {
    angularDampingOverride=1
    angularSpringConstant=20000
    bAngularSpring=true
    bOverrideAngularDamping=true
    bUseGravityOverride=true
    gravityScale=0.35
    simulationSpace=RootRelative
    boundBone {boneName=dyn_hairFrontSide_l_03}
    componentAppliedLinearAccClamp=<10000 10000 10000>
    componentLinearAccScale=<1 1 1>
    componentLinearVelScale=<1 1 1>
    gravityOverride {z=-980}
    physicsBodyDefinitions [{
      boundBone {boneName=dyn_hairFrontSide_l_03}
      boxExtents=<30 50 50>
      constraintSetup {angularLimitsMax=<5 2 15>, angularLimitsMin=<-5 -15 0>, angularTarget=<1 0 0>}
      localJointOffset=<4 0 0>
     }]
    simSpaceSettings {maxAngularAcceleration=25, maxAngularVelocity=5, simSpaceAngularAlpha=1}
   }
   n14 = Anim Dynamics - Bone: dyn_hairBackSide_r_00!(ComponentPose=@n13, Alpha=1, GravityOverride:Vector="0,0,-980", ExternalForce="0,0,0") -> (Pose) {
    angularDampingOverride=1
    angularSpringConstant=20000
    bAngularSpring=true
    bOverrideAngularDamping=true
    bUseGravityOverride=true
    gravityScale=0.35
    simulationSpace=RootRelative
    boundBone {boneName=dyn_hairBackSide_r_00}
    componentAppliedLinearAccClamp=<10000 10000 10000>
    componentLinearAccScale=<1 1 1>
    componentLinearVelScale=<1 1 1>
    gravityOverride {z=-980}
    physicsBodyDefinitions [{
      boundBone {boneName=dyn_hairBackSide_r_00}
      boxExtents=<30 50 50>
      constraintSetup {angularLimitsMax=<5 2 5>, angularLimitsMin=<-5 -25 -5>, angularTarget=<1 0 0>}
      localJointOffset=<4 0 0>
     }]
    simSpaceSettings {maxAngularAcceleration=25, maxAngularVelocity=5, simSpaceAngularAlpha=1}
   }
   n15 = Anim Dynamics - Bone: dyn_hairBackSide_r_01!(ComponentPose=@n14, Alpha=1, GravityOverride:Vector="0,0,-980", ExternalForce="0,0,0") -> (Pose) {
    angularDampingOverride=1
    angularSpringConstant=20000
    bAngularSpring=true
    bOverrideAngularDamping=true
    bUseGravityOverride=true
    gravityScale=0.35
    simulationSpace=RootRelative
    boundBone {boneName=dyn_hairBackSide_r_01}
    componentAppliedLinearAccClamp=<10000 10000 10000>
    componentLinearAccScale=<1 1 1>
    componentLinearVelScale=<1 1 1>
    gravityOverride {z=-980}
    physicsBodyDefinitions [{
      boundBone {boneName=dyn_hairBackSide_r_01}
      boxExtents=<30 50 50>
      constraintSetup {angularLimitsMax=<5 2 5>, angularLimitsMin=<-5 -15 -5>, angularTarget=<1 0 0>}
      localJointOffset=<4 0 0>
     }]
    simSpaceSettings {maxAngularAcceleration=25, maxAngularVelocity=5, simSpaceAngularAlpha=1}
   }
   n16 = Anim Dynamics - Bone: dyn_hairBackSide_r_02!(ComponentPose=@n15, Alpha=1, GravityOverride:Vector="0,0,-980", ExternalForce="0,0,0") -> (Pose) {
    angularDampingOverride=1
    angularSpringConstant=20000
    bAngularSpring=true
    bOverrideAngularDamping=true
    bUseGravityOverride=true
    gravityScale=0.35
    simulationSpace=RootRelative
    boundBone {boneName=dyn_hairBackSide_r_02}
    componentAppliedLinearAccClamp=<10000 10000 10000>
    componentLinearAccScale=<1 1 1>
    componentLinearVelScale=<1 1 1>
    gravityOverride {z=-980}
    physicsBodyDefinitions [{
      boundBone {boneName=dyn_hairBackSide_r_02}
      boxExtents=<30 50 50>
      constraintSetup {angularLimitsMax=<5 2 5>, angularLimitsMin=<-5 -15 -5>, angularTarget=<1 0 0>}
      localJointOffset=<4 0 0>
     }]
    simSpaceSettings {maxAngularAcceleration=25, maxAngularVelocity=5, simSpaceAngularAlpha=1}
   }
   n17 = Anim Dynamics - Bone: dyn_hairBackSide_l_00!(ComponentPose=@n16, Alpha=1, GravityOverride:Vector="0,0,-980", ExternalForce="0,0,0") -> (Pose) {
    angularDampingOverride=1
    angularSpringConstant=20000
    bAngularSpring=true
    bOverrideAngularDamping=true
    bUseGravityOverride=true
    gravityScale=0.35
    simulationSpace=RootRelative
    boundBone {boneName=dyn_hairBackSide_l_00}
    componentAppliedLinearAccClamp=<10000 10000 10000>
    componentLinearAccScale=<1 1 1>
    componentLinearVelScale=<1 1 1>
    gravityOverride {z=-980}
    physicsBodyDefinitions [{
      boundBone {boneName=dyn_hairBackSide_l_00}
      boxExtents=<30 50 50>
      constraintSetup {angularLimitsMax=<5 2 5>, angularLimitsMin=<-5 -25 -5>, angularTarget=<1 0 0>}
      localJointOffset=<4 0 0>
     }]
    simSpaceSettings {maxAngularAcceleration=25, maxAngularVelocity=5, simSpaceAngularAlpha=1}
   }
   n18 = Anim Dynamics - Bone: dyn_hairBackSide_l_01!(ComponentPose=@n17, Alpha=1, GravityOverride:Vector="0,0,-980", ExternalForce="0,0,0") -> (Pose) {
    angularDampingOverride=1
    angularSpringConstant=20000
    bAngularSpring=true
    bOverrideAngularDamping=true
    bUseGravityOverride=true
    gravityScale=0.35
    simulationSpace=RootRelative
    boundBone {boneName=dyn_hairBackSide_l_01}
    componentAppliedLinearAccClamp=<10000 10000 10000>
    componentLinearAccScale=<1 1 1>
    componentLinearVelScale=<1 1 1>
    gravityOverride {z=-980}
    physicsBodyDefinitions [{
      boundBone {boneName=dyn_hairBackSide_l_01}
      boxExtents=<30 50 50>
      constraintSetup {angularLimitsMax=<5 2 5>, angularLimitsMin=<-5 -15 -5>, angularTarget=<1 0 0>}
      localJointOffset=<4 0 0>
     }]
    simSpaceSettings {maxAngularAcceleration=25, maxAngularVelocity=5, simSpaceAngularAlpha=1}
   }
   n19 = Anim Dynamics - Bone: dyn_hairBackSide_l_02!(ComponentPose=@n18, Alpha=1, GravityOverride:Vector="0,0,-980", ExternalForce="0,0,0") -> (Pose) {
    angularDampingOverride=1
    angularSpringConstant=20000
    bAngularSpring=true
    bOverrideAngularDamping=true
    bUseGravityOverride=true
    gravityScale=0.35
    simulationSpace=RootRelative
    boundBone {boneName=dyn_hairBackSide_l_02}
    componentAppliedLinearAccClamp=<10000 10000 10000>
    componentLinearAccScale=<1 1 1>
    componentLinearVelScale=<1 1 1>
    gravityOverride {z=-980}
    physicsBodyDefinitions [{
      boundBone {boneName=dyn_hairBackSide_l_02}
      boxExtents=<30 50 50>
      constraintSetup {angularLimitsMax=<5 2 5>, angularLimitsMin=<-5 -15 -5>, angularTarget=<1 0 0>}
      localJointOffset=<4 0 0>
     }]
    simSpaceSettings {maxAngularAcceleration=25, maxAngularVelocity=5, simSpaceAngularAlpha=1}
   }
   n20 = Anim Dynamics - Bone: dyn_scarfBow_01!(ComponentPose=@n19, Alpha=1, GravityOverride:Vector="0,0,-980", ExternalForce="0,0,0") -> (Pose) {
    angularDampingOverride=1
    angularSpringConstant=20000
    bAngularSpring=true
    bOverrideAngularDamping=true
    bUseGravityOverride=true
    gravityScale=0.35
    simulationSpace=RootRelative
    boundBone {boneName=dyn_scarfBow_01}
    componentAppliedLinearAccClamp=<10000 10000 10000>
    componentLinearAccScale=<1 1 1>
    componentLinearVelScale=<1 1 1>
    gravityOverride {z=-980}
    physicsBodyDefinitions [{
      boundBone {boneName=dyn_scarfBow_01}
      boxExtents=<30 40 40>
      constraintSetup {angularLimitsMax=<8 15 10>, angularLimitsMin=<-8 -30 -10>, angularTarget=<1 0 0>}
      localJointOffset=<4 0 0>
     }]
    simSpaceSettings {maxAngularAcceleration=25, maxAngularVelocity=5, simSpaceAngularAlpha=1}
   }
   n21 = Anim Dynamics - Bone: dyn_scarfBow_02!(ComponentPose=@n20, Alpha=1, GravityOverride:Vector="0,0,-980", ExternalForce="0,0,0") -> (Pose) {
    angularDampingOverride=1
    angularSpringConstant=20000
    bAngularSpring=true
    bOverrideAngularDamping=true
    bUseGravityOverride=true
    gravityScale=0.35
    simulationSpace=RootRelative
    boundBone {boneName=dyn_scarfBow_02}
    componentAppliedLinearAccClamp=<10000 10000 10000>
    componentLinearAccScale=<1 1 1>
    componentLinearVelScale=<1 1 1>
    gravityOverride {z=-980}
    physicsBodyDefinitions [{
      boundBone {boneName=dyn_scarfBow_02}
      boxExtents=<30 40 40>
      constraintSetup {angularLimitsMax=<12 20 15>, angularLimitsMin=<-12 -25 -15>, angularTarget=<1 0 0>}
      localJointOffset=<4 0 0>
     }]
    simSpaceSettings {maxAngularAcceleration=25, maxAngularVelocity=5, simSpaceAngularAlpha=1}
   }
   Save cached pose 'CR_Cache_Dynamics'(Pose=Component To Local(ComponentPose=@n21)) {cacheName=CR_Cache_Dynamics}
   n22 = Control Rig!(
    Source=Copy Pose From Mesh(SourceMeshComponent=Get SkeletalMesh() -> (SkeletalMesh:Skeletal Mesh Component Object Reference))
    bAlphaBoolEnabled:bool=True
    pub isCrouching:bool=Get bIsCrouching()
    pub isSkydive:bool=Get bIsSkydiveFloating()
   ) -> (Pose) {
    alphaInputType=Bool
    alphaBoolBlend {blendOption=Linear}
    controlRigAssetReference {blueprintRigClass=&/F_MED_MarineCarve_FaceAcc_CtrlRig.F_MED_MarineCarve_FaceAcc_CtrlRig_C}
   }
   Save cached pose 'CRCache_Input'(Pose=@n22) {cacheName=CRCache_Input}
  ]}
 EventGraph {nodes [
   n1 = Event OnCharacterCustomizationLoaded()
   > Set SkeletalMesh(SkeletalMesh:Skeletal Mesh Component Object Reference=@n1.CustomBaseMesh) -> (Output_Get:Skeletal Mesh Component Object Reference)
   n2 = Event BlueprintUpdateAnimation()
   > Set Delta Time X(Delta Time X=@n2.DeltaTimeX)
   > n3 = Sequence()
   > n4 = "Gravity Override State Function (Parameter Asset) (Nativized)"(
    CharacterPartAnimInstance=Self-Reference()
    GameTimeInSeconds=GetGameTimeInSeconds()
    Parameters=&/F_MED_MarineCarve_FaceAcc_Parameters
   )
   > Set GravityOverride(GravityOverride=@n4)
   n5 = "integer > integer"(A=Get CurrentLOD(), B=1) {is=title K2Node_CallFunction}
   n6 = NOT Boolean(A=Get bRigidBodyEnabled())
   > Set Is Rigid Body Enabled(Is Rigid Body Enabled:bool=NOT Boolean(A=OR Boolean(A=@n6, B=@n5))) -> (Output_Get:bool)
   n7 = Branch(execute=@n3.then_1, Condition=Get bIsInFrontEnd())
   > Set Sim Space Rig B(Sim Space Rig B:Sim Space Settings Structure=Get Sim Space FE() -> (Sim Space FE:Sim Space Settings Structure)) -> (Output_Get:Sim Space Settings Structure)
   Set Sim Space Rig B(execute=@n7.else, Sim Space Rig B=Get Sim Space BR() -> (Sim Space BR:Sim Space Settings Structure))
   K2Node_Composite_0 = OLDSETUP!() {is=title K2Node_Composite}
  ]}
 OLDSETUP {
  path=EventGraph.K2Node_Composite_0.OLDSETUP
  nodes [
   n1 = "Gravity Override State Function (Parameter Asset) (Nativized)"(
    CharacterPartAnimInstance=Self-Reference()
    GameTimeInSeconds=GetGameTimeInSeconds()
    Parameters=&F_MED_MarineCarve_FaceAcc_Parameters
   )
   > Set GravityOverride(GravityOverride=@n1)
   n2 = Branch(Condition=Get bIsSkydiveFloating())
   > Set IsSkydiveFloat(IsSkydiveFloat:double=1) -> (Output_Get:double)
   Set IsSkydiveFloat(execute=@n2.else, IsSkydiveFloat=0)
   Reroute Node()
   n3 = Get bIsCrouching()
   n4 = Get Delta Time X()
   > Set isCrouching float(isCrouching float:double=FInterpTo(Current=Get isCrouching float() -> (isCrouching float:double), Target="To Float (Boolean)"(InBool=@n3), DeltaTime=@n4, InterpSpeed=15)) -> (Output_Get:double)
   n5 = Get bIsTargeting()
   > Set isTargeting float(isTargeting float:double=FInterpTo(Current=Get isTargeting float() -> (isTargeting float:double), Target="To Float (Boolean)"(InBool=@n5), DeltaTime=@n4, InterpSpeed=15)) -> (Output_Get:double)
   Set CrouchWalk(CrouchWalk:bool=AND Boolean(A=Get bIsCrouching(), B=Get bIsMoving2D())) -> (Output_Get:bool)
   > Set CrouchTargeting(CrouchTargeting:bool=AND Boolean(A=Get bIsCrouching(), B=Get bIsTargeting())) -> (Output_Get:bool)
   Set isUsingPistol(isUsingPistol:double=0) -> (Output_Get:double)
   > n6 = Switch on EFortWeaponCoreAnimation!(execute, Selection:EFortWeaponCoreAnimation Enum=Get WeaponCoreAnim!() -> (WeaponCoreAnim)) -> (
    Melee:Exec
    Pistol:Exec
    Shotgun:Exec
    PaperBlueprint:Exec
    Rifle:Exec
    MeleeOneHand:Exec
    MachinePistol:Exec
    RocketLauncher:Exec
    GrenadeLauncher:Exec
    GoingCommando:Exec
    AssaultRifle:Exec
    TacticalShotgun:Exec
    SniperRifle:Exec
    TrapPlacement:Exec
    ShoulderLauncher:Exec
    AbilityDecoTool:Exec
    Crossbow:Exec
    C4:Exec
    RemoteControl:Exec
    DualWield:Exec
    AR_BullPup:Exec
    AR_ForwardGrip:Exec
    MedPackPaddles:Exec
    SMG_P90:Exec
    AR_DrumGun:Exec
    Consumable_Small:Exec
    Consumable_Large:Exec
    Balloon:Exec
    MountedTurret:Exec
    CreativeTool:Exec
    ExplosiveBow:Exec
    AshtonIndigo:Exec
    AshtonChicago:Exec
    MeleeDualWield:Exec
    Unarmed:Exec
   )
   Set isUsingPistol(execute=@n6.Pistol, isUsingPistol=1)
   n7 = Get SkeletalMesh()
   n8 = Set CrouchHead Up(CrouchHead Up:double=SelectFloat(
     A=GetAlphaFromReferenceTransformDelta(SkeletalMeshComponent=@n7, BoneName=spine_05, BaseName=head, InputRange=(X=0,Y=90), TwistAxis="0,0,1")
     B=0
     bPickA=Get bIsCrouching()
    )) -> (Output_Get:double)
   > Set CrouchingTargetHeadUP(CrouchingTargetHeadUP:double=SelectFloat(A=@n8, B=0, bPickA=Get CrouchTargeting() -> (CrouchTargeting:bool))) -> (Output_Get:double)
   n9 = NOT Boolean(A=Get bIsCrouching())
   n10 = Get bIsTargeting()
   > Set AimUp(AimUp:double=SelectFloat(A=@n8, B=0, bPickA=AND Boolean(A=@n9, B=@n10))) -> (Output_Get:double)
   Inputs!()
   Outputs()
   Set SkeletalMesh(SkeletalMesh=GetOwnerSkeletalMeshComponent())
  ]
 }
]
variables [
 AimUp {type=number}
 armLeftAlpha {type=number}
 armRightAlpha {type=number}
 CrouchHead Up {type=number}
 CrouchingTargetHeadUP {type=number}
 CrouchTargeting {type=boolean}
 CrouchWalk {type=boolean}
 Delta Time X {type=number}
 ExtremeHeadPoseAlpha {type=number}
 Gravity Override Params {type=GravityOverrideParamsStruct}
 GravityOverride {type=Vector}
 Head look down alpha {type=number}
 Head look up alpha {type=number}
 HeadRotateExtreme {type=number}
 HeadRotateUp {type=number}
 HeadTiltAlpha {type=number}
 Is Ctrl Rig Active {type=boolean}
 Is Rigid Body Enabled {type=boolean}
 isCrouching float {type=number}
 isCtrlRigActive {type=boolean}
 IsSkydiveFloat {type=number}
 isTargeting float {type=number}
 isUsingPistol {type=number}
 Machine gun corr {type=number}
 Parameters {type=FortCharacterDynamicsParameters, default=null}
 Pistol CrouchAimup {type=number}
 Pistol Equipped {type=boolean}
 Pub Pawn Delta Rotation {type=Rotator}
 RigidBodyAlpha {type=number}
 Sim Space BR {
  type=SimSpaceSettings
  default {
   dampingAlpha=1
   maxAngularAcceleration=100
   maxAngularVelocity=3
   maxLinearAcceleration=800
   maxLinearVelocity=400
   velocityScaleZ=4
   worldAlpha=1
   externalLinearDragV=<1 1 1>
  }
 }
 Sim Space FE {
  type=SimSpaceSettings
  default {
   dampingAlpha=1
   maxAngularAcceleration=1000
   maxAngularVelocity=3
   maxLinearAcceleration=1000
   maxLinearVelocity=500
   velocityScaleZ=1.5
   worldAlpha=1
   externalLinearDragV=<0.1 0.1 0.2>
  }
 }
 Sim Space Rig B {
  type=SimSpaceSettings
  default {
   dampingAlpha=1
   maxAngularAcceleration=10000
   maxAngularVelocity=10000
   maxLinearAcceleration=10000
   maxLinearVelocity=10000
   velocityScaleZ=1
   worldAlpha=0.7
   externalLinearDragV=<0.5 0.5 0.5>
  }
 }
 SkeletalMesh {type=SkeletalMeshComponent}
 TESTBOOL {type=boolean}
]
