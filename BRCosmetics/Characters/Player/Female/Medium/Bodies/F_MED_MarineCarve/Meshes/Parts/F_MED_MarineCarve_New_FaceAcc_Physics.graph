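# Physics asset for the MarineCarve medium female face accessory mesh, defining bodies and joints for secondary motion of the character's hair.
# Capsules on upper spine, neck, head, clavicles and upper arms form a body proxy joined by locked constraints.
# Small base hair bodies lock to the head and root two- or three-link dynamic chains with narrow swing and twist limits.
# Front-side hair exists only on the left, and side-hair tips use boxes instead of capsules.
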PhysicsAsset /BRCosmetics/Characters/Player/Female/Medium/Bodies/F_MED_MarineCarve/Meshes/Parts/F_MED_MarineCarve_New_FaceAcc_Physics
bodies {
 base_hairBackSide_l_01 [{shapeType=Capsule, length=0.505, radius=0.505, shapeName=base_hairBackSide_l_01_capsule, rotation=<rot 0 -90 0>}]
 base_hairBackSide_r_01 [{shapeType=Capsule, length=0.505, radius=0.505, shapeName=base_hairBackSide_r_01_capsule, rotation=<rot 0 -90 0>}]
 base_hairFrontSide_l_01 [{shapeType=Capsule, length=0.505, radius=0.505, shapeName=base_hairFrontSide_l_01_capsule, rotation=<rot 0 -90 0>}]
 base_hairFront_l_01 [{shapeType=Capsule, length=0.505, radius=0.505, shapeName=base_hairFront_l_01_capsule, rotation=<rot 0 -90 0>}]
 base_hairFront_r_01 [{shapeType=Capsule, length=0.505, radius=0.505, shapeName=base_hairFront_r_01_capsule, rotation=<rot 0 -90 0>}]
 base_hairSide_l_01 [{shapeType=Capsule, length=0.505, radius=0.505, shapeName=base_hairSide_l_01_capsule, rotation=<rot 0 -90 0>}]
 base_hairSide_r_01 [{shapeType=Capsule, length=0.505, radius=0.505, shapeName=base_hairSide_r_01_capsule, rotation=<rot 0 -90 0>}]
 clavicle_l [{shapeType=Capsule, length=9.2905, radius=6.3848, shapeName=clavicle_l_capsule, center=<3.4042 -1.0681 -1.5934>, rotation=<rot 0 -90 -82>}]
 clavicle_r [{shapeType=Capsule, length=9.2905, radius=6.3848, shapeName=clavicle_r_capsule, center=<-3.4979 1.6812 0.9262>, rotation=<rot 0 -90 -82>}]
 dyn_hairBackSide_l_01 [{shapeType=Capsule, length=7.6, radius=2, shapeName=dyn_hairBackSide_l_01_capsule, center=<4.2 0 -1.6>, rotation=<rot 0 0 90>}]
 dyn_hairBackSide_l_02 [{shapeType=Capsule, length=7, radius=2, shapeName=dyn_hairBackSide_l_02_capsule, center=<2.3 0 -3.46>, rotation=<rot 0 0 90>}]
 dyn_hairBackSide_r_01 [{shapeType=Capsule, length=7, radius=2, shapeName=dyn_hairBackSide_r_01_capsule, center=<4.2 0 -1.6>, rotation=<rot 0 0 90>}]
 dyn_hairBackSide_r_02 [{shapeType=Capsule, length=7, radius=2, shapeName=dyn_hairBackSide_r_02_capsule, center=<2.3 0 -3.46>, rotation=<rot 0 0 90>}]
 dyn_hairFrontSide_l_01 [{
   shapeType=Capsule
   length=2.3785
   radius=0.505
   shapeName=dyn_hairFrontSide_l_01_capsule
   center=<2.8376 -0.0826 0.1495>
   rotation=<rot 0.3376 90.0001 -88.0844>
  }]
 dyn_hairFrontSide_l_02 [{
   shapeType=Capsule
   length=2.6551
   radius=0.5357
   shapeName=dyn_hairFrontSide_l_02_capsule
   center=<2.2251 -0.0298 0.0516>
   rotation=<rot 0 87.3771 -88.1265>
  }]
 dyn_hairFrontSide_l_03 [{
   shapeType=Capsule
   length=2.836
   radius=0.5357
   shapeName=dyn_hairFrontSide_l_02_capsule
   center=<1.1247 0.4871 -0.549>
   rotation=<rot 9.2785 105.1736 -116.6707>
  }]
 dyn_hairFront_l_01 [{shapeType=Capsule, length=1.9422, radius=0.8834, shapeName=dyn_hairFront_l_01_capsule, center=<2.9027 0.2088 0.171>, rotation=<rot 90 0 0>}]
 dyn_hairFront_l_02 [{
   shapeType=Capsule
   length=1.9754
   radius=1.252
   shapeName=dyn_hairFront_l_02_capsule
   center=<2.7374 0.0699 0.1041>
   rotation=<rot 0 89.3851 -86.9904>
  }]
 dyn_hairFront_l_03 [{
   shapeType=Capsule
   length=5.0522
   radius=0.7898
   shapeName=dyn_hairFront_l_03_capsule
   center=<3.1712 0.1185 -0.0692>
   rotation=<rot 0 -87.0598 -87.3155>
  }]
 dyn_hairFront_r_01 [{
   shapeType=Capsule
   length=1.2319
   radius=0.5014
   shapeName=dyn_hairFront_r_01_capsule
   center=<2.2861 0.1195 0.0718>
   rotation=<rot 0 -90.7425 -88.1062>
  }]
 dyn_hairFront_r_02 [{
   shapeType=Capsule
   length=1.1119
   radius=0.4197
   shapeName=dyn_hairFront_r_02_capsule
   center=<1.6775 0.2007 0.0376>
   rotation=<rot 0 95.9212 -88.9255>
  }]
 dyn_hairFront_r_03 [{
   shapeType=Capsule
   length=1.4059
   radius=0.3661
   shapeName=dyn_hairFront_r_03_capsule
   center=<0.9347 0.0073 0.0337>
   rotation=<rot 0 96.4239 -88.74>
  }]
 dyn_hairSide_l_01 [{shapeType=Capsule, length=7, radius=2, shapeName=dyn_hairSide_l_01_capsule, center=<3.0535 0 -1.3247>, rotation=<rot 0 0 90>}]
 dyn_hairSide_l_02 [{shapeType=Box, extentX=6, extentY=8, extentZ=3, shapeName=dyn_hairSide_l_02_box, center=<1.7632 0 -0.9603>}]
 dyn_hairSide_r_01 [{shapeType=Capsule, length=7, radius=2, shapeName=dyn_hairSide_r_01_capsule, center=<3.0535 0 -1.1938>, rotation=<rot 0 0 90>}]
 dyn_hairSide_r_02 [{shapeType=Box, extentX=6, extentY=8, extentZ=3, shapeName=dyn_hairSide_r_02_box, center=<1.7632 0 -2.005>}]
 head [{shapeType=Capsule, length=0, radius=9.8227, shapeName=head_capsule, center=<1.9216 -0.0918 -0.2471>, rotation=<rot 4.7724 -82.6317 -90>}]
 neck_01 [{shapeType=Capsule, length=4.1383, radius=5.1617, shapeName=neck_01_capsule, center=<-0.9 0.17 0>, rotation=<rot 0 -90 90>}]
 neck_02 [{shapeType=Capsule, length=3.3111, radius=4.6658, shapeName=neck_02_capsule, rotation=<rot 0 -90 0>}]
 spine_05 [{shapeType=Capsule, length=10.2, radius=8.6, shapeName=spine_05_capsule, center=<0 0.1 0>, rotation=<rot 0 -90 0>}]
 upperarm_l [{shapeType=Capsule, length=19.6, radius=5.0992, shapeName=upperarm_l_capsule, center=<9.3934 0 -0.9>, rotation=<rot 90 0 0>}]
 upperarm_r [{shapeType=Capsule, length=19.6, radius=5.0093, shapeName=upperarm_r_capsule, center=<-9.2657 0 0.9>, rotation=<rot 90 0 0>}]
}
constraints [
 {
  bone1Name=clavicle_l
  bone2Name=spine_05
  swing1LimitDegrees=45
  swing1Motion=Locked
  swing2LimitDegrees=45
  swing2Motion=Locked
  twistLimitDegrees=45
  twistMotion=Locked
 }
 {
  bone1Name=upperarm_l
  bone2Name=clavicle_l
  swing1LimitDegrees=45
  swing1Motion=Locked
  swing2LimitDegrees=45
  swing2Motion=Locked
  twistLimitDegrees=45
  twistMotion=Locked
 }
 {
  bone1Name=clavicle_r
  bone2Name=spine_05
  swing1LimitDegrees=45
  swing1Motion=Locked
  swing2LimitDegrees=45
  swing2Motion=Locked
  twistLimitDegrees=45
  twistMotion=Locked
 }
 {
  bone1Name=upperarm_r
  bone2Name=clavicle_r
  swing1LimitDegrees=45
  swing1Motion=Locked
  swing2LimitDegrees=45
  swing2Motion=Locked
  twistLimitDegrees=45
  twistMotion=Locked
 }
 {
  bone1Name=neck_01
  bone2Name=spine_05
  swing1LimitDegrees=45
  swing1Motion=Locked
  swing2LimitDegrees=45
  swing2Motion=Locked
  twistLimitDegrees=45
  twistMotion=Locked
 }
 {
  bone1Name=neck_02
  bone2Name=neck_01
  swing1LimitDegrees=45
  swing1Motion=Locked
  swing2LimitDegrees=45
  swing2Motion=Locked
  twistLimitDegrees=45
  twistMotion=Locked
 }
 {
  bone1Name=head
  bone2Name=neck_02
  swing1LimitDegrees=45
  swing1Motion=Locked
  swing2LimitDegrees=45
  swing2Motion=Locked
  twistLimitDegrees=45
  twistMotion=Locked
 }
 {
  bone1Name=base_hairBackSide_l_01
  bone2Name=head
  swing1LimitDegrees=45
  swing1Motion=Locked
  swing2LimitDegrees=45
  swing2Motion=Locked
  twistLimitDegrees=45
  twistMotion=Locked
 }
 {bone1Name=dyn_hairBackSide_l_01, bone2Name=base_hairBackSide_l_01, swing1LimitDegrees=15, swing2LimitDegrees=25, twistLimitDegrees=5}
 {bone1Name=dyn_hairBackSide_l_02, bone2Name=dyn_hairBackSide_l_01, swing1LimitDegrees=15, swing2LimitDegrees=25, twistLimitDegrees=5}
 {
  bone1Name=base_hairBackSide_r_01
  bone2Name=head
  swing1LimitDegrees=45
  swing1Motion=Locked
  swing2LimitDegrees=45
  swing2Motion=Locked
  twistLimitDegrees=45
  twistMotion=Locked
 }
 {bone1Name=dyn_hairBackSide_r_01, bone2Name=base_hairBackSide_r_01, swing1LimitDegrees=15, swing2LimitDegrees=25, twistLimitDegrees=5}
 {bone1Name=dyn_hairBackSide_r_02, bone2Name=dyn_hairBackSide_r_01, swing1LimitDegrees=15, swing2LimitDegrees=25, twistLimitDegrees=5}
 {
  bone1Name=base_hairFrontSide_l_01
  bone2Name=head
  swing1LimitDegrees=45
  swing1Motion=Locked
  swing2LimitDegrees=45
  swing2Motion=Locked
  twistLimitDegrees=45
  twistMotion=Locked
 }
 {bone1Name=dyn_hairFrontSide_l_01, bone2Name=base_hairFrontSide_l_01, swing1LimitDegrees=15, swing2LimitDegrees=2, twistLimitDegrees=0}
 {bone1Name=dyn_hairFrontSide_l_02, bone2Name=dyn_hairFrontSide_l_01, swing1LimitDegrees=10, swing2LimitDegrees=10, twistLimitDegrees=5}
 {
  bone1Name=base_hairFront_l_01
  bone2Name=head
  swing1LimitDegrees=45
  swing1Motion=Locked
  swing2LimitDegrees=45
  swing2Motion=Locked
  twistLimitDegrees=45
  twistMotion=Locked
 }
 {bone1Name=dyn_hairFront_l_01, bone2Name=base_hairFront_l_01, swing1LimitDegrees=15, swing2LimitDegrees=10, twistLimitDegrees=5}
 {bone1Name=dyn_hairFront_l_02, bone2Name=dyn_hairFront_l_01, swing1LimitDegrees=15, swing2LimitDegrees=10, twistLimitDegrees=5}
 {bone1Name=dyn_hairFront_l_03, bone2Name=dyn_hairFront_l_02, swing1LimitDegrees=15, swing2LimitDegrees=10, twistLimitDegrees=5}
 {
  bone1Name=base_hairFront_r_01
  bone2Name=head
  swing1LimitDegrees=45
  swing1Motion=Locked
  swing2LimitDegrees=45
  swing2Motion=Locked
  twistLimitDegrees=45
  twistMotion=Locked
 }
 {bone1Name=dyn_hairFront_r_01, bone2Name=base_hairFront_r_01, swing1LimitDegrees=15, swing2LimitDegrees=10, twistLimitDegrees=5}
 {bone1Name=dyn_hairFront_r_02, bone2Name=dyn_hairFront_r_01, swing1LimitDegrees=15, swing2LimitDegrees=10, twistLimitDegrees=5}
 {bone1Name=dyn_hairFront_r_03, bone2Name=dyn_hairFront_r_02, swing1LimitDegrees=15, swing2LimitDegrees=10, twistLimitDegrees=5}
 {
  bone1Name=base_hairSide_l_01
  bone2Name=head
  swing1LimitDegrees=45
  swing1Motion=Locked
  swing2LimitDegrees=45
  swing2Motion=Locked
  twistLimitDegrees=45
  twistMotion=Locked
 }
 {bone1Name=dyn_hairSide_l_01, bone2Name=base_hairSide_l_01, swing1LimitDegrees=10, swing2LimitDegrees=12, twistLimitDegrees=0}
 {
  bone1Name=base_hairSide_r_01
  bone2Name=head
  swing1LimitDegrees=45
  swing1Motion=Locked
  swing2LimitDegrees=45
  swing2Motion=Locked
  twistLimitDegrees=45
  twistMotion=Locked
 }
 {bone1Name=dyn_hairSide_r_01, bone2Name=base_hairSide_r_01, swing1LimitDegrees=10, swing2LimitDegrees=12, twistLimitDegrees=0}
 {bone1Name=dyn_hairSide_r_02, bone2Name=dyn_hairSide_r_01, swing1LimitDegrees=16, swing2LimitDegrees=15, twistLimitDegrees=0}
 {bone1Name=dyn_hairSide_l_02, bone2Name=dyn_hairSide_l_01, swing1LimitDegrees=16, swing2LimitDegrees=15, twistLimitDegrees=0}
 {bone1Name=dyn_hairFrontSide_l_03, bone2Name=dyn_hairFrontSide_l_02, swing1LimitDegrees=20, swing2LimitDegrees=12, twistLimitDegrees=5}
]
