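# Physics asset for the female medium MarineCarve face accessory mesh, prefixed OLD, that sets up hair physics around the head and shoulders.
# Eight base hair bodies lock to the head, each anchoring a two- or three-capsule dynamic chain with small twist and swing limits, tightening toward the tips on the three-segment front chains.
# Head, neck, spine, clavicle and upper arm capsules form the body shapes, all joints locked except a free head-to-neck.
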PhysicsAsset /BRCosmetics/Characters/Player/Female/Medium/Bodies/F_MED_MarineCarve/Meshes/Parts/OLD_F_MED_MarineCarve_FaceAcc_Physics
bodies {
 base_hairBackSide_l_01 [{shapeType=Capsule, length=0.505, radius=0.505, shapeName=base_hairBackSide_l_01_capsule, rotation=<rot 0 -90 0>}]
 base_hairBackSide_r_01 [{shapeType=Capsule, length=0.505, radius=0.505, shapeName=base_hairBackSide_r_01_capsule, rotation=<rot 0 -90 0>}]
 base_hairBack_01 [{shapeType=Capsule, length=0.505, radius=0.505, shapeName=base_hairBack_01_capsule, rotation=<rot 0 -90 0>}]
 base_hairFrontSide_l_01 [{shapeType=Capsule, length=0.505, radius=0.505, shapeName=base_hairFrontSide_l_01_capsule, rotation=<rot 0 -90 0>}]
 base_hairFront_l_01 [{shapeType=Capsule, length=0.505, radius=0.505, shapeName=base_hairFront_l_01_capsule, rotation=<rot 0 -90 0>}]
 base_hairFront_r_01 [{shapeType=Capsule, length=0.505, radius=0.505, shapeName=base_hairFront_r_01_capsule, rotation=<rot 0 -90 0>}]
 base_hairSide_l_01 [{shapeType=Capsule, length=0.505, radius=0.505, shapeName=base_hairSide_l_01_capsule, rotation=<rot 0 -90 0>}]
 base_hairSide_r_01 [{shapeType=Capsule, length=0.505, radius=0.505, shapeName=base_hairSide_r_01_capsule, rotation=<rot 0 -90 0>}]
 clavicle_l [{shapeType=Capsule, length=10, radius=6, shapeName=clavicle_l_capsule, rotation=<rot 90 0 0>}]
 clavicle_r [{shapeType=Capsule, length=10, radius=6, shapeName=clavicle_r_capsule, rotation=<rot 90 0 0>}]
 dyn_hairBackSide_l_01 [{shapeType=Capsule, length=5, radius=2, shapeName=dyn_hairBackSide_l_01_capsule, center=<5 -1 0>, rotation=<rot 90 0 0>}]
 dyn_hairBackSide_l_02 [{shapeType=Capsule, length=2, radius=2, shapeName=dyn_hairBackSide_l_02_capsule, center=<5 -2.5 0>, rotation=<rot 90 0 0>}]
 dyn_hairBackSide_r_01 [{shapeType=Capsule, length=5, radius=2, shapeName=dyn_hairBackSide_r_01_capsule, center=<5 -1 0>, rotation=<rot 90 0 0>}]
 dyn_hairBackSide_r_02 [{shapeType=Capsule, length=2, radius=2, shapeName=dyn_hairBackSide_r_02_capsule, center=<5 -2.5 0>, rotation=<rot 90 0 0>}]
 dyn_hairBack_01 [{shapeType=Capsule, length=5, radius=2, shapeName=dyn_hairBack_01_capsule, center=<5 -1 0>, rotation=<rot 90 0 0>}]
 dyn_hairBack_02 [{shapeType=Capsule, length=2, radius=2, shapeName=dyn_hairBack_02_capsule, center=<5 0 0>, rotation=<rot 90 0 0>}]
 dyn_hairFrontSide_l_01 [{shapeType=Capsule, length=6, radius=2, shapeName=dyn_hairFrontSide_l_01_capsule, center=<3 -2 0>, rotation=<rot 90 0 0>}]
 dyn_hairFrontSide_l_02 [{shapeType=Capsule, length=6, radius=2, shapeName=dyn_hairFrontSide_l_02_capsule, center=<3 -2 0>, rotation=<rot 90 0 0>}]
 dyn_hairFront_l_01 [{shapeType=Capsule, length=4, radius=1, shapeName=dyn_hairFront_l_01_capsule, center=<3 -1.5 0>, rotation=<rot 90 0 0>}]
 dyn_hairFront_l_02 [{shapeType=Capsule, length=4, radius=1, shapeName=dyn_hairFront_l_02_capsule, center=<3 -2.5 0>, rotation=<rot 90 0 0>}]
 dyn_hairFront_l_03 [{shapeType=Capsule, length=4, radius=1, shapeName=dyn_hairFront_l_03_capsule, center=<3 -2.5 0>, rotation=<rot 90 0 0>}]
 dyn_hairFront_r_01 [{shapeType=Capsule, length=3, radius=1, shapeName=dyn_hairFront_r_01_capsule, center=<2 0 2.5>, rotation=<rot 90 -24.3552 0>}]
 dyn_hairFront_r_02 [{shapeType=Capsule, length=3, radius=1, shapeName=dyn_hairFront_r_02_capsule, center=<1.6105 0 3>, rotation=<rot 90 0 0>}]
 dyn_hairFront_r_03 [{shapeType=Capsule, length=3, radius=1, shapeName=dyn_hairFront_r_03_capsule, center=<2 0 3.5>, rotation=<rot 90 2.5816 0>}]
 dyn_hairSide_l_01 [{shapeType=Capsule, length=5, radius=2, shapeName=dyn_hairSide_l_01_capsule, center=<5 -1 0>, rotation=<rot 90 0 0>}]
 dyn_hairSide_l_02 [{shapeType=Capsule, length=2, radius=2, shapeName=dyn_hairSide_l_02_capsule, center=<5 0 2>, rotation=<rot -60 0 0>}]
 dyn_hairSide_r_01 [{shapeType=Capsule, length=5, radius=2, shapeName=dyn_hairSide_r_01_capsule, center=<5 -2 0>, rotation=<rot 90 0 0>}]
 dyn_hairSide_r_02 [{shapeType=Capsule, length=2, radius=2, shapeName=dyn_hairSide_r_02_capsule, center=<5 -1 -2>, rotation=<rot 60 0 0>}]
 head [{shapeType=Capsule, length=0, radius=9.1452, shapeName=head_capsule, center=<0.0339 3.128 -0.3357>, rotation=<rot 78.9943 -84.0878 -83.9777>}]
 neck_02 [{shapeType=Capsule, length=3.0203, radius=8.7193, shapeName=neck_02_capsule, center=<3.5762 0 0>, rotation=<rot 0 -90 0>}]
 spine_05 [{shapeType=Capsule, length=15, radius=0, shapeName=spine_05_capsule, center=<2.4464 -1.2057 0>, rotation=<rot 0 -90 0>}]
 upperarm_l [{shapeType=Capsule, length=10, radius=6, shapeName=upperarm_l_capsule, center=<6 0 0>, rotation=<rot 90 0 0>}]
 upperarm_r [{shapeType=Capsule, length=10, radius=6, shapeName=upperarm_r_capsule, center=<-6 0 0>, rotation=<rot 90 0 0>}]
}
constraints [
 {
  bone1Name=upperarm_l
  bone2Name=clavicle_l
  swing1LimitDegrees=45
  swing1Motion=Locked
  swing2LimitDegrees=45
  swing2Motion=Locked
  twistLimitDegrees=45
  twistMotion=Locked
 }
 {
  bone1Name=upperarm_r
  bone2Name=clavicle_r
  swing1LimitDegrees=45
  swing1Motion=Locked
  swing2LimitDegrees=45
  swing2Motion=Locked
  twistLimitDegrees=45
  twistMotion=Locked
 }
 {
  bone1Name=head
  bone2Name=neck_02
  swing1LimitDegrees=45
  swing1Motion=Free
  swing2LimitDegrees=45
  swing2Motion=Free
  twistLimitDegrees=45
  twistMotion=Free
 }
 {
  bone1Name=base_hairBackSide_l_01
  bone2Name=head
  swing1LimitDegrees=45
  swing1Motion=Locked
  swing2LimitDegrees=45
  swing2Motion=Locked
  twistLimitDegrees=45
  twistMotion=Locked
 }
 {bone1Name=dyn_hairBackSide_l_01, bone2Name=base_hairBackSide_l_01, swing1LimitDegrees=30, swing2LimitDegrees=10, twistLimitDegrees=5}
 {bone1Name=dyn_hairBackSide_l_02, bone2Name=dyn_hairBackSide_l_01, swing1LimitDegrees=30, swing2LimitDegrees=10, twistLimitDegrees=5}
 {
  bone1Name=base_hairBackSide_r_01
  bone2Name=head
  swing1LimitDegrees=45
  swing1Motion=Locked
  swing2LimitDegrees=45
  swing2Motion=Locked
  twistLimitDegrees=45
  twistMotion=Locked
 }
 {bone1Name=dyn_hairBackSide_r_01, bone2Name=base_hairBackSide_r_01, swing1LimitDegrees=30, swing2LimitDegrees=10, twistLimitDegrees=5}
 {bone1Name=dyn_hairBackSide_r_02, bone2Name=dyn_hairBackSide_r_01, swing1LimitDegrees=30, swing2LimitDegrees=10, twistLimitDegrees=5}
 {
  bone1Name=base_hairBack_01
  bone2Name=head
  swing1LimitDegrees=45
  swing1Motion=Locked
  swing2LimitDegrees=45
  swing2Motion=Locked
  twistLimitDegrees=45
  twistMotion=Locked
 }
 {bone1Name=dyn_hairBack_01, bone2Name=base_hairBack_01, swing1LimitDegrees=30, swing2LimitDegrees=10, twistLimitDegrees=5}
 {bone1Name=dyn_hairBack_02, bone2Name=dyn_hairBack_01, swing1LimitDegrees=30, swing2LimitDegrees=10, twistLimitDegrees=5}
 {
  bone1Name=base_hairFrontSide_l_01
  bone2Name=head
  swing1LimitDegrees=45
  swing1Motion=Locked
  swing2LimitDegrees=45
  swing2Motion=Locked
  twistLimitDegrees=45
  twistMotion=Locked
 }
 {bone1Name=dyn_hairFrontSide_l_01, bone2Name=base_hairFrontSide_l_01, swing1LimitDegrees=20, swing2LimitDegrees=20, twistLimitDegrees=5}
 {bone1Name=dyn_hairFrontSide_l_02, bone2Name=dyn_hairFrontSide_l_01, swing1LimitDegrees=20, swing2LimitDegrees=20, twistLimitDegrees=5}
 {
  bone1Name=base_hairFront_l_01
  bone2Name=head
  swing1LimitDegrees=45
  swing1Motion=Locked
  swing2LimitDegrees=45
  swing2Motion=Locked
  twistLimitDegrees=45
  twistMotion=Locked
 }
 {bone1Name=dyn_hairFront_l_01, bone2Name=base_hairFront_l_01, swing1LimitDegrees=20, swing2LimitDegrees=20, twistLimitDegrees=5}
 {bone1Name=dyn_hairFront_l_02, bone2Name=dyn_hairFront_l_01, swing1LimitDegrees=15, swing2LimitDegrees=15, twistLimitDegrees=5}
 {bone1Name=dyn_hairFront_l_03, bone2Name=dyn_hairFront_l_02, swing1LimitDegrees=10, swing2LimitDegrees=10, twistLimitDegrees=5}
 {
  bone1Name=base_hairFront_r_01
  bone2Name=head
  swing1LimitDegrees=45
  swing1Motion=Locked
  swing2LimitDegrees=45
  swing2Motion=Locked
  twistLimitDegrees=45
  twistMotion=Locked
 }
 {bone1Name=dyn_hairFront_r_01, bone2Name=base_hairFront_r_01, swing1LimitDegrees=20, swing2LimitDegrees=20, twistLimitDegrees=5}
 {bone1Name=dyn_hairFront_r_02, bone2Name=dyn_hairFront_r_01, swing1LimitDegrees=15, swing2LimitDegrees=15, twistLimitDegrees=5}
 {bone1Name=dyn_hairFront_r_03, bone2Name=dyn_hairFront_r_02, swing1LimitDegrees=10, swing2LimitDegrees=10, twistLimitDegrees=5}
 {
  bone1Name=base_hairSide_l_01
  bone2Name=head
  swing1LimitDegrees=45
  swing1Motion=Locked
  swing2LimitDegrees=45
  swing2Motion=Locked
  twistLimitDegrees=45
  twistMotion=Locked
 }
 {bone1Name=dyn_hairSide_l_01, bone2Name=base_hairSide_l_01, swing1LimitDegrees=30, swing2LimitDegrees=10, twistLimitDegrees=5}
 {bone1Name=dyn_hairSide_l_02, bone2Name=dyn_hairSide_l_01, swing1LimitDegrees=30, swing2LimitDegrees=10, twistLimitDegrees=5}
 {
  bone1Name=base_hairSide_r_01
  bone2Name=head
  swing1LimitDegrees=45
  swing1Motion=Locked
  swing2LimitDegrees=45
  swing2Motion=Locked
  twistLimitDegrees=45
  twistMotion=Locked
 }
 {bone1Name=dyn_hairSide_r_01, bone2Name=base_hairSide_r_01, swing1LimitDegrees=30, swing2LimitDegrees=10, twistLimitDegrees=5}
 {bone1Name=dyn_hairSide_r_02, bone2Name=dyn_hairSide_r_01, swing1LimitDegrees=30, swing2LimitDegrees=10, twistLimitDegrees=5}
 {
  bone1Name=neck_02
  bone2Name=spine_05
  swing1LimitDegrees=45
  swing1Motion=Locked
  swing2LimitDegrees=45
  swing2Motion=Locked
  twistLimitDegrees=45
  twistMotion=Locked
 }
]
